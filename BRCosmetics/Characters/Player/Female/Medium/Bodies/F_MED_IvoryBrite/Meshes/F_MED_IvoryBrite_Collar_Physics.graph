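# Physics asset for the F_MED_IvoryBrite female body's collar, defining capsule bodies and joints for five neck flaps, left, right and back.
# Each flap is a two-link chain from spine_05 through a small base body to a larger dyn body, with joints limited to swing2 so they bend like hinges.
# Head, neck, upper spine and upper-arm capsules give body shapes around the flaps, joined by free angular constraints.
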PhysicsAsset /BRCosmetics/Characters/Player/Female/Medium/Bodies/F_MED_IvoryBrite/Meshes/F_MED_IvoryBrite_Collar_Physics
bodies {
 base_neckFlap_01_l [{shapeType=Capsule, length=0.505, radius=0.505, shapeName=base_neckFlap_01_l_capsule, rotation=<rot 0 -90 0>}]
 base_neckFlap_01_r [{shapeType=Capsule, length=0.505, radius=0.505, shapeName=base_neckFlap_01_r_capsule, rotation=<rot 0 -90 0>}]
 base_neckFlap_02_l [{shapeType=Capsule, length=0.505, radius=0.505, shapeName=base_neckFlap_02_l_capsule, rotation=<rot 0 -90 0>}]
 base_neckFlap_02_r [{shapeType=Capsule, length=0.505, radius=0.505, shapeName=base_neckFlap_02_r_capsule, rotation=<rot 0 -90 0>}]
 base_neckFlap_b [{shapeType=Capsule, length=0.505, radius=0.505, shapeName=base_neckFlap_b_capsule, rotation=<rot 0 -90 0>}]
 dyn_neckFlap_01_l [{
   shapeType=Capsule
   length=5.6363
   radius=1.5272
   shapeName=dyn_neckFlap_b_capsule
   center=<4.6534 1.3055 1.4253>
   rotation=<rot -0.0164 175.5863 -83.094>
  }]
 dyn_neckFlap_01_r [
  {
   shapeType=Capsule
   length=6.197
   radius=0.7708
   shapeName=dyn_neckFlap_b_capsule
   center=<5.3626 -1.9657 0.5546>
   rotation=<rot 0 175.5016 -90.5161>
  }
  {
   shapeType=Capsule
   length=5.0204
   radius=0.9597
   shapeName=dyn_neckFlap_b_capsule
   center=<4.0206 -2.2594 0.3727>
   rotation=<rot 0 175.5015 -88.4243>
  }
  {
   shapeType=Capsule
   length=5.98
   radius=0.7708
   shapeName=dyn_neckFlap_b_capsule
   center=<2.3455 -3.1137 0.5554>
   rotation=<rot 0.0575 -178.1045 -90.5129>
  }
 ]
 dyn_neckFlap_02_l [{
   shapeType=Capsule
   length=6.6812
   radius=1.3595
   shapeName=dyn_neckFlap_b_capsule
   center=<3.48 0.4642 0.148>
   rotation=<rot 0 175.5018 -77.6243>
  }]
 dyn_neckFlap_02_r [{
   shapeType=Capsule
   length=7.11
   radius=0.9307
   shapeName=dyn_neckFlap_b_capsule
   center=<3.1461 -1.9345 -0.0853>
   rotation=<rot -0.0593 179.9968 -90.6954>
  }]
 dyn_neckFlap_b [{
   shapeType=Capsule
   length=6.5136
   radius=1.5272
   shapeName=dyn_neckFlap_b_capsule
   center=<4.5429 -0.0988 0.5437>
   rotation=<rot 0 175.5018 -77.6243>
  }]
 head [
  {shapeType=Capsule, length=4.5, radius=6.5, shapeName=head_capsule, center=<2.5512 0.4621 0>, rotation=<rot 0 -95.6039 90>}
  {shapeType=Capsule, length=1.5, radius=3, shapeName=head_capsule, center=<-3.8051 5.9177 0>}
 ]
 neck_01 [{shapeType=Capsule, length=0.9288, radius=4.4175, shapeName=neck_02_capsule, center=<1.1663 0.483 0>, rotation=<rot 0 -90 90>}]
 neck_02 [{shapeType=Capsule, length=0.8301, radius=4.1749, shapeName=neck_02_capsule, center=<0.462 0.1363 0>, rotation=<rot 0 -90 90>}]
 spine_05 [{shapeType=Capsule, length=8.7304, radius=9.7452, shapeName=spine_05_capsule, center=<1.3476 -1.9358 -0.1616>, rotation=<rot 0 -180 -180>}]
 upperarm_l [{
   shapeType=Capsule
   length=25.2103
   radius=4.8166
   shapeName=upperarm_l_capsule
   center=<10.0639 -0.5627 -0.5326>
   rotation=<rot -21.3454 -89.9929 -91.2354>
  }]
 upperarm_r [{
   shapeType=Capsule
   length=17.6551
   radius=3.8104
   shapeName=upperarm_r_capsule
   center=<-6.7333 1.0991 1.3566>
   rotation=<rot -21.1347 -86.7123 85.4583>
  }]
 upperarm_twist_01_r [{
   shapeType=Capsule
   length=11.0749
   radius=3.8104
   shapeName=upperarm_r_capsule
   center=<0.8929 -0.1597 1.4463>
   rotation=<rot -21.1347 -86.7123 87.001>
  }]
}
constraints [
 {
  bone1Name=neck_01
  bone2Name=spine_05
  swing1LimitDegrees=45
  swing1Motion=Free
  swing2LimitDegrees=45
  swing2Motion=Free
  twistLimitDegrees=45
  twistMotion=Free
 }
 {
  bone1Name=neck_02
  bone2Name=neck_01
  swing1LimitDegrees=45
  swing1Motion=Free
  swing2LimitDegrees=45
  swing2Motion=Free
  twistLimitDegrees=45
  twistMotion=Free
 }
 {
  bone1Name=head
  bone2Name=neck_02
  swing1LimitDegrees=45
  swing1Motion=Free
  swing2LimitDegrees=45
  swing2Motion=Free
  twistLimitDegrees=45
  twistMotion=Free
 }
 {
  bone1Name=upperarm_r
  bone2Name=spine_05
  swing1LimitDegrees=45
  swing1Motion=Free
  swing2LimitDegrees=45
  swing2Motion=Free
  twistLimitDegrees=45
  twistMotion=Free
 }
 {
  bone1Name=upperarm_l
  bone2Name=spine_05
  swing1LimitDegrees=45
  swing1Motion=Free
  swing2LimitDegrees=45
  swing2Motion=Free
  twistLimitDegrees=45
  twistMotion=Free
 }
 {bone1Name=base_neckFlap_01_l, bone2Name=spine_05, swing1LimitDegrees=0, swing2LimitDegrees=35, twistLimitDegrees=0}
 {bone1Name=dyn_neckFlap_01_l, bone2Name=base_neckFlap_01_l, swing1LimitDegrees=0, swing2LimitDegrees=45, twistLimitDegrees=0}
 {bone1Name=base_neckFlap_01_r, bone2Name=spine_05, swing1LimitDegrees=0, swing2LimitDegrees=35, twistLimitDegrees=0}
 {bone1Name=dyn_neckFlap_01_r, bone2Name=base_neckFlap_01_r, swing1LimitDegrees=0, swing2LimitDegrees=45, twistLimitDegrees=0}
 {bone1Name=base_neckFlap_02_l, bone2Name=spine_05, swing1LimitDegrees=0, swing2LimitDegrees=35, twistLimitDegrees=0}
 {bone1Name=dyn_neckFlap_02_l, bone2Name=base_neckFlap_02_l, swing1LimitDegrees=0, swing2LimitDegrees=35, twistLimitDegrees=0}
 {bone1Name=base_neckFlap_02_r, bone2Name=spine_05, swing1LimitDegrees=0, swing2LimitDegrees=35, twistLimitDegrees=0}
 {bone1Name=dyn_neckFlap_02_r, bone2Name=base_neckFlap_02_r, swing1LimitDegrees=0, swing2LimitDegrees=35, twistLimitDegrees=0}
 {bone1Name=base_neckFlap_b, bone2Name=spine_05, swing1LimitDegrees=0, swing2LimitDegrees=35, twistLimitDegrees=0}
 {bone1Name=dyn_neckFlap_b, bone2Name=base_neckFlap_b, swing1LimitDegrees=0, swing2LimitDegrees=35, twistLimitDegrees=0}
 {
  bone1Name=upperarm_twist_01_r
  bone2Name=upperarm_r
  swing1LimitDegrees=45
  swing1Motion=Free
  swing2LimitDegrees=45
  swing2Motion=Free
  twistLimitDegrees=45
  twistMotion=Free
 }
]
