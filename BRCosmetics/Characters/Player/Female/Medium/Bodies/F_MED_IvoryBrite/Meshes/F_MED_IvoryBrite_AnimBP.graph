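# Anim Blueprint for the F_MED_IvoryBrite female medium character body, adding procedural rig and secondary motion to the pose copied from the attached parent mesh.
# The copied pose runs through a Control Rig, then chained spring-limited AnimDynamics nodes for canisters, their rings and the leg knife.
# A RigidBody node then simulates the collar in spine_05 space.
# Each update recomputes GravityOverride using a nativized state function and parameter asset.
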AnimBlueprint /BRCosmetics/Characters/Player/Female/Medium/Bodies/F_MED_IvoryBrite/Meshes/F_MED_IvoryBrite_AnimBP
target_skeleton=&/F_MED_IvoryBrite_Skeleton
graphs [
 AnimGraph {nodes [
   Output Pose(Result=Use cached pose 'DynamicsCache'()) {layerGroup=None}
   n1 = Control Rig!(
    Source=Copy Pose From Mesh() {bUseAttachedParent=true}
    Alpha
    RBAN Enabled:bool=false
    isSkyFloating:bool=false
    isFrontEnd:bool=false
    isPlayingEmote:bool=false
    isMeleeSwinging:bool=false
    isParagliding:bool=false
   ) -> (Pose) {controlRigAssetReference {blueprintRigClass=&/F_MED_IvoryBrite_PRE_CtrlRig.F_MED_IvoryBrite_PRE_CtrlRig_C}}
   Save cached pose 'CRCache_Input'(Pose=@n1) {cacheName=CRCache_Input}
   n2 = Anim Dynamics - Bone: dyn_canister_f_01!(ComponentPose=Local To Component(LocalPose=Use cached pose 'CRCache_Input'()), Alpha=1, GravityOverride:Vector="0,0,-980") -> (Pose) {
    angularDampingOverride=0.99
    angularSpringConstant=5000
    bAngularSpring=true
    bOverrideAngularDamping=true
    bUseGravityOverride=true
    gravityScale=0.35
    simulationSpace=RootRelative
    alphaBoolBlend {blendOption=Linear}
    boundBone {boneName=dyn_canister_f_01}
    componentAppliedLinearAccClamp=<10000 10000 10000>
    componentLinearAccScale=<1 1 1>
    componentLinearVelScale=<1 1 1>
    gravityOverride {z=-980}
    physicsBodyDefinitions [{
      boundBone {boneName=dyn_canister_f_01}
      boxExtents=<26 46 50>
      constraintSetup {angularLimitsMax=<5 20 5>, angularLimitsMin=<-5 -50 -15>, angularTarget=<1 0 0>}
      localJointOffset=<1.2 0 0>
     }]
    relativeSpaceBone {boneName=pelvis}
    simSpaceSettings {maxAngularAcceleration=15, maxAngularVelocity=5, simSpaceAngularAlpha=1}
   }
   n3 = Anim Dynamics - Bone: dyn_canister_f_ring!(ComponentPose=@n2, Alpha=1, GravityOverride:Vector="0,0,-980") -> (Pose) {
    angularDampingOverride=1
    angularSpringConstant=1000
    bAngularSpring=true
    bOverrideAngularDamping=true
    bUseGravityOverride=true
    gravityScale=0.35
    simulationSpace=RootRelative
    boundBone {boneName=dyn_canister_f_ring}
    componentAppliedLinearAccClamp=<10000 10000>
    componentLinearAccScale=<1 1 1>
    componentLinearVelScale=<1 1 1>
    gravityOverride {z=-980}
    physicsBodyDefinitions [{
      boundBone {boneName=dyn_canister_f_ring}
      boxExtents=<6 8 8>
      constraintSetup {angularLimitsMax=<0 10 15>, angularLimitsMin=<0 -90 -15>, angularTarget=<1 0 0>}
      localJointOffset=<0.5 0 0>
     }]
    simSpaceSettings {maxAngularAcceleration=15, maxAngularVelocity=5, simSpaceAngularAlpha=1}
   }
   n4 = Anim Dynamics - Bone: dyn_canister_b_01!(ComponentPose=@n3, Alpha=1, GravityOverride:Vector="0,0,-980") -> (Pose) {
    angularDampingOverride=0.99
    angularSpringConstant=5000
    bAngularSpring=true
    bOverrideAngularDamping=true
    bUseGravityOverride=true
    gravityScale=0.35
    simulationSpace=RootRelative
    alphaBoolBlend {blendOption=Linear}
    boundBone {boneName=dyn_canister_b_01}
    componentAppliedLinearAccClamp=<10000 10000 10000>
    componentLinearAccScale=<1 1 1>
    componentLinearVelScale=<1 1 1>
    gravityOverride {z=-980}
    physicsBodyDefinitions [{
      boundBone {boneName=dyn_canister_b_01}
      boxExtents=<22 32 46>
      constraintSetup {angularLimitsMax=<5 20 15>, angularLimitsMin=<-5 -50 -2>, angularTarget=<1 0 0>}
      localJointOffset=<1.5 0 0>
     }]
    relativeSpaceBone {boneName=pelvis}
    simSpaceSettings {maxAngularAcceleration=15, maxAngularVelocity=5, simSpaceAngularAlpha=1}
   }
   n5 = Anim Dynamics - Bone: dyn_canister_b_ring!(ComponentPose=@n4, Alpha=1, GravityOverride:Vector="0,0,-980") -> (Pose) {
    angularDampingOverride=1
    angularSpringConstant=1000
    bAngularSpring=true
    bOverrideAngularDamping=true
    bUseGravityOverride=true
    gravityScale=0.35
    simulationSpace=RootRelative
    boundBone {boneName=dyn_canister_b_ring}
    componentAppliedLinearAccClamp=<10000 10000>
    componentLinearAccScale=<1 1 1>
    componentLinearVelScale=<1 1 1>
    gravityOverride {z=-980}
    physicsBodyDefinitions [{
      boundBone {boneName=dyn_canister_b_ring}
      boxExtents=<6 8 8>
      constraintSetup {angularLimitsMax=<0 10 15>, angularLimitsMin=<0 -90 -15>, angularTarget=<1 0 0>}
      localJointOffset=<0.5 0 0>
     }]
    simSpaceSettings {maxAngularAcceleration=15, maxAngularVelocity=5, simSpaceAngularAlpha=1}
   }
   n6 = Anim Dynamics - Bone: dyn_legKnife_01(ComponentPose=@n5, Alpha=1, ExternalForce="0,0,0") {
    angularDampingOverride=1
    angularSpringConstant=1000
    bAngularSpring=true
    bOverrideAngularDamping=true
    bUseGravityOverride=true
    gravityScale=0.35
    simulationSpace=RootRelative
    boundBone {boneName=dyn_legKnife_01}
    componentLinearAccScale=<1 1 1>
    componentLinearVelScale=<1 1 1>
    gravityOverride {z=-980}
    physicsBodyDefinitions [{
      boundBone {boneName=dyn_legKnife_01}
      boxExtents=<10 20 20>
      constraintSetup {angularLimitsMax=<0 2 7>, angularLimitsMin=<0 -4 -7>, angularTarget=<1 0 0>}
      localJointOffset=<1 0 0>
     }]
   }
   n7 = RigidBody!(ComponentPose=@n6, Alpha, OverrideWorldGravity:Vector="0,0,-980") -> (Pose) {
    bOverrideWorldGravity=true
    overridePhysicsAsset=&/F_MED_IvoryBrite_Collar_Physics
    simulationSpace=BaseBoneSpace
    baseBoneRef {boneName=spine_05}
    simSpaceSettings {maxAngularAcceleration=2000, maxAngularVelocity=700, maxLinearAcceleration=1400, maxLinearVelocity=500, worldAlpha=1}
   }
   Save cached pose 'DynamicsCache'(Pose=Component To Local(ComponentPose=@n7)) {cacheName=DynamicsCache}
  ]}
 EventGraph {nodes [
   Event BlueprintUpdateAnimation()
   > n1 = "Gravity Override State Function (Parameter Asset) (Nativized)"(
    CharacterPartAnimInstance=Self-Reference()
    GameTimeInSeconds=GetGameTimeInSeconds()
    Parameters=&/F_MED_IvoryBrite_Parameters
   )
   > Set GravityOverride(GravityOverride=@n1)
  ]}
]
variables [
 GravityOverride {type=Vector}
 head rot {type=Rotator}
 neck01 rot {type=Rotator}
 neck02 rot {type=Rotator}
 Twist Reduction Value {type=number}
 Y Reduction Value {type=number}
 Z Reduction Value {type=number}
]
