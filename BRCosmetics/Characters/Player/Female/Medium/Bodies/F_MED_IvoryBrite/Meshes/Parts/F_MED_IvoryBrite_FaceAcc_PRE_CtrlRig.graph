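# Control Rig for IvoryBrite's face-accessory mesh, posing the ponytail, bandana knot and two ribbons against body motion.
# Head rotation and pelvis height become look and crouch alphas, which join clavicle angles, bone distances and gameplay flags to weight layered corrective transforms.
# Ribbon nulls follow weighted body-bone constraints plus offsets, driving B-spline IK chains after a twist pass.
# Back bling and dynamics flags select alternate corrective branches.
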ControlRigBlueprint /BRCosmetics/Characters/Player/Female/Medium/Bodies/F_MED_IvoryBrite/Meshes/Parts/F_MED_IvoryBrite_FaceAcc_PRE_CtrlRig
graphs [
 CollapseNode_ContainedGraph {nodes [
   RigVMFunctionEntryNode()
   > RigVMFunction_Sequence = RigVMAggregateNode()
   &RigUnit_ParentConstraint(
    AdvancedSettings=(RotationOrderForFilter=XZY)
    Child=(Type=Null,Name="dyn_ribbon_l_01_null")
    Parents=((Item=(Type=Bone,Name="base_ribbon_l_01"),Weight=1))
    Weight=1
    ExecutePin=@RigVMFunction_Sequence.A
   )
   > &RigUnit_ParentConstraint(
    AdvancedSettings=(RotationOrderForFilter=XZY)
    Child=(Type=Null,Name="dyn_ribbon_l_02_null")
    Parents=((Item=(Type=Bone,Name="spine_03"),Weight=0.0046),(Item=(Type=Bone,Name="neck_01"),Weight=0.0053),(Item=(Type=Bone,Name="spine_05"),Weight=0.4807),(Item=(Type=Bone,Name="clavicle_r"),Weight=0.5094))
    Weight=1
   )
   > &RigUnit_ParentConstraint(
    AdvancedSettings=(RotationOrderForFilter=XZY)
    Child=(Type=Null,Name="dyn_ribbon_l_03_null")
    Parents="((Item=(Type=Bone,Name=\"deform_upperarm_up_r\"),Weight=0.0013),(Item=(Type=Bone,Name=\"clavicle_l\"),Weight=0.0028),(Item=(Type=Bone,Name=\"deform_upperarm_bk_r\"),Weight=0.0207),(Item=(Type=Bone,Name=\"upperarm_twist_01_r\"),Weight=0.0304),(Item=(Type=Bone,Name= ...(truncated)"
    Weight=1
   )
   > &RigUnit_ParentConstraint(
    AdvancedSettings=(RotationOrderForFilter=XZY)
    Child=(Type=Null,Name="dyn_ribbon_l_04_null")
    Parents="((Item=(Type=Bone,Name=\"spine_01\"),Weight=0.0083),(Item=(Type=Bone,Name=\"clavicle_l\"),Weight=0.0283),(Item=(Type=Bone,Name=\"upperarm_twist_01_r\"),Weight=0.04),(Item=(Type=Bone,Name=\"spine_02\"),Weight=0.0453),(Item=(Type=Bone,Name=\"spine_03\"),Weight=0.0457) ...(truncated)"
    Weight=1
   )
   > &RigUnit_ParentConstraint(
    AdvancedSettings=(RotationOrderForFilter=XZY)
    Child=(Type=Null,Name="dyn_ribbon_l_05_null")
    Parents="((Item=(Type=Bone,Name=\"clavicle_l\"),Weight=0.0067),(Item=(Type=Bone,Name=\"pelvis\"),Weight=0.0357),(Item=(Type=Bone,Name=\"clavicle_r\"),Weight=0.0559),(Item=(Type=Bone,Name=\"spine_05\"),Weight=0.0657),(Item=(Type=Bone,Name=\"spine_01\"),Weight=0.0755),(Item=(T ...(truncated)"
    Weight=1
   )
   > &RigUnit_ParentConstraint(
    AdvancedSettings=(RotationOrderForFilter=XZY)
    Child=(Type=Null,Name="dyn_ribbon_l_06_null")
    Parents=((Item=(Type=Bone,Name="spine_01"),Weight=0.02),(Item=(Type=Bone,Name="spine_02"),Weight=0.045),(Item=(Type=Bone,Name="pelvis"),Weight=0.953))
    Weight=1
   )
   > &RigUnit_ParentConstraint(
    AdvancedSettings=(RotationOrderForFilter=XZY)
    Child=(Type=Null,Name="dyn_ribbon_l_07_null")
    Parents=((Item=(Type=Bone,Name="pelvis"),Weight=1))
    Weight=1
   )
   > &RigUnit_ParentConstraint(
    AdvancedSettings=(RotationOrderForFilter=XZY)
    Child=(Type=Null,Name="dyn_ribbon_l_08_null")
    Parents=((Item=(Type=Bone,Name="pelvis"),Weight=1))
    Weight=1
   )
   &RigUnit_ParentConstraint(
    AdvancedSettings=(RotationOrderForFilter=XZY)
    Child=(Type=Null,Name="dyn_ribbon_r_01_null")
    Parents=((Item=(Type=Bone,Name="base_ribbon_r_01"),Weight=1))
    Weight=1
    ExecutePin=@RigVMFunction_Sequence.B
   )
   > &RigUnit_ParentConstraint(
    AdvancedSettings=(RotationOrderForFilter=XZY)
    Child=(Type=Null,Name="dyn_ribbon_r_02_null")
    Parents="((Item=(Type=Bone,Name=\"deform_upperarm_bk_r\"),Weight=0.0187),(Item=(Type=Bone,Name=\"upperarm_twist_01_r\"),Weight=0.0343),(Item=(Type=Bone,Name=\"deform_upperarm_up_r\"),Weight=0.0388),(Item=(Type=Bone,Name=\"spine_05\"),Weight=0.1484),(Item=(Type=Bone,Name=\"c ...(truncated)"
    Weight=1
   )
   > &RigUnit_ParentConstraint(
    AdvancedSettings=(RotationOrderForFilter=XZY)
    Child=(Type=Null,Name="dyn_ribbon_r_03_null")
    Parents="((Item=(Type=Bone,Name=\"spine_04\"),Weight=0.0241),(Item=(Type=Bone,Name=\"deform_upperarm_up_r\"),Weight=0.0594),(Item=(Type=Bone,Name=\"spine_05\"),Weight=0.0698),(Item=(Type=Bone,Name=\"upperarm_twist_01_r\"),Weight=0.1507),(Item=(Type=Bone,Name=\"deform_uppera ...(truncated)"
    Weight=1
   )
   > &RigUnit_ParentConstraint(
    AdvancedSettings=(RotationOrderForFilter=XZY)
    Child=(Type=Null,Name="dyn_ribbon_r_04_null")
    Parents="((Item=(Type=Bone,Name=\"spine_02\"),Weight=0.0033),(Item=(Type=Bone,Name=\"spine_03\"),Weight=0.0125),(Item=(Type=Bone,Name=\"deform_upperarm_bk_r\"),Weight=0.0443),(Item=(Type=Bone,Name=\"upperarm_twist_01_r\"),Weight=0.1273),(Item=(Type=Bone,Name=\"spine_05\"),We ...(truncated)"
    Weight=1
   )
   > &RigUnit_ParentConstraint(
    AdvancedSettings=(RotationOrderForFilter=XZY)
    Child=(Type=Null,Name="dyn_ribbon_r_05_null")
    Parents="((Item=(Type=Bone,Name=\"upperarm_twist_01_r\"),Weight=0.0093),(Item=(Type=Bone,Name=\"pelvis\"),Weight=0.036),(Item=(Type=Bone,Name=\"clavicle_r\"),Weight=0.0579),(Item=(Type=Bone,Name=\"spine_05\"),Weight=0.0595),(Item=(Type=Bone,Name=\"spine_01\"),Weight=0.0715), ...(truncated)"
    Weight=1
   )
   > &RigUnit_ParentConstraint(
    AdvancedSettings=(InterpolationType=Average,RotationOrderForFilter=XZY)
    Child=(Type=Null,Name="dyn_ribbon_r_06_null")
    Parents=((Item=(Type=Bone,Name="pelvis"),Weight=1))
    Weight=1
   )
   > &RigUnit_ParentConstraint(
    AdvancedSettings=(InterpolationType=Average,RotationOrderForFilter=XZY)
    Child=(Type=Null,Name="dyn_ribbon_r_07_null")
    Parents=((Item=(Type=Bone,Name="pelvis"),Weight=1))
    Weight=1
   )
   > &RigUnit_ParentConstraint(
    AdvancedSettings=(RotationOrderForFilter=XZY)
    Child=(Type=Null,Name="dyn_ribbon_r_08_null")
    Parents=((Item=(Type=Bone,Name="pelvis"),Weight=1))
    Weight=1
   )
   > &RigUnit_ParentConstraint(
    AdvancedSettings=(RotationOrderForFilter=XZY)
    Child=(Type=Null,Name="dyn_ribbon_r_09_null")
    Parents=((Item=(Type=Bone,Name="pelvis"),Weight=1))
    Weight=1
   )
   RigVMFunctionReturnNode(ExecuteContext=@RigVMFunction_Sequence.C)
  ]}
 CollapseNode_ContainedGraph {nodes [
   RigVMFunctionEntryNode()
   > GetAlphaFromQuatDelta = RigVMFunctionReferenceNode(
    Rotation Order=ZYX
    child=(Type=Bone,Name="head")
    outputMinMax=(X=0,Y=1)
    parent=(Type=Bone,Name="neck_01")
    premult=1
    rotationMinMax=(X=-15,Y=-43)
    triggerAxis=Z
    twistAxis=(X=1,Y=0,Z=0)
   )
   n1 = &RigVMFunction_AnimEvalRichCurve(
    Curve="(EditorCurveData=(Keys=((InterpMode=RCIM_Cubic,TangentMode=RCTM_User,ArriveTangent=0.6623,LeaveTangent=0.6623),(InterpMode=RCIM_Cubic,TangentMode=RCTM_User,Time=0.5,Value=0.4,ArriveTangent=0.8898,LeaveTangent=0.8898),(InterpMode=RCIM_Cubic,TangentMode=RCTM ...(truncated)"
    SourceMinimum=0
    TargetMinimum=0
    Value=@GetAlphaFromQuatDelta.alpha
   )
   > RigVMVariableNode(Value=@n1.Result) {variable=HeadLookingUp}
   > GetAlphaFromQuatDelta_1 = RigVMFunctionReferenceNode(
    Rotation Order=ZYX
    child=(Type=Bone,Name="head")
    outputMinMax=(X=0,Y=1)
    parent=(Type=Bone,Name="neck_01")
    premult=1
    rotationMinMax=(X=-50,Y=-80)
    triggerAxis=Z
    twistAxis=(X=1,Y=0,Z=0)
   )
   > RigVMVariableNode(Value=@GetAlphaFromQuatDelta_1.alpha) {variable=HeadLookingUp_Extreme}
   > GetAlphaFromQuatDelta_2 = RigVMFunctionReferenceNode(
    Rotation Order=ZYX
    child=(Type=Bone,Name="head")
    outputMinMax=(X=0,Y=1)
    parent=(Type=Bone,Name="neck_01")
    premult=1
    rotationMinMax=(X=-20,Y=-63)
    triggerAxis=Z
    twistAxis=(X=1,Y=0,Z=0)
   )
   > RigVMVariableNode(Value=@GetAlphaFromQuatDelta_2.alpha) {variable=HeadLookingUp_Crouching}
   n2 = &RigVMFunction_AnimEvalRichCurve(
    Curve=(EditorCurveData=(Keys=((InterpMode=RCIM_Cubic),(InterpMode=RCIM_Cubic,Time=0.5,Value=1),(Time=0.8))))
    SourceMaximum=2
    SourceMinimum=0
    TargetMinimum=0
    Value=&RigVMFunction_MathDoubleRemap(
     SourceMaximum=75
     SourceMinimum=45
     TargetMaximum=1
     TargetMinimum=0
     bClamp=false
     Value=&RigUnit_GetRelativeTransformForItem(Child=(Type=Bone,Name="pelvis"), Parent=(Type=Bone,Name="root"), bChildInitial=False, bParentInitial=False).RelativeTransform.Translation.Z
    ).Result
   )
   > RigVMVariableNode(Value=@n2.Result) {variable=CrouchWalking}
   n3 = &RigUnit_GetRelativeTransformForItem(Child=(Type=Bone,Name="pelvis"), Parent=(Type=Bone,Name="root"), bChildInitial=False, bParentInitial=False)
   > RigVMVariableNode(Value=&RigVMFunction_AnimEvalRichCurve(
     Curve=(EditorCurveData=(Keys=((),(Time=1,Value=1))))
     SourceMinimum=0
     TargetMinimum=0
     Value=&RigVMFunction_MathDoubleRemap(SourceMaximum=85, SourceMinimum=44, TargetMaximum=0, TargetMinimum=1, bClamp=false, Value=@n3.RelativeTransform.Translation.Z).Result
    ).Result) {variable=isCrouching}
   n4 = &RigVMFunction_AnimEvalRichCurve(
    Curve=(EditorCurveData=(Keys=((InterpMode=RCIM_Cubic),(InterpMode=RCIM_Cubic,Time=0.5,Value=1),(Time=0.8))))
    SourceMaximum=2
    SourceMinimum=0
    TargetMinimum=0
    Value=&RigVMFunction_MathDoubleRemap(
     SourceMaximum=60
     SourceMinimum=45
     TargetMaximum=0
     TargetMinimum=1
     bClamp=false
     Value=&RigUnit_GetRelativeTransformForItem(Child=(Type=Bone,Name="pelvis"), Parent=(Type=Bone,Name="root"), bChildInitial=False, bParentInitial=False).RelativeTransform.Translation.Z
    ).Result
   )
   > RigVMVariableNode(Value=@n4.Result) {variable=isJustCrouching}
   > GetAlphaFromQuatDelta_3 = RigVMFunctionReferenceNode(
    Rotation Order=ZYX
    child=(Type=Bone,Name="head")
    onlyTwist=true
    outputMinMax=(X=0,Y=1)
    parent=(Type=Bone,Name="neck_01")
    premult=1
    rotationMinMax=(X=-50,Y=50)
    triggerAxis=X
    twistAxis=(X=1,Y=0,Z=0)
   )
   Remap = &RigVMFunction_MathDoubleRemap(SourceMaximum=0.5, SourceMinimum=0, TargetMaximum=0, TargetMinimum=1, Value=@GetAlphaFromQuatDelta_3.alpha)
   > RigVMVariableNode(Value=@Remap.Result) {variable=LookingLeft}
   > RigVMVariableNode(Value=&RigVMFunction_MathDoubleRemap(SourceMaximum=1, SourceMinimum=0.5, TargetMaximum=1, TargetMinimum=0, Value=@GetAlphaFromQuatDelta_3.alpha).Result) {variable=LookingRight}
   > GetAlphaFromQuatDelta_4 = RigVMFunctionReferenceNode(
    Rotation Order=ZYX
    child=(Type=Bone,Name="head")
    outputMinMax=(X=0,Y=1)
    parent=(Type=Bone,Name="neck_01")
    premult=1
    rotationMinMax=(X=-20,Y=0)
    triggerAxis=Z
    twistAxis=(X=1,Y=0,Z=0)
   )
   n5 = &RigVMFunction_AnimEvalRichCurve(
    Curve="(EditorCurveData=(Keys=((InterpMode=RCIM_Cubic),(InterpMode=RCIM_Cubic,TangentMode=RCTM_User,Time=0.5,Value=0.25,ArriveTangent=0.9828,LeaveTangent=0.9828),(InterpMode=RCIM_Cubic,TangentMode=RCTM_User,Time=1,Value=1,ArriveTangent=1.918,LeaveTangent=1.918))) ...(truncated)"
    SourceMinimum=0
    TargetMinimum=0
    Value=@GetAlphaFromQuatDelta_4.alpha
   )
   > RigVMVariableNode(Value=@n5.Result) {variable=LookingDown}
   > GetAlphaFromQuatDelta_5 = RigVMFunctionReferenceNode(
    Rotation Order=ZYX
    child=(Type=Bone,Name="head")
    outputMinMax=(X=0,Y=1)
    parent=(Type=Bone,Name="spine_05")
    premult=1
    rotationMinMax=(X=-30,Y=-36)
    triggerAxis=Z
    twistAxis=(X=1,Y=0,Z=0)
   )
   RigVMFunction_AnimEvalRichCurve = &RigVMFunction_AnimEvalRichCurve(Curve=(EditorCurveData=(Keys=((),(Time=0.45,Value=1),(Time=1)))), SourceMinimum=0, TargetMinimum=0, Value=@GetAlphaFromQuatDelta_5.alpha)
   > RigVMVariableNode(Value=@RigVMFunction_AnimEvalRichCurve.Result) {variable=LookingStraightAhead}
   > RigVMFunctionReturnNode()
  ]}
 CollapseNode_ContainedGraph {nodes [
   RigVMFunctionEntryNode()
   > RigVMFunction_Sequence = RigVMAggregateNode()
   Subtract = &RigVMFunction_MathDoubleSub(A=1, B=RigVMVariableNode() {variable=isJustCrouching}.Value)
   &RigUnit_ModifyTransforms(
    ItemToModify="((Item=(Type=Bone,Name=\"base_ponytail_01\"),Transform=(Rotation=(X=0,Y=0,Z=-0.1736,W=0.9848),Translation=(X=0,Y=0,Z=0),Scale3D=(X=1,Y=1,Z=1))),(Item=(Type=Bone,Name=\"dyn_ponytail_02\"),Transform=(Rotation=(X=0,Y=0,Z=-0.1736,W=0.9848),Translation=(X=0,Y=0,Z=0 ...(truncated)"
    WeightMinimum=0
    Weight=&RigVMFunction_MathDoubleMul(B=@Subtract.Result, A=RigVMVariableNode() {variable=HeadLookingUp}.Value).Result
    ExecutePin=@RigVMFunction_Sequence.A
   )
   n1 = &RigVMFunction_AnimEvalRichCurve(
    Curve=(EditorCurveData=(Keys=((InterpMode=RCIM_Cubic,TangentMode=RCTM_User,ArriveTangent=1.5596,LeaveTangent=1.5596),(InterpMode=RCIM_Cubic,TangentMode=RCTM_User,Time=1,Value=1,ArriveTangent=0.4206,LeaveTangent=0.4206))))
    SourceMinimum=0
    TargetMinimum=0
    Value=RigVMVariableNode() {variable=HeadLookingUp_Crouching}.Value
   )
   Multiply_3 = RigVMAggregateNode(C=1, A=@n1.Result, B=RigVMVariableNode() {variable=isJustCrouching}.Value)
   > &RigUnit_ModifyTransforms(
    ItemToModify="((Item=(Type=Bone,Name=\"base_ponytail_01\"),Transform=(Rotation=(X=0,Y=0,Z=-0.3007,W=0.9537),Translation=(X=2,Y=-1,Z=0),Scale3D=(X=1,Y=1,Z=1))),(Item=(Type=Bone,Name=\"dyn_ponytail_02\"),Transform=(Rotation=(X=0,Y=0,Z=-0.2164,W=0.9763),Translation=(X=-0.5,Y=- ...(truncated)"
    WeightMinimum=0
    Weight=@Multiply_3.Result
   )
   n2 = &RigVMFunction_AlphaInterp(
    InterpSpeedDecreasing=10
    InterpSpeedIncreasing=10
    Scale=1
    bClampResult=False
    Value=&RigVMFunction_MathBoolToFloat(Value=RigVMVariableNode() {variable=isMoving2D}.Value).Result
   )
   Multiply_7 = RigVMAggregateNode(
    C=&RigVMFunction_MathDoubleSub(A=1, B=@n2.Result).Result
    A=RigVMVariableNode() {variable=CrouchWalking}.Value
    B=RigVMVariableNode() {variable=HeadLookingUp_Extreme}.Value
   )
   > &RigUnit_ModifyTransforms(
    ItemToModify="((Item=(Type=Bone,Name=\"base_ponytail_01\"),Transform=(Rotation=(X=0,Y=0,Z=-0.3007,W=0.9537),Translation=(X=1,Y=-2,Z=0),Scale3D=(X=1,Y=0.9,Z=1))),(Item=(Type=Bone,Name=\"dyn_ponytail_02\"),Transform=(Rotation=(X=0,Y=0,Z=-0.0436,W=0.999),Translation=(X=-0.5,Y= ...(truncated)"
    WeightMinimum=0
    Weight=@Multiply_7.Result
   )
   RigVMFunction_MathBoolToFloat = &RigVMFunction_MathBoolToFloat(Value=RigVMVariableNode() {variable=isMoving2D}.Value)
   n3 = &RigVMFunction_AlphaInterp(InterpSpeedDecreasing=6, InterpSpeedIncreasing=7, Scale=1, bClampResult=False, Value=@RigVMFunction_MathBoolToFloat.Result)
   Multiply_4 = RigVMAggregateNode(C=@n3.Result, B=RigVMVariableNode() {variable=CrouchWalking}.Value, A=RigVMVariableNode() {variable=HeadLookingUp_Extreme}.Value)
   > &RigUnit_ModifyTransforms(
    ItemToModify="((Item=(Type=Bone,Name=\"base_ponytail_01\"),Transform=(Rotation=(X=0,Y=0,Z=-0.4226,W=0.9063),Translation=(X=0,Y=-3,Z=0),Scale3D=(X=1,Y=1,Z=1))),(Item=(Type=Bone,Name=\"dyn_ponytail_02\"),Transform=(Rotation=(X=0,Y=0,Z=0,W=1),Translation=(X=-0.5,Y=-2,Z=0),Scal ...(truncated)"
    WeightMinimum=0
    Weight=@Multiply_4.Result
   )
   RigVMFunction_AnimEvalRichCurve = &RigVMFunction_AnimEvalRichCurve(
    Curve="(EditorCurveData=(Keys=((InterpMode=RCIM_Cubic,TangentMode=RCTM_User,ArriveTangent=0.4251,LeaveTangent=0.4251),(InterpMode=RCIM_Cubic,TangentMode=RCTM_User,Time=0.5,Value=0.3,ArriveTangent=0.7137,LeaveTangent=0.7137),(InterpMode=RCIM_Cubic,TangentMode=RCTM ...(truncated)"
    SourceMinimum=0
    TargetMinimum=0
    Value=RigVMVariableNode() {variable=LookingLeft}.Value
   )
   n4 = RigVMVariableNode() {variable=LookingStraightAhead}
   AlphaInterp = &RigVMFunction_AlphaInterp(
    InterpSpeedDecreasing=8
    InterpSpeedIncreasing=8
    Scale=1
    bClampResult=False
    bInterpResult=False
    Value=&RigVMFunction_MathDoubleMul(B=@RigVMFunction_AnimEvalRichCurve.Result, A=&RigVMFunction_MathDoubleSub(A=1, B=@n4.Value).Result).Result
   )
   > &RigUnit_ModifyTransforms(
    ItemToModify="((Item=(Type=Bone,Name=\"base_ponytail_01\"),Transform=(Rotation=(X=0,Y=-0.342,Z=0,W=0.9397),Translation=(X=0,Y=-1,Z=0),Scale3D=(X=1,Y=1,Z=1))),(Item=(Type=Bone,Name=\"dyn_ponytail_02\"),Transform=(Rotation=(X=-0.25,Y=-0.25,Z=-0.067,W=0.933),Translation=(X=0,Y ...(truncated)"
    WeightMinimum=0
    Weight=@AlphaInterp.Result
   )
   n5 = &RigVMFunction_MathDoubleSub(A=1, B=RigVMVariableNode() {variable=LookingStraightAhead}.Value)
   n6 = RigVMVariableNode() {variable=LookingRight}
   n7 = RigVMVariableNode() {variable=isMeleeSwinging}
   > &RigUnit_ModifyTransforms(
    ItemToModify="((Item=(Type=Bone,Name=\"base_ponytail_01\"),Transform=(Rotation=(X=0,Y=0.2588,Z=0,W=0.9659),Translation=(X=0,Y=0,Z=0),Scale3D=(X=1,Y=1,Z=1))),(Item=(Type=Bone,Name=\"dyn_ponytail_02\"),Transform=(Rotation=(X=0.2393,Y=0.1893,Z=0.0381,W=0.9515),Translation=(X=0 ...(truncated)"
    WeightMinimum=0
    Weight=RigVMDispatchNode(
     Values=(0,0)
     Values.0=&RigVMFunction_MathDoubleMul(A=@n5.Result, B=@n6.Value).Result
     Index=&RigVMFunction_MathBoolToInteger(Value=@n7.Value).Result
     Values.1=@n6.Value
    ) {template="DISPATCH_RigVMDispatch_SelectInt32(in Index,in Values,out Result)"}.Result
   )
   > &RigUnit_ModifyTransforms(
    ItemToModify="((Item=(Type=Bone,Name=\"base_ponytail_01\"),Transform=(Rotation=(X=0,Y=0,Z=0.1736,W=0.9848),Translation=(X=0,Y=0,Z=0),Scale3D=(X=1,Y=1,Z=1))),(Item=(Type=Bone,Name=\"dyn_ponytail_02\"),Transform=(Rotation=(X=0,Y=0,Z=0,W=1),Translation=(X=0,Y=0,Z=0),Scale3D=(X ...(truncated)"
    WeightMinimum=0
    Weight=RigVMVariableNode() {variable=LookingDown}.Value
   )
   Multiply = &RigVMFunction_MathDoubleMul(A=RigVMVariableNode() {variable=LookingDown}.Value, B=RigVMVariableNode() {variable=LookingLeft}.Value)
   > &RigUnit_ModifyTransforms(
    ItemToModify="((Item=(Type=Bone,Name=\"base_ponytail_01\"),Transform=(Rotation=(X=0,Y=0,Z=0.0872,W=0.9962),Translation=(X=0,Y=0,Z=0),Scale3D=(X=1,Y=1,Z=1))),(Item=(Type=Bone,Name=\"dyn_ponytail_02\"),Transform=(Rotation=(X=0,Y=0,Z=0.0436,W=0.999),Translation=(X=0,Y=0,Z=0),S ...(truncated)"
    WeightMinimum=0
    Weight=@Multiply.Result
   )
   > &RigUnit_ModifyTransforms(
    ItemToModify="((Item=(Type=Bone,Name=\"base_ponytail_01\"),Transform=(Rotation=(X=0,Y=0,Z=0.0872,W=0.9962),Translation=(X=0,Y=0,Z=0),Scale3D=(X=1,Y=1,Z=1))),(Item=(Type=Bone,Name=\"dyn_ponytail_02\"),Transform=(Rotation=(X=0,Y=0,Z=0.0436,W=0.999),Translation=(X=0,Y=0,Z=0),S ...(truncated)"
    WeightMinimum=0
    Weight=&RigVMFunction_MathDoubleMul(A=RigVMVariableNode() {variable=LookingDown}.Value, B=RigVMVariableNode() {variable=LookingRight}.Value).Result
   )
   n8 = RigVMVariableNode() {variable=isPlayingEmote}
   n9 = &RigVMFunction_AlphaInterp(
    InterpSpeedDecreasing=10
    InterpSpeedIncreasing=10
    Scale=1
    bClampResult=False
    Value=&RigVMFunction_MathDoubleMul(B=&RigVMFunction_MathBoolToFloat(Value=@n8.Value).Result, A=RigVMVariableNode() {variable=HeadLookingUp_Crouching}.Value).Result
   )
   > &RigUnit_ModifyTransforms(
    ItemToModify=((Item=(Type=Bone,Name="base_ponytail_01"),Transform=(Rotation=(X=0,Y=0,Z=-0.0872,W=0.9962),Translation=(X=-0.5,Y=-1,Z=0),Scale3D=(X=1,Y=1,Z=1))))
    WeightMinimum=0
    Weight=@n9.Result
   )
   > GetAlphaFromQuatDelta_1 = RigVMFunctionReferenceNode(
    Rotation Order=ZYX
    child=(Type=Bone,Name="head")
    outputMinMax=(X=0,Y=1)
    parent=(Type=Bone,Name="spine_05")
    premult=1
    rotationMinMax=(X=-45,Y=-8)
    triggerAxis=Z
    twistAxis=(X=1,Y=0,Z=0)
   )
   > GetAlphaFromQuatDelta_2 = RigVMFunctionReferenceNode(
    Rotation Order=ZYX
    child=(Type=Bone,Name="head")
    onlyTwist=true
    outputMinMax=(X=0,Y=1)
    parent=(Type=Bone,Name="spine_05")
    premult=1
    rotationMinMax=(X=10,Y=-52)
    triggerAxis=X
    twistAxis=(X=1,Y=0,Z=0)
   )
   Multiply_15 = RigVMAggregateNode(
    A=@GetAlphaFromQuatDelta_2.alpha
    C=&RigVMFunction_MathBoolToFloat(Value=RigVMVariableNode() {variable=isPlayingEmote}.Value).Result
    B=@GetAlphaFromQuatDelta_1.alpha
   )
   > &RigUnit_ModifyTransforms(
    ItemToModify=((Item=(Type=Bone,Name="base_ponytail_01"),Transform=(Rotation=(X=0,Y=0,Z=0.0872,W=0.9962),Translation=(X=0,Y=0,Z=0),Scale3D=(X=1,Y=1,Z=1))))
    WeightMinimum=0
    Weight=@Multiply_15.Result
   )
   RigVMFunctionReturnNode(ExecuteContext=@RigVMFunction_Sequence.C)
   n10 = &RigVMFunction_MathBoolNot(Value=RigVMVariableNode() {variable=isWearingBackBling}.Value)
   n11 = RigVMVariableNode() {variable=isTacticalJumping}
   n12 = &RigVMFunction_AlphaInterp(
    InterpSpeedDecreasing=10
    InterpSpeedIncreasing=10
    Scale=1
    bClampResult=False
    Value=&RigVMFunction_MathBoolToFloat(Value=&RigVMFunction_MathBoolAnd(B=@n10.Result, A=@n11.Value).Result).Result
   )
   &RigUnit_ModifyTransforms(
    ItemToModify="((Item=(Type=Bone,Name=\"dyn_ponytail_02\"),Transform=(Rotation=(X=0,Y=0,Z=-0.1736,W=0.9848),Translation=(X=0,Y=0,Z=0),Scale3D=(X=1,Y=1,Z=1))),(Item=(Type=Bone,Name=\"dyn_ponytail_03\"),Transform=(Rotation=(X=0,Y=0,Z=-0.1736,W=0.9848),Translation=(X=0,Y=0,Z=0) ...(truncated)"
    WeightMinimum=0
    Weight=@n12.Result
    ExecutePin=@RigVMFunction_Sequence.B
   )
   > GetAlphaFromQuatDelta = RigVMFunctionReferenceNode(
    Rotation Order=ZYX
    child=(Type=Bone,Name="head")
    outputMinMax=(X=0,Y=1)
    parent=(Type=Bone,Name="neck_01")
    premult=1
    rotationMinMax=(X=0,Y=-33)
    triggerAxis=Z
    twistAxis=(X=1,Y=0,Z=0)
   )
   RigVMFunction_MathBoolNot = &RigVMFunction_MathBoolNot(Value=RigVMVariableNode() {variable=isTargeting}.Value)
   RigVMFunction_MathBoolAnd = &RigVMFunction_MathBoolAnd(A=@RigVMFunction_MathBoolNot.Result, B=RigVMVariableNode() {variable=Moving2D}.Value)
   n13 = &RigVMFunction_AlphaInterp(
    InterpSpeedDecreasing=14
    InterpSpeedIncreasing=8
    Scale=1
    bClampResult=False
    Value=&RigVMFunction_MathBoolToFloat(Value=@RigVMFunction_MathBoolAnd.Result).Result
   )
   Multiply_5 = RigVMAggregateNode(
    D=@n13.Result
    A=@GetAlphaFromQuatDelta.alpha
    C=&RigVMFunction_MathBoolToFloat(Value=RigVMVariableNode() {variable=animDynamicsEnabled}.Value).Result
    B=RigVMVariableNode() {variable=CrouchWalking}.Value
   )
   > n14 = &RigUnit_ModifyTransforms(
    ItemToModify=((Item=(Type=Bone,Name="base_ponytail_01"),Transform=(Rotation=(X=0,Y=0,Z=0.061,W=0.9981),Translation=(X=0,Y=0,Z=0),Scale3D=(X=1,Y=1,Z=1))))
    WeightMinimum=0
    Weight=@Multiply_5.Result
   )
   n15 = RigVMVariableNode() {variable=RigidBodiesEnabled}
   n16 = RigVMVariableNode() {variable=animDynamicsEnabled}
   RigVMFunction_MathBoolOr = &RigVMFunction_MathBoolOr(
    A=&RigVMFunction_MathBoolAnd(A=&RigVMFunction_MathBoolNot(Value=@n15.Value).Result, B=&RigVMFunction_MathBoolNot(Value=@n16.Value).Result).Result
    B=RigVMVariableNode() {variable=isWearingBackBling}.Value
   )
   RigVMFunction_ControlFlowBranch = &RigVMFunction_ControlFlowBranch(Condition=@RigVMFunction_MathBoolOr.Result, ExecuteContext=@n14.ExecutePin)
   GetAlphaFromQuatDelta_3 = RigVMFunctionReferenceNode(
    Rotation Order=ZYX
    child=(Type=Bone,Name="clavicle_r")
    outputMinMax=(X=0,Y=1)
    parent=(Type=Null,Name="clavicle_r_null")
    premult=1
    rotationMinMax=(X=5,Y=-12)
    triggerAxis=Z
    twistAxis=(X=1,Y=0,Z=0)
    ExecuteContext=@RigVMFunction_ControlFlowBranch.False
   )
   n17 = RigVMVariableNode() {variable=isCrouching}
   > &RigUnit_ModifyTransforms(
    ItemToModify="((Item=(Type=Bone,Name=\"base_ponytail_01\"),Transform=(Rotation=(X=0,Y=-0.0872,Z=0,W=0.9962),Translation=(X=0,Y=0,Z=0),Scale3D=(X=1,Y=1,Z=1))),(Item=(Type=Bone,Name=\"dyn_ponytail_02\"),Transform=(Rotation=(X=0,Y=-0.0872,Z=0,W=0.9962),Translation=(X=0,Y=0,Z=0 ...(truncated)"
    WeightMinimum=0
    Weight=&RigVMFunction_MathDoubleMul(B=&RigVMFunction_MathDoubleSub(A=1, B=@n17.Value).Result, A=@GetAlphaFromQuatDelta_3.alpha).Result
   )
   > &RigUnit_ModifyTransforms(
    ItemToModify="((Item=(Type=Bone,Name=\"base_ponytail_01\"),Transform=(Rotation=(X=0,Y=-0.1736,Z=0,W=0.9848),Translation=(X=0,Y=0,Z=0),Scale3D=(X=1,Y=1,Z=1))),(Item=(Type=Bone,Name=\"dyn_ponytail_02\"),Transform=(Rotation=(X=0,Y=-0.1305,Z=0,W=0.9914),Translation=(X=0,Y=0,Z=0 ...(truncated)"
    WeightMinimum=0
    Weight=&RigVMFunction_MathDoubleMul(B=RigVMVariableNode() {variable=isCrouching}.Value, A=@GetAlphaFromQuatDelta_3.alpha).Result
   )
   n18 = RigVMVariableNode() {variable=isFrontEnd}
   &RigUnit_ModifyTransforms(
    ItemToModify="((Item=(Type=Bone,Name=\"base_ponytail_01\"),Transform=(Rotation=(X=0.0027,Y=-0.0435,Z=0.061,W=0.9972),Translation=(X=0,Y=-0.5,Z=2),Scale3D=(X=1,Y=1,Z=1))),(Item=(Type=Bone,Name=\"dyn_ponytail_02\"),Transform=(Rotation=(X=0,Y=0,Z=-0.0262,W=0.9997),Translation= ...(truncated)"
    WeightMinimum=0
    ExecutePin=@RigVMFunction_ControlFlowBranch.True
    Weight=RigVMDispatchNode(Values=(0,0.5), Index=&RigVMFunction_MathBoolToInteger(Value=@n18.Value).Result) {template="DISPATCH_RigVMDispatch_SelectInt32(in Index,in Values,out Result)"}.Result
   )
   n19 = &RigVMFunction_MathBoolToFloat(Value=RigVMVariableNode() {variable=isTacticalJumping}.Value)
   n20 = RigVMVariableNode() {variable=isWearingBackBling}
   n21 = &RigVMFunction_AlphaInterp(
    InterpSpeedDecreasing=10
    InterpSpeedIncreasing=10
    Scale=1
    bClampResult=False
    Value=RigVMDispatchNode(
     Values=(0,0)
     Values.1=&RigVMFunction_MathDoubleMul(B=0.2, A=@n19.Result).Result
     Values.0=@n19.Result
     Index=&RigVMFunction_MathBoolToInteger(Value=@n20.Value).Result
    ) {template="DISPATCH_RigVMDispatch_SelectInt32(in Index,in Values,out Result)"}.Result
   )
   > &RigUnit_ModifyTransforms(
    ItemToModify="((Item=(Type=Bone,Name=\"base_ponytail_01\"),Transform=(Rotation=(X=0,Y=0,Z=-0.1305,W=0.9914),Translation=(X=0,Y=0,Z=1),Scale3D=(X=1,Y=1,Z=1))),(Item=(Type=Bone,Name=\"dyn_ponytail_02\"),Transform=(Rotation=(X=0,Y=0,Z=-0.1305,W=0.9914),Translation=(X=0,Y=0,Z=0 ...(truncated)"
    WeightMinimum=0
    Weight=@n21.Result
   )
   > GetAlphaFromQuatDelta_4 = RigVMFunctionReferenceNode(
    Rotation Order=ZYX
    child=(Type=Bone,Name="clavicle_l")
    outputMinMax=(X=0,Y=1)
    parent=(Type=Null,Name="clavicle_l_null")
    premult=1
    rotationMinMax=(X=-14,Y=-18)
    triggerAxis=Z
    twistAxis=(X=1,Y=0,Z=0)
   )
   n22 = &RigVMFunction_AlphaInterp(InterpSpeedDecreasing=7, InterpSpeedIncreasing=10, Scale=1, bClampResult=False, Value=@GetAlphaFromQuatDelta_4.alpha)
   > &RigUnit_ModifyTransforms(
    ItemToModify="((Item=(Type=Bone,Name=\"base_ponytail_01\"),Transform=(Rotation=(X=-0.0076,Y=-0.0868,Z=-0.0868,W=0.9924),Translation=(X=0,Y=0,Z=0),Scale3D=(X=1,Y=1,Z=1))),(Item=(Type=Bone,Name=\"dyn_ponytail_02\"),Transform=(Rotation=(X=-0.0076,Y=-0.0868,Z=-0.0868,W=0.9924), ...(truncated)"
    WeightMinimum=0
    Weight=@n22.Result
   )
   > GetAlphaFromQuatDelta_5 = RigVMFunctionReferenceNode(
    Rotation Order=ZYX
    child=(Type=Bone,Name="clavicle_r")
    outputMinMax=(X=0,Y=1)
    parent=(Type=Null,Name="clavicle_r_null")
    premult=1
    rotationMinMax=(X=0,Y=-12)
    triggerAxis=Z
    twistAxis=(X=1,Y=0,Z=0)
   )
   > &RigUnit_ModifyTransforms(
    ItemToModify="((Item=(Type=Bone,Name=\"base_ponytail_01\"),Transform=(Rotation=(X=0,Y=-0.0872,Z=0,W=0.9962),Translation=(X=0,Y=0,Z=0),Scale3D=(X=1,Y=1,Z=1))),(Item=(Type=Bone,Name=\"dyn_ponytail_02\"),Transform=(Rotation=(X=0,Y=-0.0872,Z=0,W=0.9962),Translation=(X=0,Y=0,Z=0 ...(truncated)"
    WeightMinimum=0
    Weight=@GetAlphaFromQuatDelta_5.alpha
   )
   > &RigUnit_AimConstraintLocalSpaceOffset(
    Child=(Type=Bone,Name="dyn_ponytail_02")
    Filter=(bX=True,bY=True,bZ=True)
    Parents=((Item=(Type=Bone,Name="spine_04"),Weight=1))
    UpAxis=(X=0,Y=-1,Z=0)
    Weight=1
    WorldUp=(Target=(X=0,Y=1,Z=0),Kind=Direction,Space=(Type=Bone,Name="spine_05"))
   )
   RigVMFunction_MathVectorDistance = &RigVMFunction_MathVectorDistance(
    A=&RigUnit_GetTransform(Item=(Type=Bone,Name="dyn_ponytail_04"), bInitial=False).Transform.Translation
    B=&RigUnit_GetTransform(Item=(Type=Bone,Name="upperarm_r"), bInitial=False).Transform.Translation
   )
   Remap = &RigVMFunction_MathDoubleRemap(SourceMaximum=22, SourceMinimum=14, TargetMaximum=0, TargetMinimum=1, Value=@RigVMFunction_MathVectorDistance.Result)
   > &RigUnit_ModifyTransforms(
    ItemToModify=((Item=(Type=Bone,Name="dyn_ribbon_l_02"),Transform=(Rotation=(X=0,Y=0,Z=0,W=1),Translation=(X=0,Y=0,Z=0),Scale3D=(X=1,Y=1,Z=1))))
    WeightMinimum=0
    Weight=@Remap.Result
   )
  ]}
 CollapseNode_ContainedGraph {nodes [
   RigVMFunctionEntryNode()
   > RigVMFunction_Sequence = RigVMAggregateNode()
   NullConstraints = RigVMCollapseNode(ExecuteContext=@RigVMFunction_Sequence.A)
   > &RigUnit_ModifyTransforms(
    ItemToModify="((Item=(Type=Null,Name=\"dyn_ribbon_r_03_null\"),Transform=(Rotation=(X=0,Y=0,Z=0,W=1),Translation=(X=0,Y=0,Z=3),Scale3D=(X=1,Y=1,Z=1))),(Item=(Type=Null,Name=\"dyn_ribbon_r_02_null\"),Transform=(Rotation=(X=0,Y=0,Z=0,W=1),Translation=(X=0,Y=0,Z=2),Scale3D=(X= ...(truncated)"
    Weight=1
    WeightMinimum=0
   )
   &RigUnit_ModifyTransforms(
    ItemToModify="((Item=(Type=Bone,Name=\"base_BandanaKnot\"),Transform=(Rotation=(X=0,Y=-0.7071,Z=0,W=0.7071),Translation=(X=0,Y=0,Z=0),Scale3D=(X=1,Y=1,Z=1))),(Item=(Type=Null,Name=\"dyn_ribbon_r_01_null\"),Transform=(Rotation=(X=0,Y=0,Z=0,W=1),Translation=(X=0,Y=0,Z=2),Scal ...(truncated)"
    WeightMinimum=0
    ExecutePin=@RigVMFunction_Sequence.B
    Weight=RigVMVariableNode() {variable=LookingLeft}.Value
   )
   > &RigUnit_ModifyTransforms(
    ItemToModify="((Item=(Type=Bone,Name=\"base_BandanaKnot\"),Transform=(Rotation=(X=-0.0432,Y=-0.7058,Z=-0.0432,W=0.7058),Translation=(X=0,Y=0,Z=0),Scale3D=(X=1,Y=1,Z=1))),(Item=(Type=Null,Name=\"dyn_ribbon_l_01_null\"),Transform=(Rotation=(X=0,Y=0,Z=0,W=1),Translation=(X=0,Y ...(truncated)"
    WeightMinimum=0
    Weight=RigVMVariableNode() {variable=LookingRight}.Value
   )
   AlphaInterp = &RigVMFunction_AlphaInterp(
    InterpSpeedDecreasing=5
    InterpSpeedIncreasing=5
    Scale=1
    bClampResult=False
    bInterpResult=False
    Value=RigVMVariableNode() {variable=HeadLookingUp}.Value
   )
   RigVMFunction_MathBoolNot = &RigVMFunction_MathBoolNot(Value=RigVMVariableNode() {variable=isTargeting}.Value)
   RigVMFunction_MathBoolToFloat = &RigVMFunction_MathBoolToFloat(Value=@RigVMFunction_MathBoolNot.Result)
   n1 = &RigVMFunction_AlphaInterp(InterpSpeedDecreasing=2, InterpSpeedIncreasing=2, Scale=1, bClampResult=False, Value=@RigVMFunction_MathBoolToFloat.Result)
   Multiply = &RigVMFunction_MathDoubleMul(A=@AlphaInterp.Result, B=@n1.Result)
   > &RigUnit_ModifyTransforms(
    ItemToModify="((Item=(Type=Bone,Name=\"base_BandanaKnot\"),Transform=(Rotation=(X=0,Y=0,Z=-0.4226,W=0.9063),Translation=(X=1,Y=1,Z=-0.25),Scale3D=(X=1,Y=1,Z=1))),(Item=(Type=Null,Name=\"dyn_ribbon_r_01_null\"),Transform=(Rotation=(X=0,Y=0,Z=0,W=1),Translation=(X=1,Y=1,Z=-2) ...(truncated)"
    WeightMinimum=0
    Weight=@Multiply.Result
   )
   n2 = &RigVMFunction_AlphaInterp(
    InterpSpeedDecreasing=2
    InterpSpeedIncreasing=2
    Scale=1
    bClampResult=False
    Value=&RigVMFunction_MathBoolToFloat(Value=RigVMVariableNode() {variable=isTargeting}.Value).Result
   )
   > &RigUnit_ModifyTransforms(
    ItemToModify="((Item=(Type=Bone,Name=\"base_BandanaKnot\"),Transform=(Rotation=(X=0,Y=0,Z=-0.4226,W=0.9063),Translation=(X=1,Y=1,Z=-0.25),Scale3D=(X=1,Y=1,Z=1))),(Item=(Type=Null,Name=\"dyn_ribbon_r_01_null\"),Transform=(Rotation=(X=0,Y=0,Z=0,W=1),Translation=(X=1,Y=-3,Z=-1 ...(truncated)"
    WeightMinimum=0
    Weight=&RigVMFunction_MathDoubleMul(B=@n2.Result, A=RigVMVariableNode() {variable=HeadLookingUp_Crouching}.Value).Result
   )
   > GetAlphaFromQuatDelta = RigVMFunctionReferenceNode(
    Rotation Order=ZYX
    child=(Type=Bone,Name="clavicle_r")
    outputMinMax=(X=0,Y=1)
    parent=(Type=Null,Name="clavicle_r_null")
    premult=1
    rotationMinMax=(X=0,Y=45)
    triggerAxis=Z
    twistAxis=(X=1,Y=0,Z=0)
   )
   > &RigUnit_ModifyTransforms(
    ItemToModify="((Item=(Type=Null,Name=\"dyn_ribbon_l_02_null\"),Transform=(Rotation=(X=0,Y=0,Z=0,W=1),Translation=(X=0,Y=0,Z=4),Scale3D=(X=1,Y=1,Z=1))),(Item=(Type=Null,Name=\"dyn_ribbon_l_03_null\"),Transform=(Rotation=(X=0,Y=0,Z=0,W=1),Translation=(X=0,Y=1,Z=6),Scale3D=(X= ...(truncated)"
    WeightMinimum=0
    Weight=@GetAlphaFromQuatDelta.alpha
   )
   > GetAlphaFromQuatDelta_1 = RigVMFunctionReferenceNode(
    Rotation Order=ZYX
    child=(Type=Bone,Name="clavicle_r")
    outputMinMax=(X=0,Y=1)
    parent=(Type=Null,Name="clavicle_r_null")
    premult=1
    rotationMinMax=(X=0,Y=-40)
    triggerAxis=Y
    twistAxis=(X=1,Y=0,Z=0)
   )
   > &RigUnit_ModifyTransforms(
    ItemToModify="((Item=(Type=Null,Name=\"dyn_ribbon_r_02_null\"),Transform=(Rotation=(X=0,Y=0,Z=0,W=1),Translation=(X=6,Y=0,Z=-5),Scale3D=(X=1,Y=1,Z=1))),(Item=(Type=Null,Name=\"dyn_ribbon_r_03_null\"),Transform=(Rotation=(X=0,Y=0,Z=0,W=1),Translation=(X=1,Y=0,Z=-1),Scale3D=( ...(truncated)"
    WeightMinimum=0
    Weight=@GetAlphaFromQuatDelta_1.alpha
   )
   n3 = RigVMVariableNode() {variable=HeadLookingUp}
   Maximum = &RigVMFunction_MathDoubleMax(
    B=&RigVMFunction_MathDoubleSub(A=1, B=&RigVMFunction_MathDoubleMul(B=0.6, A=@n3.Value).Result).Result
    A=RigVMVariableNode() {variable=isCrouching}.Value
   )
   RigVMFunction_MathVectorDistance = &RigVMFunction_MathVectorDistance(
    A=&RigUnit_GetTransform(Item=(Type=Bone,Name="dyn_ponytail_03"), bInitial=False).Transform.Translation
    B=&RigUnit_GetTransform(Item=(Type=Bone,Name="upperarm_r"), bInitial=False).Transform.Translation
   )
   > &RigUnit_ModifyTransforms(
    ItemToModify="((Item=(Type=Null,Name=\"dyn_ribbon_l_03_null\"),Transform=(Rotation=(X=0,Y=0,Z=0,W=1),Translation=(X=0,Y=-3,Z=-10),Scale3D=(X=1,Y=1,Z=1))),(Item=(Type=Null,Name=\"dyn_ribbon_r_03_null\"),Transform=(Rotation=(X=0,Y=0,Z=0,W=1),Translation=(X=0,Y=0,Z=-20),Scale3 ...(truncated)"
    WeightMinimum=0
    Weight=&RigVMFunction_MathDoubleMul(
     B=@Maximum.Result
     A=&RigVMFunction_MathDoubleRemap(SourceMaximum=10, SourceMinimum=15, TargetMaximum=1, TargetMinimum=0, Value=@RigVMFunction_MathVectorDistance.Result).Result
    ).Result
   )
   ItemArray_4 = RigVMDispatchNode(Value=((Type=Bone,Name="dyn_ribbon_l_01"),(Type=Bone,Name="dyn_ribbon_l_02"),(Type=Bone,Name="dyn_ribbon_l_03"),(Type=Bone,Name="dyn_ribbon_l_04"))) {template="DISPATCH_RigVMDispatch_Constant(io Value)"}
   TwistBones = RigVMFunctionReferenceNode(
    Axis to Twist=(bX=True,bY=False,bZ=False)
    End=(Type=Bone,Name="spine_04")
    Start=(Type=Bone,Name="head")
    Bones to Twist=@ItemArray_4.Value
    ExecuteContext=@RigVMFunction_Sequence.D
   )
   ItemArray = RigVMDispatchNode(Value="((Type=Null,Name=\"dyn_ribbon_l_01_null\"),(Type=Null,Name=\"dyn_ribbon_l_02_null\"),(Type=Null,Name=\"dyn_ribbon_l_03_null\"),(Type=Null,Name=\"dyn_ribbon_l_04_null\"),(Type=Null,Name=\"dyn_ribbon_l_05_null\"),(Type=Null,Name=\"dyn_ribbon_l_06_null\"),(Type=Null,Name ...(truncated)") {template="DISPATCH_RigVMDispatch_Constant(io Value)"}
   ItemArray_2 = RigVMDispatchNode(Value="((Type=Bone,Name=\"dyn_ribbon_l_01\"),(Type=Bone,Name=\"dyn_ribbon_l_02\"),(Type=Bone,Name=\"dyn_ribbon_l_03\"),(Type=Bone,Name=\"dyn_ribbon_l_04\"),(Type=Bone,Name=\"dyn_ribbon_l_05\"),(Type=Bone,Name=\"dyn_ribbon_l_06\"),(Type=Bone,Name=\"dyn_ribbon_l_07\"),(Type=Bone ...(truncated)") {template="DISPATCH_RigVMDispatch_Constant(io Value)"}
   > SplineIK = RigVMFunctionReferenceNode(
    Primary Axis=(X=1,Y=0,Z=0)
    Samples Per Segment=4
    Secondary Spline Direction=(X=0,Y=5,Z=0)
    Spline Mode=BSpline
    Up Axis=(X=0,Y=0,Z=0)
    debug=true
    Controls=@ItemArray.Value
    Bones=@ItemArray_2.Value
   )
   ItemArray_7 = RigVMDispatchNode(Value=((Type=Bone,Name="dyn_ribbon_r_01"),(Type=Bone,Name="dyn_ribbon_r_02"),(Type=Bone,Name="dyn_ribbon_r_03"),(Type=Bone,Name="dyn_ribbon_r_04"))) {template="DISPATCH_RigVMDispatch_Constant(io Value)"}
   > TwistBones_1 = RigVMFunctionReferenceNode(
    Axis to Twist=(bX=True,bY=False,bZ=False)
    Debug=false
    End=(Type=Bone,Name="spine_04")
    Start=(Type=Bone,Name="head")
    Bones to Twist=@ItemArray_7.Value
   )
   ItemArray_1 = RigVMDispatchNode(Value="((Type=Null,Name=\"dyn_ribbon_r_01_null\"),(Type=Null,Name=\"dyn_ribbon_r_02_null\"),(Type=Null,Name=\"dyn_ribbon_r_03_null\"),(Type=Null,Name=\"dyn_ribbon_r_04_null\"),(Type=Null,Name=\"dyn_ribbon_r_05_null\"),(Type=Null,Name=\"dyn_ribbon_r_06_null\"),(Type=Null,Name ...(truncated)") {template="DISPATCH_RigVMDispatch_Constant(io Value)"}
   ItemArray_3 = RigVMDispatchNode(Value="((Type=Bone,Name=\"dyn_ribbon_r_01\"),(Type=Bone,Name=\"dyn_ribbon_r_02\"),(Type=Bone,Name=\"dyn_ribbon_r_03\"),(Type=Bone,Name=\"dyn_ribbon_r_04\"),(Type=Bone,Name=\"dyn_ribbon_r_05\"),(Type=Bone,Name=\"dyn_ribbon_r_06\"),(Type=Bone,Name=\"dyn_ribbon_r_07\"),(Type=Bone ...(truncated)") {template="DISPATCH_RigVMDispatch_Constant(io Value)"}
   > SplineIK_1 = RigVMFunctionReferenceNode(
    Primary Axis=(X=1,Y=0,Z=0)
    Samples Per Segment=4
    Secondary Spline Direction=(X=0,Y=5,Z=0)
    Spline Mode=BSpline
    Up Axis=(X=0,Y=0,Z=0)
    debug=false
    Controls=@ItemArray_1.Value
    Bones=@ItemArray_3.Value
   )
   RigVMFunction_MathBoolAnd = RigVMAggregateNode(
    C=&RigVMFunction_MathBoolNot(Value=RigVMVariableNode() {variable=isWearingBackBling}.Value).Result
    B=RigVMVariableNode() {variable=RigidBodiesEnabled}.Value
    A=RigVMVariableNode() {variable=isSkyFloating}.Value
   )
   n4 = &RigVMFunction_AlphaInterp(
    InterpSpeedDecreasing=10
    InterpSpeedIncreasing=10
    Scale=1
    bClampResult=False
    Value=&RigVMFunction_MathBoolToFloat(Value=@RigVMFunction_MathBoolAnd.Result).Result
   )
   &RigUnit_ModifyTransforms(
    ItemToModify="((Item=(Type=Bone,Name=\"dyn_BandanaKnot\"),Transform=(Rotation=(X=0,Y=0,Z=-0.0872,W=0.9962),Translation=(X=0,Y=0,Z=0),Scale3D=(X=1,Y=1,Z=1))),(Item=(Type=Bone,Name=\"base_ribbon_l_01\"),Transform=(Rotation=(X=0,Y=0,Z=-0.1736,W=0.9848),Translation=(X=0,Y=0,Z=0 ...(truncated)"
    WeightMinimum=0
    Weight=@n4.Result
    ExecutePin=@RigVMFunction_Sequence.E
   )
   > RigVMFunctionReturnNode()
   n5 = RigVMVariableNode() {variable=RigidBodiesEnabled}
   n6 = RigVMVariableNode() {variable=animDynamicsEnabled}
   NonDynamicFixes = RigVMCollapseNode(
    Enabled=&RigVMFunction_MathBoolOr(
     B=/Script/RigVM.RigVMFunction_MathBoolNand2(A=@n5.Value, B=@n6.Value) {is=struct RigVMUnitNode, template="Nand::Execute(in A,in B,out Result)"}.Result
     A=RigVMVariableNode() {variable=isWearingBackBling}.Value
    ).Result
    ExecuteContext=@RigVMFunction_Sequence.C
   )
   n7 = RigVMVariableNode() {variable=RigidBodiesEnabled}
   n8 = RigVMVariableNode() {variable=animDynamicsEnabled}
   RigVMFunction_MathBoolToInteger = &RigVMFunction_MathBoolToInteger(Value=&RigVMFunction_MathBoolOr(
     B=&RigVMFunction_MathBoolNand2(A=@n7.Value, B=@n8.Value) {template="Nand::Execute(in A,in B,out Result)"}.Result
     A=RigVMVariableNode() {variable=isWearingBackBling}.Value
    ).Result)
   &RigUnit_ModifyTransforms(
    ExecutePin=()
    ItemToModify="((Item=(Type=Bone,Name=\"dyn_ribbon_r_01\"),Transform=(Rotation=(X=-0.067,Y=-0.25,Z=-0.25,W=0.933),Translation=(X=0,Y=0,Z=0),Scale3D=(X=1,Y=1,Z=1))),(Item=(Type=Bone,Name=\"dyn_ribbon_l_01\"),Transform=(Rotation=(X=-0.0449,Y=-0.2549,Z=-0.1677,W=0.9513),Transla ...(truncated)"
    WeightMinimum=0
    Weight=RigVMDispatchNode(Values=(0,0), Index=@RigVMFunction_MathBoolToInteger.Result, Values.0=RigVMVariableNode() {variable=LookingRight}.Value) {template="DISPATCH_RigVMDispatch_SelectInt32(in Index,in Values,out Result)"}.Result
   )
  ]}
 CollapseNode_ContainedGraph {nodes [
   n1 = RigVMFunctionEntryNode()
   RigVMFunction_ControlFlowBranch = &RigVMFunction_ControlFlowBranch(Condition=@n1.Enabled, ExecuteContext=@n1.ExecuteContext)
   RigVMFunctionReturnNode(ExecuteContext=@RigVMFunction_ControlFlowBranch.Completed)
   Remap = &RigVMFunction_MathDoubleRemap(
    SourceMaximum=22
    SourceMinimum=14
    TargetMaximum=0
    TargetMinimum=1
    Value=&RigVMFunction_MathVectorDistance(
     A=&RigUnit_GetTransform(Item=(Type=Bone,Name="dyn_ponytail_04"), bInitial=False).Transform.Translation
     B=&RigUnit_GetTransform(Item=(Type=Bone,Name="upperarm_r"), bInitial=False).Transform.Translation
    ).Result
   )
   n2 = &RigVMFunction_AlphaInterp(
    InterpSpeedDecreasing=10
    InterpSpeedIncreasing=10
    Scale=1
    bClampResult=False
    Value=&RigVMFunction_MathBoolToFloat(Value=RigVMVariableNode() {variable=isTargeting}.Value).Result
   )
   Subtract = &RigVMFunction_MathDoubleSub(A=1, B=@n2.Result)
   &RigUnit_ModifyTransforms(
    ItemToModify="((Item=(Type=Null,Name=\"dyn_ribbon_l_03_null\"),Transform=(Rotation=(X=0,Y=0,Z=0,W=1),Translation=(X=0,Y=-1,Z=-3),Scale3D=(X=1,Y=1,Z=1))),(Item=(Type=Null,Name=\"dyn_ribbon_r_03_null\"),Transform=(Rotation=(X=0,Y=0,Z=0,W=1),Translation=(X=0,Y=-1,Z=-2),Scale3D ...(truncated)"
    WeightMinimum=0
    Weight=&RigVMFunction_MathDoubleMul(B=@Remap.Result, A=@Subtract.Result).Result
    ExecutePin=@RigVMFunction_ControlFlowBranch.True
   )
   > GetAlphaFromQuatDelta = RigVMFunctionReferenceNode(
    Rotation Order=ZYX
    child=(Type=Bone,Name="head")
    onlyTwist=true
    outputMinMax=(X=0,Y=1)
    parent=(Type=Null,Name="head_null")
    premult=1
    rotationMinMax=(X=0,Y=-40)
    twistAxis=(X=1,Y=0,Z=0)
   )
   > &RigUnit_ModifyTransforms(
    ItemToModify=((Item=(Type=Null,Name="dyn_ribbon_r_03_null"),Transform=(Rotation=(X=0,Y=0,Z=0,W=1),Translation=(X=0,Y=-3,Z=1),Scale3D=(X=1,Y=1,Z=1))))
    WeightMinimum=0
    Weight=@GetAlphaFromQuatDelta.alpha
   )
   RigVMFunction_MathBoolToFloat = &RigVMFunction_MathBoolToFloat(Value=RigVMVariableNode() {variable=isMeleeSwinging}.Value)
   AlphaInterp = &RigVMFunction_AlphaInterp(InterpSpeedDecreasing=10, InterpSpeedIncreasing=10, Scale=1, bClampResult=False, Value=@RigVMFunction_MathBoolToFloat.Result)
   > &RigUnit_ModifyTransforms(
    ItemToModify="((Item=(Type=Null,Name=\"dyn_ribbon_r_03_null\"),Transform=(Rotation=(X=0,Y=0,Z=0,W=1),Translation=(X=0,Y=2,Z=4),Scale3D=(X=1,Y=1,Z=1))),(Item=(Type=Null,Name=\"dyn_ribbon_l_03_null\"),Transform=(Rotation=(X=0,Y=0,Z=0,W=1),Translation=(X=0,Y=3,Z=4),Scale3D=(X= ...(truncated)"
    WeightMinimum=0
    Weight=@AlphaInterp.Result
   )
   > GetAlphaFromQuatDelta_1 = RigVMFunctionReferenceNode(
    Rotation Order=ZYX
    child=(Type=Bone,Name="clavicle_r")
    debugRotation=false
    onlyTwist=false
    outputMinMax=(X=0,Y=1)
    parent=(Type=Null,Name="clavicle_r_null")
    premult=1
    rotationMinMax=(X=0,Y=-40)
    triggerAxis=Y
    twistAxis=(X=1,Y=0,Z=0)
   )
   > GetAlphaFromQuatDelta_2 = RigVMFunctionReferenceNode(
    Rotation Order=ZYX
    child=(Type=Bone,Name="upperarm_r")
    debugRotation=false
    onlyTwist=false
    outputMinMax=(X=0,Y=1)
    parent=(Type=Null,Name="upperarm_r_null")
    premult=1
    rotationMinMax=(X=0,Y=-35)
    triggerAxis=Y
    twistAxis=(X=1,Y=0,Z=0)
   )
   Multiply = &RigVMFunction_MathDoubleMul(A=@GetAlphaFromQuatDelta_2.alpha, B=@GetAlphaFromQuatDelta_1.alpha)
   RigVMFunction_MathBoolOr = &RigVMFunction_MathBoolOr(A=RigVMVariableNode() {variable=isPlayingEmote}.Value, B=RigVMVariableNode() {variable=isParachuteOpened}.Value)
   RigVMFunction_MathBoolToInteger = &RigVMFunction_MathBoolToInteger(Value=@RigVMFunction_MathBoolOr.Result)
   n3 = &RigVMFunction_AlphaInterp(
    InterpSpeedDecreasing=10
    InterpSpeedIncreasing=20
    Scale=1
    bClampResult=False
    Value=RigVMDispatchNode(
     Values=(0,0)
     Values.0=&RigVMFunction_MathDoubleMul(B=0.5, A=@Multiply.Result).Result
     Index=@RigVMFunction_MathBoolToInteger.Result
     Values.1=@Multiply.Result
    ) {template="DISPATCH_RigVMDispatch_SelectInt32(in Index,in Values,out Result)"}.Result
   )
   > &RigUnit_ModifyTransforms(
    ItemToModify="((Item=(Type=Null,Name=\"dyn_ribbon_r_02_null\"),Transform=(Rotation=(X=0,Y=0,Z=0,W=1),Translation=(X=-4,Y=-8,Z=-2),Scale3D=(X=1,Y=1,Z=1))),(Item=(Type=Null,Name=\"dyn_ribbon_r_03_null\"),Transform=(Rotation=(X=0,Y=0,Z=0,W=1),Translation=(X=-2,Y=-3,Z=2),Scale3 ...(truncated)"
    WeightMinimum=0
    Weight=@n3.Result
   )
   > GetAlphaFromQuatDelta_3 = RigVMFunctionReferenceNode(
    Rotation Order=ZYX
    child=(Type=Bone,Name="upperarm_r")
    debugRotation=false
    onlyTwist=false
    outputMinMax=(X=0,Y=1)
    parent=(Type=Null,Name="upperarm_r_null")
    premult=1
    rotationMinMax=(X=0,Y=140)
    triggerAxis=Z
    twistAxis=(X=1,Y=0,Z=0)
   )
   > &RigUnit_ModifyTransforms(
    ItemToModify="((Item=(Type=Null,Name=\"dyn_ribbon_r_02_null\"),Transform=(Rotation=(X=0,Y=0,Z=0,W=1),Translation=(X=-4,Y=-5,Z=0),Scale3D=(X=1,Y=1,Z=1))),(Item=(Type=Null,Name=\"dyn_ribbon_r_03_null\"),Transform=(Rotation=(X=0,Y=0,Z=0,W=1),Translation=(X=-2,Y=0,Z=2),Scale3D= ...(truncated)"
    WeightMinimum=0
    Weight=@GetAlphaFromQuatDelta_3.alpha
   )
  ]}
 CollapseNode_ContainedGraph {nodes [
   RigVMFunctionEntryNode()
   > RigVMFunction_Sequence_1 = RigVMAggregateNode()
   &RigUnit_ParentConstraint(
    AdvancedSettings=(RotationOrderForFilter=XZY)
    Child=(Type=Null,Name="dyn_ribbon_l_01_null")
    Parents=((Item=(Type=Bone,Name="base_ribbon_l_01"),Weight=1))
    Weight=1
    ExecutePin=@RigVMFunction_Sequence_1.A
   )
   > &RigUnit_ParentConstraint(
    AdvancedSettings=(RotationOrderForFilter=XZY)
    Child=(Type=Null,Name="dyn_ribbon_l_02_null")
    Parents=((Item=(Type=Bone,Name="spine_03"),Weight=0.0046),(Item=(Type=Bone,Name="neck_01"),Weight=0.0053),(Item=(Type=Bone,Name="spine_05"),Weight=0.4807),(Item=(Type=Bone,Name="clavicle_r"),Weight=0.5094))
    Weight=1
   )
   > &RigUnit_ParentConstraint(
    AdvancedSettings=(RotationOrderForFilter=XZY)
    Child=(Type=Null,Name="dyn_ribbon_l_03_null")
    Parents="((Item=(Type=Bone,Name=\"deform_upperarm_up_r\"),Weight=0.0013),(Item=(Type=Bone,Name=\"clavicle_l\"),Weight=0.0028),(Item=(Type=Bone,Name=\"deform_upperarm_bk_r\"),Weight=0.0207),(Item=(Type=Bone,Name=\"upperarm_twist_01_r\"),Weight=0.0304),(Item=(Type=Bone,Name= ...(truncated)"
    Weight=1
   )
   > &RigUnit_ParentConstraint(
    AdvancedSettings=(RotationOrderForFilter=XZY)
    Child=(Type=Null,Name="dyn_ribbon_l_04_null")
    Parents="((Item=(Type=Bone,Name=\"spine_01\"),Weight=0.0083),(Item=(Type=Bone,Name=\"clavicle_l\"),Weight=0.0283),(Item=(Type=Bone,Name=\"upperarm_twist_01_r\"),Weight=0.04),(Item=(Type=Bone,Name=\"spine_02\"),Weight=0.0453),(Item=(Type=Bone,Name=\"spine_03\"),Weight=0.0457) ...(truncated)"
    Weight=1
   )
   > &RigUnit_ParentConstraint(
    AdvancedSettings=(RotationOrderForFilter=XZY)
    Child=(Type=Null,Name="dyn_ribbon_l_05_null")
    Parents="((Item=(Type=Bone,Name=\"clavicle_l\"),Weight=0.0067),(Item=(Type=Bone,Name=\"pelvis\"),Weight=0.0357),(Item=(Type=Bone,Name=\"clavicle_r\"),Weight=0.0559),(Item=(Type=Bone,Name=\"spine_05\"),Weight=0.0657),(Item=(Type=Bone,Name=\"spine_01\"),Weight=0.0755),(Item=(T ...(truncated)"
    Weight=1
   )
   > &RigUnit_ParentConstraint(
    AdvancedSettings=(RotationOrderForFilter=XZY)
    Child=(Type=Null,Name="dyn_ribbon_l_06_null")
    Parents=((Item=(Type=Bone,Name="spine_01"),Weight=0.02),(Item=(Type=Bone,Name="spine_02"),Weight=0.045),(Item=(Type=Bone,Name="pelvis"),Weight=0.953))
    Weight=1
   )
   > &RigUnit_ParentConstraint(
    AdvancedSettings=(RotationOrderForFilter=XZY)
    Child=(Type=Null,Name="dyn_ribbon_l_07_null")
    Parents=((Item=(Type=Bone,Name="pelvis"),Weight=1))
    Weight=1
   )
   > &RigUnit_ParentConstraint(
    AdvancedSettings=(RotationOrderForFilter=XZY)
    Child=(Type=Null,Name="dyn_ribbon_l_08_null")
    Parents=((Item=(Type=Bone,Name="pelvis"),Weight=1))
    Weight=1
   )
   &RigUnit_ParentConstraint(
    AdvancedSettings=(RotationOrderForFilter=XZY)
    Child=(Type=Null,Name="dyn_ribbon_r_01_null")
    Parents=((Item=(Type=Bone,Name="base_ribbon_r_01"),Weight=1))
    Weight=1
    ExecutePin=@RigVMFunction_Sequence_1.B
   )
   > &RigUnit_ParentConstraint(
    AdvancedSettings=(RotationOrderForFilter=XZY)
    Child=(Type=Null,Name="dyn_ribbon_r_02_null")
    Parents="((Item=(Type=Bone,Name=\"deform_upperarm_bk_r\"),Weight=0.0187),(Item=(Type=Bone,Name=\"upperarm_twist_01_r\"),Weight=0.0343),(Item=(Type=Bone,Name=\"deform_upperarm_up_r\"),Weight=0.0388),(Item=(Type=Bone,Name=\"spine_05\"),Weight=0.1484),(Item=(Type=Bone,Name=\"c ...(truncated)"
    Weight=1
   )
   > &RigUnit_ParentConstraint(
    AdvancedSettings=(RotationOrderForFilter=XZY)
    Child=(Type=Null,Name="dyn_ribbon_r_03_null")
    Parents="((Item=(Type=Bone,Name=\"spine_04\"),Weight=0.0241),(Item=(Type=Bone,Name=\"deform_upperarm_up_r\"),Weight=0.0594),(Item=(Type=Bone,Name=\"spine_05\"),Weight=0.0698),(Item=(Type=Bone,Name=\"upperarm_twist_01_r\"),Weight=0.1507),(Item=(Type=Bone,Name=\"deform_uppera ...(truncated)"
    Weight=1
   )
   > &RigUnit_ParentConstraint(
    AdvancedSettings=(RotationOrderForFilter=XZY)
    Child=(Type=Null,Name="dyn_ribbon_r_04_null")
    Parents="((Item=(Type=Bone,Name=\"spine_02\"),Weight=0.0033),(Item=(Type=Bone,Name=\"spine_03\"),Weight=0.0125),(Item=(Type=Bone,Name=\"deform_upperarm_bk_r\"),Weight=0.0443),(Item=(Type=Bone,Name=\"upperarm_twist_01_r\"),Weight=0.1273),(Item=(Type=Bone,Name=\"spine_05\"),We ...(truncated)"
    Weight=1
   )
   > &RigUnit_ParentConstraint(
    AdvancedSettings=(RotationOrderForFilter=XZY)
    Child=(Type=Null,Name="dyn_ribbon_r_05_null")
    Parents="((Item=(Type=Bone,Name=\"upperarm_twist_01_r\"),Weight=0.0093),(Item=(Type=Bone,Name=\"pelvis\"),Weight=0.036),(Item=(Type=Bone,Name=\"clavicle_r\"),Weight=0.0579),(Item=(Type=Bone,Name=\"spine_05\"),Weight=0.0595),(Item=(Type=Bone,Name=\"spine_01\"),Weight=0.0715), ...(truncated)"
    Weight=1
   )
   > &RigUnit_ParentConstraint(
    AdvancedSettings=(InterpolationType=Average,RotationOrderForFilter=XZY)
    Child=(Type=Null,Name="dyn_ribbon_r_06_null")
    Parents=((Item=(Type=Bone,Name="pelvis"),Weight=1))
    Weight=1
   )
   > &RigUnit_ParentConstraint(
    AdvancedSettings=(InterpolationType=Average,RotationOrderForFilter=XZY)
    Child=(Type=Null,Name="dyn_ribbon_r_07_null")
    Parents=((Item=(Type=Bone,Name="pelvis"),Weight=1))
    Weight=1
   )
   > &RigUnit_ParentConstraint(
    AdvancedSettings=(RotationOrderForFilter=XZY)
    Child=(Type=Null,Name="dyn_ribbon_r_08_null")
    Parents=((Item=(Type=Bone,Name="pelvis"),Weight=1))
    Weight=1
   )
   > &RigUnit_ParentConstraint(
    AdvancedSettings=(RotationOrderForFilter=XZY)
    Child=(Type=Null,Name="dyn_ribbon_r_09_null")
    Parents=((Item=(Type=Bone,Name="pelvis"),Weight=1))
    Weight=1
   )
   RigVMFunctionReturnNode(ExecuteContext=@RigVMFunction_Sequence_1.C)
  ]}
 Multiply_16_ContainedGraph {nodes [
   n1 = RigVMFunctionEntryNode()
   RigVMFunctionReturnNode(Result=&RigVMFunction_MathDoubleMul(B=@n1.C, A=&RigVMFunction_MathDoubleMul(A=@n1.A, B=@n1.B).Result).Result)
  ]}
 Multiply_5_ContainedGraph {nodes [
   n1 = RigVMFunctionEntryNode()
   RigVMFunctionReturnNode(Result=&RigVMFunction_MathDoubleMul(B=@n1.C, A=&RigVMFunction_MathDoubleMul(A=@n1.A, B=@n1.B).Result).Result)
  ]}
 Multiply_5_ContainedGraph {nodes [
   n1 = RigVMFunctionEntryNode()
   RigVMFunctionReturnNode(Result=&RigVMFunction_MathDoubleMul(B=@n1.C, A=&RigVMFunction_MathDoubleMul(A=@n1.A, B=@n1.B).Result).Result)
  ]}
 Multiply_6_ContainedGraph {nodes [
   n1 = RigVMFunctionEntryNode()
   n2 = &RigVMFunction_MathDoubleMul(A=@n1.A, B=@n1.B)
   RigVMFunctionReturnNode(Result=&RigVMFunction_MathDoubleMul(B=@n1.D, A=&RigVMFunction_MathDoubleMul(B=@n1.C, A=@n2.Result).Result).Result)
  ]}
 Multiply_8_ContainedGraph {nodes [
   n1 = RigVMFunctionEntryNode()
   RigVMFunctionReturnNode(Result=&RigVMFunction_MathDoubleMul(B=@n1.C, A=&RigVMFunction_MathDoubleMul(A=@n1.A, B=@n1.B).Result).Result)
  ]}
 New Function_ContainedGraph {nodes [
   RigVMFunctionEntryNode()
   RigVMFunction_MathBoolToInteger = &RigVMFunction_MathBoolToInteger(Value=RigVMVariableNode() {variable=onlyTwist}.Value)
   RigVMFunction_MathQuaternionSwingTwist = &RigVMFunction_MathQuaternionSwingTwist(
    Input=&RigUnit_GetRelativeTransformForItem(
     bChildInitial=False
     bParentInitial=False
     Child=RigVMVariableNode() {variable=child}.Value
     Parent=RigVMVariableNode() {variable=parent}.Value
    ).RelativeTransform.Rotation
    TwistAxis=RigVMVariableNode() {variable=twistAxis}.Value
   )
   n1 = RigVMDispatchNode(
    Values=((),())
    Index=@RigVMFunction_MathBoolToInteger.Result
    Values.0=@RigVMFunction_MathQuaternionSwingTwist.Swing
    Values.1=@RigVMFunction_MathQuaternionSwingTwist.Twist
   ) {template="DISPATCH_RigVMDispatch_SelectInt32(in Index,in Values,out Result)"}
   > RigVMVariableNode(Value=@n1.Result) {variable=entry_outputQuat}
   RigVMFunction_MathQuaternionToEuler = &RigVMFunction_MathQuaternionToEuler(Value=@n1.Result, RotationOrder=RigVMVariableNode() {variable=Rotation Order}.Value)
   > DISPATCH_RigVMDispatch_Print = RigVMDispatchNode(
    ScreenColor=(R=1,G=1,B=1,A=1)
    ScreenDuration=1.5
    Value=@RigVMFunction_MathQuaternionToEuler.Result
    Enabled=RigVMVariableNode() {variable=debugRotation}.Value
   ) {template="DISPATCH_RigVMDispatch_Print(in Prefix,in Value,in Enabled,in ScreenDuration,in ScreenColor)"}
   Cast = RigVMDispatchNode(Value=RigVMVariableNode() {variable=triggerAxis}.Value) {template="DISPATCH_RigVMDispatch_CastEnumToInt(in Value,out Result)"}
   DISPATCH_RigVMDispatch_SelectInt32 = RigVMDispatchNode(
    Values=(0,0,0)
    Index=@Cast.Result
    Values.0=@RigVMFunction_MathQuaternionToEuler.Result.X
    Values.1=@RigVMFunction_MathQuaternionToEuler.Result.Y
    Values.2=@RigVMFunction_MathQuaternionToEuler.Result.Z
   ) {template="DISPATCH_RigVMDispatch_SelectInt32(in Index,in Values,out Result)"}
   n2 = RigVMVariableNode() {variable=rotationMinMax}
   n3 = RigVMVariableNode() {variable=outputMinMax}
   Remap = &RigVMFunction_MathDoubleRemap(
    bClamp=False
    Value=@DISPATCH_RigVMDispatch_SelectInt32.Result
    SourceMinimum=@n2.Value.X
    SourceMaximum=@n2.Value.Y
    TargetMinimum=@n3.Value.X
    TargetMaximum=@n3.Value.Y
   )
   Absolute = &RigVMFunction_MathDoubleAbs(Value=@Remap.Result)
   n4 = RigVMVariableNode() {variable=absoluteValue}
   Multiply = &RigVMFunction_MathDoubleMul(
    A=&RigVMFunction_MathDoubleClamp(Value=RigVMDispatchNode(True=@Absolute.Result, False=@Remap.Result, Condition=@n4.Value) {template="DISPATCH_RigVMDispatch_If(in Condition,in True,in False,out Result)"}.Result).Result
    B=RigVMVariableNode() {variable=premult}.Value
   )
   > RigVMFunctionReturnNode(alpha=@Multiply.Result, rawQuat=RigVMVariableNode() {variable=entry_outputQuat}.Value)
  ]}
 RigVMFunctionLibrary {nodes [GetAlphaFromQuatDelta = RigVMCollapseNode(
    ExecuteContext=()
    Rotation Order=ZYX
    child=(Type=Bone,Name="None")
    outputMinMax=(X=0,Y=1)
    parent=(Type=Bone,Name="None")
    premult=1
    rotationMinMax=(X=0,Y=0)
    twistAxis=(X=1,Y=0,Z=0)
   )]}
 RigVMFunction_MathBoolAnd_1_ContainedGraph {nodes [
   n1 = RigVMFunctionEntryNode()
   RigVMFunction_MathBoolAnd = &RigVMFunction_MathBoolAnd(A=@n1.A, B=@n1.B)
   RigVMFunctionReturnNode(Result=&RigVMFunction_MathBoolAnd(B=@n1.C, A=@RigVMFunction_MathBoolAnd.Result).Result)
  ]}
 RigVMFunction_Sequence_1_ContainedGraph {nodes [
   n1 = RigVMFunctionEntryNode()
   RigVMFunction_Sequence = &RigVMFunction_Sequence(ExecuteContext=@n1.ExecuteContext)
   n2 = &RigVMFunction_Sequence(ExecuteContext=@RigVMFunction_Sequence.B)
   RigVMFunctionReturnNode(A=@RigVMFunction_Sequence.A, B=@n2.A, C=@n2.B)
  ]}
 RigVMFunction_Sequence_1_ContainedGraph {nodes [
   n1 = RigVMFunctionEntryNode()
   RigVMFunction_Sequence = &RigVMFunction_Sequence(ExecuteContext=@n1.ExecuteContext)
   n2 = &RigVMFunction_Sequence(ExecuteContext=@RigVMFunction_Sequence.B)
   RigVMFunctionReturnNode(A=@RigVMFunction_Sequence.A, B=@n2.A, C=@n2.B)
  ]}
 RigVMFunction_Sequence_1_ContainedGraph {nodes [
   n1 = RigVMFunctionEntryNode()
   RigVMFunction_Sequence = &RigVMFunction_Sequence(ExecuteContext=@n1.ExecuteContext)
   n2 = &RigVMFunction_Sequence(ExecuteContext=@RigVMFunction_Sequence.B)
   n3 = &RigVMFunction_Sequence(ExecuteContext=@n2.B)
   n4 = &RigVMFunction_Sequence(ExecuteContext=@n3.B)
   RigVMFunctionReturnNode(A=@RigVMFunction_Sequence.A, B=@n2.A, C=@n3.A, D=@n4.A, E=@n4.B)
  ]}
 RigVMFunction_Sequence_1_ContainedGraph {nodes [
   n1 = RigVMFunctionEntryNode()
   RigVMFunction_Sequence = &RigVMFunction_Sequence(ExecuteContext=@n1.ExecuteContext)
   n2 = &RigVMFunction_Sequence(ExecuteContext=@RigVMFunction_Sequence.B)
   RigVMFunctionReturnNode(A=@RigVMFunction_Sequence.A, B=@n2.A, C=@n2.B)
  ]}
 RigVMModel {nodes [
   n1 = &RigUnit_PrepareForExecution()
   RigUnit_GetTransform = &RigUnit_GetTransform(Item=(Type=Bone,Name="dyn_ribbon_r_02"), bInitial=true)
   RigUnit_SetTransform = &RigUnit_SetTransform(
    Item=(Type=Null,Name="dyn_ribbon_r_02_null")
    Weight=1
    bInitial=true
    bPropagateToChildren=true
    ExecutePin=@n1.ExecutePin
    Value=@RigUnit_GetTransform.Transform
   )
   RigUnit_BeginExecution = &RigUnit_BeginExecution()
   Constraint = RigVMCollapseNode(ExecuteContext=@RigUnit_BeginExecution.ExecutePin)
   > DefiningVariables = RigVMCollapseNode()
   > PonyTailSetups = RigVMCollapseNode()
   > SplineIKs = RigVMCollapseNode()
   > &RigUnit_ModifyTransforms(
    ItemToModify="((Item=(Type=Bone,Name=\"dyn_ponytail_03\"),Transform=(Rotation=(X=0,Y=0,Z=0.1305,W=0.9914),Translation=(X=0,Y=0,Z=0),Scale3D=(X=1,Y=1,Z=1))),(Item=(Type=Bone,Name=\"dyn_ribbon_r_03\"),Transform=(Rotation=(X=0,Y=0,Z=0.0175,W=0.9998),Translation=(X=0,Y=0,Z=0),S ...(truncated)"
    WeightMinimum=0
    Weight=&RigVMFunction_MathBoolToFloat(Value=RigVMVariableNode() {variable=isWearingCape}.Value).Result
   )
  ]}
]
hierarchy [
 root {loc=<0 0 0>, rot=<0 0 0 1>}
 pelvis {parent=root, loc=<0 -0.5301 98.0571>, rot=<0.0127 0.707 -0.0127 -0.707>}
 spine_01 {parent=pelvis, loc=<6.1208 -0.9328 0>, rot=<0 0 0.034 -0.9994>}
 spine_02 {parent=spine_01, loc=<6.5997 -0.2095 0>, rot=<0 0 0.0238 -0.9997>}
 spine_03 {parent=spine_02, loc=<6.4914 -0.1237 0>, rot=<0 0 0.0149 -0.9999>}
 spine_04 {parent=spine_03, loc=<6.9598 0.0794 0>, rot=<0 0 -0.0132 -0.9999>}
 spine_05 {parent=spine_04, loc=<6.9987 0 0>, rot=<0 0 -0.0015 -1>}
 neck_01 {parent=spine_05, loc=<11.3138 0.1529 0>, rot=<0 0 -0.2003 -0.9797>}
 neck_02 {parent=neck_01, loc=<4.5627 0 0>, rot=<0 0 0.0504 -0.9987>}
 head {parent=neck_02, loc=<5.3914 0 0>, rot=<0 0 0.115 -0.9934>}
 base_ponytail_01 {parent=head, loc=<3.3598 -4.0934 0.9259>, rot=<0.202 -0.9794 0 0>}
 dyn_ponytail_01 {parent=base_ponytail_01, loc=<0 0 0>, rot=<0 0 0 1>}
 dyn_ponytail_02 {parent=dyn_ponytail_01, loc=<5.3524 0 0>, rot=<0 0 0.1268 -0.9919>}
 dyn_ponytail_03 {parent=dyn_ponytail_02, loc=<8.3717 0 0>, rot=<0 0 -0.1771 -0.9842>}
 dyn_ponytail_04 {parent=dyn_ponytail_03, loc=<6.8164 -0.0038 0>, rot=<0 0 0.0157 -0.9999>}
 base_frontRibbonCenter {parent=head, loc=<11.3394 8.129 -3.3899>, rot=<-0.4502 -0.8351 0.1016 0.2992>}
 dyn_frontRibbonCenter {parent=base_frontRibbonCenter, loc=<0 0 0>, rot=<0 0 0 1>}
 dyn_RibbonUp {parent=dyn_frontRibbonCenter, loc=<0.0544 0.7337 -0.7145>, rot=<0.4457 -0.4945 -0.5241 -0.5312>}
 dyn_RibbonDn {parent=dyn_frontRibbonCenter, loc=<0.0676 0.5399 0.7103>, rot=<0.482 0.4961 0.5308 -0.4897>}
 base_frontHair_c {parent=head, loc=<8.9591 7.5203 -3.4418>, rot=<-0.0664 -0.9264 -0.06 -0.3657>}
 dyn_frontHair_c {parent=base_frontHair_c, loc=<0 0 0>, rot=<0 0 0 1>}
 dyn_frontHair_f_01 {parent=dyn_frontHair_c, loc=<2.764 2.1157 -0.261>, rot=<-0.0295 -0.1061 0.1582 -0.9812>}
 dyn_frontHair_f_02 {parent=dyn_frontHair_f_01, loc=<4.7619 0 0>, rot=<-0.0119 -0.2064 0.1494 -0.9669>}
 dyn_frontHair_b_01 {parent=dyn_frontHair_c, loc=<3.197 -2.0521 0.1055>, rot=<-0.0364 -0.1253 0.1045 -0.9859>}
 dyn_frontHair_b_02 {parent=dyn_frontHair_b_01, loc=<3.4736 0 0>, rot=<-0.0105 -0.2232 0.0524 -0.9733>}
 base_BandanaKnot {parent=head, loc=<2.5922 -6.2949 4.5101>, rot=<0.5506 -0.808 0.1561 0.1401>}
 dyn_BandanaKnot {parent=base_BandanaKnot, loc=<0 0 0>, rot=<0 0 0 1>}
 base_ribbon_l_01 {parent=dyn_BandanaKnot, loc=<1.3003 -0.34 1.016>, rot=<0.1102 0.0401 -0.2582 -0.9589>}
 dyn_ribbon_l_01 {parent=base_ribbon_l_01, loc=<0 0 0>, rot=<0 0 0 1>}
 dyn_ribbon_l_02 {parent=dyn_ribbon_l_01, loc=<9.2939 0.0038 0.0244>, rot=<-0.0623 0.0962 -0.2376 -0.9646>}
 dyn_ribbon_l_03 {parent=dyn_ribbon_l_02, loc=<10.1961 0 0>, rot=<-0.0005 -0.0173 -0.1519 -0.9882>}
 dyn_ribbon_l_04 {parent=dyn_ribbon_l_03, loc=<10.2687 0 0>, rot=<-0.0041 -0.0303 0.0146 -0.9994>}
 dyn_ribbon_l_05 {parent=dyn_ribbon_l_04, loc=<10.2631 0 0>, rot=<-0.0032 -0.0363 0.0665 -0.9971>}
 dyn_ribbon_l_06 {parent=dyn_ribbon_l_05, loc=<10.273 0 0>, rot=<0.0024 -0.022 0.0742 -0.997>}
 dyn_ribbon_l_07 {parent=dyn_ribbon_l_06, loc=<10.2657 0 0>, rot=<0.0033 -0.0016 0.0621 -0.9981>}
 dyn_ribbon_l_08 {parent=dyn_ribbon_l_07, loc=<10.2788 0 0>, rot=<0 0 0 1>}
 base_ribbon_r_01 {parent=dyn_BandanaKnot, loc=<0.9997 -0.5198 -1.2164>, rot=<0.0783 -0.0811 -0.2828 -0.9525>}
 dyn_ribbon_r_01 {parent=base_ribbon_r_01, loc=<0 0 0>, rot=<0 0 0 1>}
 dyn_ribbon_r_02 {parent=dyn_ribbon_r_01, loc=<8.2395 0 -0.4261>, rot=<-0.0409 0.0569 -0.1711 -0.9828>}
 dyn_ribbon_r_03 {parent=dyn_ribbon_r_02, loc=<11.3459 0 0>, rot=<-0.0175 0.0646 -0.1486 -0.9866>}
 dyn_ribbon_r_04 {parent=dyn_ribbon_r_03, loc=<11.3752 0 0>, rot=<0.0018 0.0129 -0.0324 -0.9994>}
 dyn_ribbon_r_05 {parent=dyn_ribbon_r_04, loc=<11.3975 0 0>, rot=<-0.0042 -0.0396 0.0398 -0.9984>}
 dyn_ribbon_r_06 {parent=dyn_ribbon_r_05, loc=<11.3937 0 0>, rot=<0.0005 -0.0321 0.0786 -0.9964>}
 dyn_ribbon_r_07 {parent=dyn_ribbon_r_06, loc=<11.3786 0 0>, rot=<0.0033 -0.0076 0.0636 -0.9979>}
 dyn_ribbon_r_08 {parent=dyn_ribbon_r_07, loc=<11.3867 0 0>, rot=<0.0118 -0.0235 0.0785 -0.9966>}
 dyn_ribbon_r_09 {parent=dyn_ribbon_r_08, loc=<11.3776 0 0>, rot=<0 0 0 1>}
 base_NoMask_frontHair_c {parent=head, loc=<8.3479 7.4032 -4.0297>, rot=<-0.0664 -0.9264 -0.06 -0.3657>}
 dyn_NoMask_frontHair_c {parent=base_NoMask_frontHair_c, loc=<0 0 0>, rot=<0 0 0 1>}
 dyn_frontHair_NoMask_f_01 {parent=dyn_NoMask_frontHair_c, loc=<1.9476 2.3734 -0.261>, rot=<-0.0295 -0.1061 0.1582 -0.9812>}
 dyn_frontHair_NoMask_f_02 {parent=dyn_frontHair_NoMask_f_01, loc=<4.7619 0 0>, rot=<-0.0119 -0.2064 0.1494 -0.9669>}
 dyn_frontHair_NoMask_b_01 {parent=dyn_NoMask_frontHair_c, loc=<2.3806 -1.7944 0.1055>, rot=<-0.0364 -0.1253 0.1045 -0.9859>}
 dyn_frontHair_NoMask_b_02 {parent=dyn_frontHair_NoMask_b_01, loc=<3.4736 0 0>, rot=<-0.0105 -0.2232 0.0524 -0.9733>}
 clavicle_l {parent=spine_05, loc=<5.4831 -0.1053 -3.2946>, rot=<0.0465 -0.7289 -0.009 -0.6829>}
 upperarm_l {parent=clavicle_l, loc=<9.8658 -1.2136 -0.4105>, rot=<0.0164 -0.3245 -0.0269 -0.9454>}
 lowerarm_l {parent=upperarm_l, loc=<26.465 -0.8345 0.259>, rot=<0.0036 0.0161 -0.2298 -0.9731>}
 hand_l {parent=lowerarm_l, loc=<22.7355 -0.3203 -0.6105>, rot=<-0.6784 -0.0223 0.0735 -0.7307>}
 thumb_01_l {parent=hand_l, loc=<1.3118 -0.3191 -2.459>, rot=<0.6546 -0.4042 0.0664 -0.6354>}
 thumb_02_l {parent=thumb_01_l, loc=<5.9336 -0.0448 0.0401>, rot=<-0.0037 0.0213 0.298 -0.9543>}
 thumb_03_l {parent=thumb_02_l, loc=<2.2223 0.3682 -0.0907>, rot=<0.0118 -0.013 0.0066 -0.9998>}
 index_metacarpal_l {parent=hand_l, loc=<1.6535 0.0183 -2.96>, rot=<0.0158 -0.0124 -0.0173 -0.9996>}
 index_01_l {parent=index_metacarpal_l, loc=<8.3008 0.0357 -0.0524>, rot=<-0.0044 0.0485 0.0924 -0.9945>}
 index_02_l {parent=index_01_l, loc=<3.3587 -0.3004 0.0452>, rot=<-0.002 0.0027 0.3225 -0.9466>}
 index_03_l {parent=index_02_l, loc=<2.5578 -0.123 -0.0252>, rot=<-0.002 0.0067 0.2558 -0.9667>}
 middle_metacarpal_l {parent=hand_l, loc=<2.1856 -0.1725 -0.5984>, rot=<0.006 0.0168 -0.0329 -0.9993>}
 middle_01_l {parent=middle_metacarpal_l, loc=<7.3275 -0.2447 0.0129>, rot=<-0.0208 0.0042 0.0858 -0.9961>}
 middle_02_l {parent=middle_01_l, loc=<4.0123 0.1401 0.1175>, rot=<-0.0218 0.0258 0.3512 -0.9357>}
 middle_03_l {parent=middle_02_l, loc=<2.6077 0.0534 0.1373>, rot=<0.0082 0.0188 0.1708 -0.9851>}
 ring_metacarpal_l {parent=hand_l, loc=<2.2074 -0.3625 1.3638>, rot=<-0.0284 0.0722 -0.0458 -0.9959>}
 ring_01_l {parent=ring_metacarpal_l, loc=<7.0295 -0.6197 -0.1446>, rot=<-0.0489 -0.0252 0.1892 -0.9804>}
 ring_02_l {parent=ring_01_l, loc=<3.6174 0.4647 -0.1027>, rot=<0 0 0.2806 -0.9598>}
 ring_03_l {parent=ring_02_l, loc=<2.4152 0.114 0.0199>, rot=<0 0 0.1189 -0.9929>}
 pinky_metacarpal_l {parent=hand_l, loc=<1.7616 -0.9753 3.3814>, rot=<-0.0397 0.0814 -0.069 -0.9935>}
 pinky_01_l {parent=pinky_metacarpal_l, loc=<7.1709 -0.4831 -0.1094>, rot=<-0.0654 -0.0035 0.2242 -0.9723>}
 pinky_02_l {parent=pinky_01_l, loc=<2.8274 0.0893 -0.1043>, rot=<0.0037 0.0011 0.2854 -0.9584>}
 pinky_03_l {parent=pinky_02_l, loc=<2.2046 0.0307 -0.0078>, rot=<-0.0002 0.0098 0.1445 -0.9895>}
 deform_wrist_dn_base_l {parent=hand_l, loc=<0 -1.5326 0>, rot=<0 0 0 -1>}
 deform_wrist_dn_l {parent=deform_wrist_dn_base_l, loc=<0 0 0>, rot=<0 0 0 1>}
 deform_wrist_up_base_l {parent=hand_l, loc=<0 2.3009 0>, rot=<0 0 0 -1>}
 deform_wrist_up_l {parent=deform_wrist_up_base_l, loc=<0 0 0>, rot=<0 0 0 1>}
 lowerarm_twist_02_l {parent=lowerarm_l, loc=<10.6707 0 0>, rot=<0 0 0 1>}
 lowerarm_twist_01_l {parent=lowerarm_l, loc=<18.316 0 0>, rot=<0 0 0 1>}
 deform_elbow_in_base_l {parent=lowerarm_l, loc=<0.1834 3.6062 -0.2572>, rot=<0 0 0 -1>}
 deform_elbow_in_l {parent=deform_elbow_in_base_l, loc=<0 0 0>, rot=<0 0 0 1>}
 deform_elbow_out_base_l {parent=lowerarm_l, loc=<0 -4 0>, rot=<0 0 0 1>}
 deform_elbow_out_l {parent=deform_elbow_out_base_l, loc=<0 0 0>, rot=<0 0 0 1>}
 upperarm_twist_01_l {parent=upperarm_l, loc=<7.2135 -0.1134 0>, rot=<0 0 0 1>}
 deform_upperarm_bk_base_l {parent=upperarm_twist_01_l, loc=<-6.4053 -4.3746 0.314>, rot=<0 0 0 -1>}
 deform_upperarm_bk_l {parent=deform_upperarm_bk_base_l, loc=<0 0 0>, rot=<0 0 0 1>}
 deform_upperarm_fr_base_l {parent=upperarm_twist_01_l, loc=<-5.6312 5.092 0.3354>, rot=<0 0 0 -1>}
 deform_upperarm_fr_l {parent=deform_upperarm_fr_base_l, loc=<0 0 0>, rot=<0 0 0 1>}
 deform_upperarm_up_base_l {parent=upperarm_twist_01_l, loc=<-8.891 0.4825 3.9517>, rot=<0 0 0 -1>}
 deform_upperarm_up_l {parent=deform_upperarm_up_base_l, loc=<0 0 0>, rot=<0 0 0 1>}
 upperarm_twist_02_l {parent=upperarm_l, loc=<19.7998 -0.1134 0>, rot=<0 0 0 1>}
 clavicle_r {parent=spine_05, loc=<5.4831 -0.1053 3.2946>, rot=<0.7289 0.0465 -0.6829 0.009>}
 upperarm_r {parent=clavicle_r, loc=<-9.8658 1.2136 0.4105>, rot=<0.0164 -0.3245 -0.0269 -0.9454>}
 upperarm_twist_01_r {parent=upperarm_r, loc=<-7.2135 0.1134 0>, rot=<0 0 0 -1>}
 deform_upperarm_bk_base_r {parent=upperarm_twist_01_r, loc=<6.4053 4.3746 -0.314>, rot=<0 0 0 -1>}
 deform_upperarm_bk_r {parent=deform_upperarm_bk_base_r, loc=<0 0 0>, rot=<0 0 0 1>}
 deform_upperarm_fr_base_r {parent=upperarm_twist_01_r, loc=<5.6312 -5.092 -0.3354>, rot=<0 0 0 -1>}
 deform_upperarm_fr_r {parent=deform_upperarm_fr_base_r, loc=<0 0 0>, rot=<0 0 0 1>}
 deform_upperarm_up_base_r {parent=upperarm_twist_01_r, loc=<8.891 -0.4825 -3.9517>, rot=<0 0 0 -1>}
 deform_upperarm_up_r {parent=deform_upperarm_up_base_r, loc=<0 0 0>, rot=<0 0 0 1>}
 upperarm_twist_02_r {parent=upperarm_r, loc=<-19.7998 0.1134 0>, rot=<0 0 0 -1>}
 lowerarm_r {parent=upperarm_r, loc=<-26.465 0.8345 -0.259>, rot=<0.0036 0.0161 -0.2298 -0.9731>}
 hand_r {parent=lowerarm_r, loc=<-22.7355 0.3203 0.6106>, rot=<-0.6784 -0.0223 0.0735 -0.7307>}
 thumb_01_r {parent=hand_r, loc=<-1.3118 0.3191 2.459>, rot=<0.6546 -0.4042 0.0664 -0.6354>}
 thumb_02_r {parent=thumb_01_r, loc=<-5.9336 0.0448 -0.0401>, rot=<-0.0037 0.0213 0.298 -0.9543>}
 thumb_03_r {parent=thumb_02_r, loc=<-2.2223 -0.3682 0.0907>, rot=<0.0118 -0.013 0.0066 -0.9998>}
 index_metacarpal_r {parent=hand_r, loc=<-1.6535 -0.0183 2.96>, rot=<0.0158 -0.0124 -0.0173 -0.9996>}
 index_01_r {parent=index_metacarpal_r, loc=<-8.3008 -0.0357 0.0524>, rot=<-0.0044 0.0485 0.0924 -0.9945>}
 index_02_r {parent=index_01_r, loc=<-3.3587 0.3004 -0.0452>, rot=<-0.002 0.0027 0.3225 -0.9466>}
 index_03_r {parent=index_02_r, loc=<-2.5578 0.123 0.0252>, rot=<-0.002 0.0067 0.2558 -0.9667>}
 middle_metacarpal_r {parent=hand_r, loc=<-2.1856 0.1725 0.5984>, rot=<0.006 0.0168 -0.0329 -0.9993>}
 middle_01_r {parent=middle_metacarpal_r, loc=<-7.3275 0.2447 -0.0129>, rot=<-0.0208 0.0042 0.0858 -0.9961>}
 middle_02_r {parent=middle_01_r, loc=<-4.0123 -0.1401 -0.1175>, rot=<-0.0218 0.0258 0.3512 -0.9357>}
 middle_03_r {parent=middle_02_r, loc=<-2.6077 -0.0534 -0.1373>, rot=<0.0082 0.0188 0.1708 -0.9851>}
 ring_metacarpal_r {parent=hand_r, loc=<-2.2074 0.3625 -1.3638>, rot=<-0.0284 0.0722 -0.0458 -0.9959>}
 ring_01_r {parent=ring_metacarpal_r, loc=<-7.0295 0.6197 0.1446>, rot=<-0.0489 -0.0252 0.1892 -0.9804>}
 ring_02_r {parent=ring_01_r, loc=<-3.6174 -0.4647 0.1027>, rot=<0 0 0.2806 -0.9598>}
 ring_03_r {parent=ring_02_r, loc=<-2.4152 -0.114 -0.0199>, rot=<0 0 0.1189 -0.9929>}
 pinky_metacarpal_r {parent=hand_r, loc=<-1.7616 0.9753 -3.3814>, rot=<-0.0397 0.0814 -0.069 -0.9935>}
 pinky_01_r {parent=pinky_metacarpal_r, loc=<-7.1709 0.4831 0.1094>, rot=<-0.0654 -0.0035 0.2242 -0.9723>}
 pinky_02_r {parent=pinky_01_r, loc=<-2.8274 -0.0893 0.1043>, rot=<0.0037 0.0011 0.2854 -0.9584>}
 pinky_03_r {parent=pinky_02_r, loc=<-2.2046 -0.0307 0.0078>, rot=<-0.0002 0.0098 0.1445 -0.9895>}
 deform_wrist_dn_base_r {parent=hand_r, loc=<-0.0001 1.5328 0>, rot=<0 0 0 -1>}
 deform_wrist_dn_r {parent=deform_wrist_dn_base_r, loc=<0 0 0>, rot=<0 0 0 1>}
 deform_wrist_up_base_r {parent=hand_r, loc=<-0.0002 -2.3007 -0.0001>, rot=<0 0 0 -1>}
 deform_wrist_up_r {parent=deform_wrist_up_base_r, loc=<0 0 0>, rot=<0 0 0 1>}
 lowerarm_twist_02_r {parent=lowerarm_r, loc=<-10.6707 0 0>, rot=<0 0 0 -1>}
 lowerarm_twist_01_r {parent=lowerarm_r, loc=<-18.316 0 0>, rot=<0 0 0 -1>}
 deform_elbow_in_base_r {parent=lowerarm_r, loc=<-0.1834 -3.6062 0.2572>, rot=<0 0 0 -1>}
 deform_elbow_in_r {parent=deform_elbow_in_base_r, loc=<0 0 0>, rot=<0 0 0 1>}
 deform_elbow_out_base_r {parent=lowerarm_r, loc=<0 4 -0.0001>, rot=<0 0 0 -1>}
 deform_elbow_out_r {parent=deform_elbow_out_base_r, loc=<0 0 0>, rot=<0 0 0 1>}
 upperarm_r_null {parent=clavicle_r, type=Null, loc=<-9.8658 1.2136 0.4105>, rot=<0.0164 -0.3245 -0.0269 -0.9454>}
 clavicle_r_null {parent=spine_05, type=Null, loc=<5.4831 -0.1053 3.2946>, rot=<0.7289 0.0465 -0.6829 0.009>}
 clavicle_l_null {parent=spine_05, type=Null, loc=<5.483 -0.1053 -3.2946>, rot=<0.0465 -0.7289 -0.009 -0.6829>}
 head_null {parent=spine_05, type=Null, loc=<20.6573 3.5489 0>, rot=<0 0 -0.036 -0.9994>}
 deform_pec_aim_base_l {parent=spine_04, loc=<3.2456 9.3309 -1.2237>, rot=<0.0603 -0.4033 0.3318 -0.8507>}
 deform_pec_aim_l {parent=deform_pec_aim_base_l, loc=<0 0 0>, rot=<0 0 0 1>}
 deform_pec_base_l {parent=deform_pec_aim_l, loc=<9.3665 0 0>, rot=<0 0 0 1>}
 deform_pec_l {parent=deform_pec_base_l, loc=<0 0 0>, rot=<0 0 0 1>}
 deform_pec_aim_base_r {parent=spine_04, loc=<3.2452 9.3309 1.2237>, rot=<-0.4033 -0.0603 0.8507 0.3318>}
 deform_pec_aim_r {parent=deform_pec_aim_base_r, loc=<0 0 0>, rot=<0 0 0 1>}
 deform_pec_base_r {parent=deform_pec_aim_r, loc=<-9.3667 -0.0003 -0.0004>, rot=<0 0 0 1>}
 deform_pec_r {parent=deform_pec_base_r, loc=<0 0 0>, rot=<0 0 0 1>}
 thigh_r {parent=pelvis, loc=<-1.0855 -0.2109 8.7053>, rot=<0.0355 -0.0009 0.9993 0.0155>}
 thigh_twist_01_r {parent=thigh_r, loc=<27.952 0 0>, rot=<0.0013 0 0 -1>}
 calf_r {parent=thigh_r, loc=<42.5284 0.0019 0.7297>, rot=<0.0006 -0.0161 -0.0391 -0.9991>}
 foot_r {parent=calf_r, loc=<46.7401 2.5014 0.592>, rot=<0.0427 -0.0238 0.0353 -0.9982>}
 ball_r {parent=foot_r, loc=<3.5658 -10.8066 0.1544>, rot=<0.0001 -0.0001 -0.7063 -0.7079>}
 calf_twist_02_r {parent=calf_r, loc=<19.1546 1.7943 0.2638>, rot=<0.0237 0.0095 -0.0035 -0.9997>}
 calf_twist_01_r {parent=calf_r, loc=<35.8483 1.9197 0.5794>, rot=<0.0237 0.0095 -0.0035 -0.9997>}
 deform_knee_in_base_r {parent=calf_r, loc=<-0.8031 5.0231 -0.2457>, rot=<0 0 0 -1>}
 deform_knee_in_r {parent=deform_knee_in_base_r, loc=<0 0 0>, rot=<0 0 0 1>}
 deform_knee_out_base_r {parent=calf_r, loc=<0 -4 0>, rot=<0 0 0 -1>}
 deform_knee_out_r {parent=deform_knee_out_base_r, loc=<0 0 0>, rot=<0 0 0 1>}
 deform_groin_base_r {parent=thigh_r, loc=<-0.4801 -5.6776 0.2486>, rot=<0 0 0 -1>}
 deform_groin_r {parent=deform_groin_base_r, loc=<0 0 0>, rot=<0 0 0 1>}
 thigh_l {parent=pelvis, loc=<-1.0855 -0.2109 -8.7053>, rot=<0.0009 0.0355 0.0155 -0.9993>}
 calf_l {parent=thigh_l, loc=<-42.5284 -0.0019 -0.7297>, rot=<0.0006 -0.0161 -0.0391 -0.9991>}
 foot_l {parent=calf_l, loc=<-46.7401 -2.5014 -0.592>, rot=<0.0427 -0.0238 0.0353 -0.9982>}
 ball_l {parent=foot_l, loc=<-3.5658 10.8066 -0.1544>, rot=<0.0001 -0.0001 -0.7063 -0.7079>}
 calf_twist_02_l {parent=calf_l, loc=<-19.1546 -1.7943 -0.2638>, rot=<0.0237 0.0095 -0.0035 -0.9997>}
 calf_twist_01_l {parent=calf_l, loc=<-35.8483 -1.9197 -0.5794>, rot=<0.0237 0.0095 -0.0035 -0.9997>}
 deform_knee_in_base_l {parent=calf_l, loc=<0.8031 -5.0231 0.2456>, rot=<0 0 0 -1>}
 deform_knee_in_l {parent=deform_knee_in_base_l, loc=<0 0 0>, rot=<0 0 0 1>}
 deform_knee_out_base_l {parent=calf_l, loc=<0 4 0>, rot=<0 0 0 -1>}
 deform_knee_out_l {parent=deform_knee_out_base_l, loc=<0 0 0>, rot=<0 0 0 1>}
 thigh_twist_01_l {parent=thigh_l, loc=<-27.952 0 0>, rot=<0.0013 0 0 -1>}
 deform_groin_base_l {parent=thigh_l, loc=<0.4801 5.6776 -0.2486>, rot=<0 0 0 -1>}
 deform_groin_l {parent=deform_groin_base_l, loc=<0 0 0>, rot=<0 0 0 1>}
 deform_glute_base_l {parent=pelvis, loc=<-7.3466 -9.0213 -9.1245>, rot=<0.0172 0.031 -0.4477 -0.8935>}
 deform_glute_l {parent=deform_glute_base_l, loc=<0 0 0>, rot=<0 0 0 1>}
 deform_glute_base_r {parent=pelvis, loc=<-7.3464 -9.0213 9.1245>, rot=<0.031 -0.0172 0.8935 -0.4477>}
 deform_glute_r {parent=deform_glute_base_r, loc=<0 0 0>, rot=<0 0 0 1>}
 pelvis_null {parent=root, type=Null, loc=<0 -0.5301 98.0571>, rot=<0.0127 0.707 -0.0127 -0.707>}
 dyn_ribbon_l_01_null {type=Null, loc=<-4.2398 -9.8445 154.0644>, rot=<0.1641 -0.7521 0.269 -0.5788>}
 NewControl {controlType=EULER_TRANSFORM, parent=dyn_ribbon_l_01_null, type=Control, loc=<0 0 0>, rot=<0 0 0 1>}
 dyn_ribbon_l_02_null {type=Null, loc=<-6.6777 -14.8233 147.0844>, rot=<0.0306 0.692 -0.153 0.7048>}
 NewControl_2 {controlType=EULER_TRANSFORM, parent=dyn_ribbon_l_02_null, type=Control, loc=<0 0 0>, rot=<0 0 0 1>}
 dyn_ribbon_l_03_null {type=Null, loc=<-6.7245 -16.5911 137.0428>, rot=<-0.1383 -0.6914 0.044 -0.7078>}
 NewControl_3 {controlType=EULER_TRANSFORM, parent=dyn_ribbon_l_03_null, type=Control, loc=<0 0 0>, rot=<0 0 0 1>}
 dyn_ribbon_l_04_null {type=Null, loc=<-6.3119 -15.2668 126.8682>, rot=<0.1323 0.7142 -0.053 0.6853>}
 NewControl_4 {controlType=EULER_TRANSFORM, parent=dyn_ribbon_l_04_null, type=Control, loc=<0 0 0>, rot=<0 0 0 1>}
 dyn_ribbon_l_05_null {type=Null, loc=<-6.577 -14.0717 116.6784>, rot=<-0.0886 -0.7457 0.0959 -0.6534>}
 NewControl_5 {controlType=EULER_TRANSFORM, parent=dyn_ribbon_l_05_null, type=Control, loc=<0 0 0>, rot=<0 0 0 1>}
 dyn_ribbon_l_06_null {type=Null, loc=<-7.917 -14.0011 106.4933>, rot=<0.0336 0.7646 -0.1403 0.6281>}
 NewControl_6 {controlType=EULER_TRANSFORM, parent=dyn_ribbon_l_06_null, type=Control, loc=<0 0 0>, rot=<0 0 0 1>}
 dyn_ribbon_l_07_null {type=Null, loc=<-10.0595 -15.2841 96.536>, rot=<0.0159 -0.7667 0.1765 -0.6171>}
 NewControl_7 {controlType=EULER_TRANSFORM, parent=dyn_ribbon_l_07_null, type=Control, loc=<0 0 0>, rot=<0 0 0 1>}
 dyn_ribbon_l_08_null {type=Null, loc=<-12.5024 -17.772 86.8796>, rot=<-0.0159 0.7667 -0.1765 0.6171>}
 NewControl_16 {controlType=EULER_TRANSFORM, parent=dyn_ribbon_l_08_null, type=Control, loc=<0 0 0>, rot=<0 0 0 1>}
 dyn_ribbon_r_01_null {type=Null, loc=<-6.7688 -8.6686 154.3233>, rot=<0.0705 -0.7962 0.3093 -0.5153>}
 dyn_ribbon_r_01_ctrl {controlType=EULER_TRANSFORM, parent=dyn_ribbon_r_01_null, type=Control, loc=<113.055 95.5245 45.0584>, rot=<-0.0705 0.7962 -0.3093 -0.5153>}
 dyn_ribbon_r_02_null {type=Null, loc=<-10.2186 -11.9075 148.4839>, rot=<-0.0704 -0.7525 0.2443 -0.6075>}
 dyn_ribbon_r_02_ctrl {controlType=EULER_TRANSFORM, parent=dyn_ribbon_r_02_null, type=Control, loc=<-7.0274 2.562 0.736>, rot=<0.0409 -0.0569 0.1711 -0.9828>}
 dyn_ribbon_r_03_null {type=Null, loc=<-13.0764 -14.0741 137.7197>, rot=<-0.1761 -0.6885 0.1685 -0.6831>}
 dyn_ribbon_r_03_ctrl {controlType=EULER_TRANSFORM, parent=dyn_ribbon_r_03_null, type=Control, loc=<-10.7504 3.3523 1.3863>, rot=<0.0175 -0.0646 0.1486 -0.9866>}
 dyn_ribbon_r_04_null {type=Null, loc=<-13.1308 -13.9345 126.3454>, rot=<-0.1949 -0.6738 0.1474 -0.6973>}
 dyn_ribbon_r_04_ctrl {controlType=EULER_TRANSFORM, parent=dyn_ribbon_r_04_null, type=Control, loc=<-11.3476 0.7353 0.2951>, rot=<-0.0018 -0.0129 0.0324 -0.9994>}
 dyn_ribbon_r_05_null {type=Null, loc=<-12.5781 -13.2833 114.98>, rot=<-0.1765 -0.7075 0.17 -0.6628>}
 dyn_ribbon_r_05_ctrl {controlType=EULER_TRANSFORM, parent=dyn_ribbon_r_05_null, type=Control, loc=<-11.3256 -0.9102 -0.8979>, rot=<0.0042 0.0396 -0.0398 -0.9984>}
 dyn_ribbon_r_06_null {type=Null, loc=<-13.2502 -13.0052 103.6095>, rot=<-0.1253 -0.7402 0.2155 -0.6245>}
 dyn_ribbon_r_06_ctrl {controlType=EULER_TRANSFORM, parent=dyn_ribbon_r_06_null, type=Control, loc=<-11.2294 -1.7853 -0.7289>, rot=<-0.0005 0.0321 -0.0786 -0.9964>}
 dyn_ribbon_r_07_null {type=Null, loc=<-15.3967 -13.9564 92.4757>, rot=<-0.0776 -0.7521 0.2514 -0.6042>}
 dyn_ribbon_r_07_ctrl {controlType=EULER_TRANSFORM, parent=dyn_ribbon_r_07_null, type=Control, loc=<-11.2851 -1.4449 -0.1782>, rot=<-0.0033 0.0076 -0.0636 -0.9979>}
 dyn_ribbon_r_08_null {type=Null, loc=<-18.3319 -16.0873 81.6822>, rot=<-0.017 -0.7728 0.2873 -0.5656>}
 dyn_ribbon_r_08_ctrl {controlType=EULER_TRANSFORM, parent=dyn_ribbon_r_08_null, type=Control, loc=<-11.2337 -1.7759 -0.5553>, rot=<-0.0118 0.0235 -0.0785 -0.9966>}
 dyn_ribbon_r_09_null {type=Null, loc=<-22.4229 -19.4856 71.6241>, rot=<0.017 0.7728 -0.2873 0.5656>}
 dyn_ribbon_r_09_ctrl {controlType=EULER_TRANSFORM, parent=dyn_ribbon_r_09_null, type=Control, loc=<-11.3776 0 0>, rot=<0 0 0 1>}
]
variables [
 CrouchWalking {type=float, default=0}
 HeadLookingUp {type=float, default=0}
 HeadLookingUp_Crouching {type=double, default=0}
 HeadLookingUp_Extreme {type=double, default=0}
 LookingDown {type=float, default=0}
 LookingLeft {type=double, default=0}
 LookingRight {type=double, default=0}
 LookingStraightAhead {type=float, default=0}
 Moving2D {type=bool, default=False}
 RigidBodiesEnabled {type=bool, default=False}
 animDynamicsEnabled {type=bool, default=False}
 isCrouching {type=float, default=0}
 isFrontEnd {type=bool, default=False}
 isJustCrouching {type=float, default=0}
 isMeleeSwinging {type=bool, default=False}
 isMoving2D {type=bool, default=False}
 isParachuteOpened {type=bool, default=False}
 isPlayingEmote {type=bool, default=False}
 isSkyFloating {type=bool, default=False}
 isTacticalJumping {type=bool, default=False}
 isTargeting {type=bool, default=False}
 isWearingBackBling {type=bool, default=False}
 isWearingCape {type=bool, default=False}
]
