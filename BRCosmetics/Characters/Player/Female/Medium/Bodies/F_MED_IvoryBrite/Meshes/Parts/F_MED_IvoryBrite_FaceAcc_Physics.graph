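# Physics asset for the IvoryBrite female face accessory mesh, covering a ponytail and a bandana knot with two ribbon tails.
# Capsules on the pelvis, spine, neck, head, clavicles and upper arms form the body volume around the accessory chains.
# The ponytail and knot chains hang off the head via base bones, with the left and right ribbons branching from the knot.
# Ribbon segments have narrow, asymmetric swing and tight twist limits, while other joints rotate freely.
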PhysicsAsset /BRCosmetics/Characters/Player/Female/Medium/Bodies/F_MED_IvoryBrite/Meshes/Parts/F_MED_IvoryBrite_FaceAcc_Physics
bodies {
 base_BandanaKnot [{shapeType=Capsule, length=0.505, radius=0.505, shapeName=base_BandanaKnot_capsule, rotation=<rot 0 -90 0>}]
 base_ponytail_01 [{shapeType=Capsule, length=0.505, radius=0.505, shapeName=base_ponytail_01_capsule, rotation=<rot 0 -90 0>}]
 base_ribbon_l_01 [{shapeType=Capsule, length=0.505, radius=0.505, shapeName=base_ribbon_l_01_capsule, rotation=<rot 0 -90 0>}]
 base_ribbon_r_01 [{shapeType=Capsule, length=0.505, radius=0.505, shapeName=base_ribbon_r_01_capsule, rotation=<rot 0 -90 0>}]
 clavicle_l [{
   shapeType=Capsule
   length=31.4892
   radius=5.9214
   shapeName=clavicle_l_capsule
   center=<12.0135 -1.0226 -0.4793>
   rotation=<rot -0.001 -89.9999 92.2845>
  }]
 clavicle_r [{
   shapeType=Capsule
   length=31.4892
   radius=5.9214
   shapeName=clavicle_r_capsule
   center=<-12.0135 1.0226 0.4793>
   rotation=<rot -0.001 -89.9999 -87.7155>
  }]
 dyn_BandanaKnot [{shapeType=Capsule, length=0.505, radius=1.3847, shapeName=base_BandanaKnot_capsule, center=<1.2427 0 0>, rotation=<rot 0 -90 0>}]
 dyn_ponytail_01 [{
   shapeType=Capsule
   length=6.0634
   radius=2.304
   shapeName=dyn_ponytail_01_capsule
   center=<3.9915 0.6373 0.5838>
   rotation=<rot -2.2022 -90 -0.7839>
  }]
 dyn_ponytail_02 [{
   shapeType=Capsule
   length=5.1539
   radius=3.1038
   shapeName=dyn_ponytail_02_capsule
   center=<5.3617 -0.1209 -0.8458>
   rotation=<rot 0 -93.2006 -82.9499>
  }]
 dyn_ponytail_03 [{
   shapeType=Capsule
   length=2.866
   radius=4.6665
   shapeName=dyn_ponytail_02_capsule
   center=<3.5322 0.1468 -0.1521>
   rotation=<rot 0 -93.2006 -96.1105>
  }]
 dyn_ponytail_04 [{
   shapeType=Capsule
   length=4.621
   radius=2.9115
   shapeName=dyn_ponytail_02_capsule
   center=<3.4359 0.233 0.1413>
   rotation=<rot 0.5694 -94.7644 -69.9992>
  }]
 dyn_ribbon_l_01 [{
   shapeType=Capsule
   length=9.1937
   radius=1.1715
   shapeName=dyn_ribbon_r_01_capsule
   center=<4.6756 0.0857 -0.0317>
   rotation=<rot 0.1242 88.9975 -88.7765>
  }]
 dyn_ribbon_l_02 [{
   shapeType=Capsule
   length=12.2082
   radius=1.1715
   shapeName=dyn_ribbon_r_01_capsule
   center=<5.7046 0.0676 -0.0098>
   rotation=<rot 0.1242 88.9975 -88.7765>
  }]
 dyn_ribbon_l_03 [{
   shapeType=Capsule
   length=12.2082
   radius=1.1715
   shapeName=dyn_ribbon_r_01_capsule
   center=<5.7046 0.0676 -0.0098>
   rotation=<rot 0.1242 88.9975 -88.7765>
  }]
 dyn_ribbon_l_04 [{
   shapeType=Capsule
   length=12.2082
   radius=1.1715
   shapeName=dyn_ribbon_r_01_capsule
   center=<5.7046 0.0676 -0.0098>
   rotation=<rot 0.1242 88.9975 -88.7765>
  }]
 dyn_ribbon_l_05 [{
   shapeType=Capsule
   length=12.2082
   radius=1.1715
   shapeName=dyn_ribbon_r_01_capsule
   center=<5.7046 0.0676 -0.0098>
   rotation=<rot 0.1242 88.9975 -88.7765>
  }]
 dyn_ribbon_l_06 [{
   shapeType=Capsule
   length=12.2082
   radius=1.1715
   shapeName=dyn_ribbon_r_01_capsule
   center=<5.7046 0.0676 -0.0098>
   rotation=<rot 0.1242 88.9975 -88.7765>
  }]
 dyn_ribbon_l_07 [{
   shapeType=Capsule
   length=12.2082
   radius=1.1715
   shapeName=dyn_ribbon_r_01_capsule
   center=<5.7046 0.0676 -0.0098>
   rotation=<rot 0.1242 88.9975 -88.7765>
  }]
 dyn_ribbon_r_01 [{
   shapeType=Capsule
   length=6.4039
   radius=1.0115
   shapeName=dyn_ribbon_r_01_capsule
   center=<4.0259 -0.0178 -0.0705>
   rotation=<rot 0 92.9769 -87.978>
  }]
 dyn_ribbon_r_02 [{
   shapeType=Capsule
   length=12.2082
   radius=1.1715
   shapeName=dyn_ribbon_r_01_capsule
   center=<5.7046 0.0676 -0.0098>
   rotation=<rot 0.1242 88.9975 -88.7765>
  }]
 dyn_ribbon_r_03 [{
   shapeType=Capsule
   length=12.2082
   radius=1.1715
   shapeName=dyn_ribbon_r_01_capsule
   center=<5.7046 0.0676 -0.0098>
   rotation=<rot 0.1242 88.9975 -88.7765>
  }]
 dyn_ribbon_r_04 [{
   shapeType=Capsule
   length=12.2082
   radius=1.1715
   shapeName=dyn_ribbon_r_01_capsule
   center=<5.7046 0.0676 -0.0098>
   rotation=<rot 0.1242 88.9975 -88.7765>
  }]
 dyn_ribbon_r_05 [{
   shapeType=Capsule
   length=12.2082
   radius=1.1715
   shapeName=dyn_ribbon_r_01_capsule
   center=<5.7046 0.0676 -0.0098>
   rotation=<rot 0.1242 88.9975 -88.7765>
  }]
 dyn_ribbon_r_06 [{
   shapeType=Capsule
   length=12.2082
   radius=1.1715
   shapeName=dyn_ribbon_r_01_capsule
   center=<5.7046 0.0676 -0.0098>
   rotation=<rot 0.1242 88.9975 -88.7765>
  }]
 dyn_ribbon_r_07 [{
   shapeType=Capsule
   length=12.2082
   radius=1.1715
   shapeName=dyn_ribbon_r_01_capsule
   center=<5.7046 0.0676 -0.0098>
   rotation=<rot 0.1242 88.9975 -88.7765>
  }]
 dyn_ribbon_r_08 [{
   shapeType=Capsule
   length=12.2082
   radius=1.1715
   shapeName=dyn_ribbon_r_01_capsule
   center=<5.7046 0.0676 -0.0098>
   rotation=<rot 0.1242 88.9975 -88.7765>
  }]
 head [{shapeType=Capsule, length=2.6294, radius=8.9034, shapeName=head_capsule, center=<5.799 0.7399 0.2001>, rotation=<rot 0.6273 87.3617 -90>}]
 neck_01 [{shapeType=Capsule, length=1.8774, radius=4.5368, shapeName=neck_02_capsule, center=<0 0.5845 0>, rotation=<rot 0 -90 90>}]
 neck_02 [{shapeType=Capsule, length=2.0056, radius=4.4087, shapeName=neck_02_capsule, rotation=<rot 0 -90 90>}]
 pelvis [{shapeType=Capsule, length=17.4337, radius=12.8495, shapeName=pelvis_capsule, center=<-3.637 0 0>, rotation=<rot 0 -90 0>}]
 spine_01 [{shapeType=Capsule, length=20.2358, radius=9.1189, shapeName=spine_05_capsule, center=<0 -0.1013 0>, rotation=<rot 0 -90 0>}]
 spine_02 [{shapeType=Capsule, length=17.0213, radius=9.1189, shapeName=spine_05_capsule, center=<0 -0.1013 0>, rotation=<rot 0 -90 0>}]
 spine_03 [{shapeType=Capsule, length=14.5436, radius=9.1189, shapeName=spine_05_capsule, center=<0 -0.1013 0>, rotation=<rot 0 -90 0>}]
 spine_04 [{shapeType=Capsule, length=11.8443, radius=9.1189, shapeName=spine_05_capsule, center=<0 -0.1013 0>, rotation=<rot 0 -90 0>}]
 spine_05 [
  {shapeType=Capsule, length=11.0208, radius=9.2701, shapeName=spine_05_capsule, center=<0 -0.4249 0>, rotation=<rot 0 -90 0>}
  {shapeType=Capsule, length=15.1282, radius=5.9918, shapeName=spine_05_capsule, center=<6.2142 -1.5817 0>, rotation=<rot 0 -90 0>}
 ]
 upperarm_l [{
   shapeType=Capsule
   length=25.1397
   radius=5.0134
   shapeName=upperarm_l_capsule
   center=<12.8925 -0.2098 0.2078>
   rotation=<rot -0.0001 -89.9999 90.8257>
  }]
 upperarm_r [{
   shapeType=Capsule
   length=24.7931
   radius=5.36
   shapeName=upperarm_r_capsule
   center=<-12.8878 0.2098 0.1149>
   rotation=<rot -0.0001 -89.9999 -89.1743>
  }]
}
constraints [
 {
  bone1Name=spine_02
  bone2Name=spine_01
  swing1LimitDegrees=45
  swing1Motion=Free
  swing2LimitDegrees=45
  swing2Motion=Free
  twistLimitDegrees=45
  twistMotion=Free
 }
 {
  bone1Name=spine_03
  bone2Name=spine_02
  swing1LimitDegrees=45
  swing1Motion=Free
  swing2LimitDegrees=45
  swing2Motion=Free
  twistLimitDegrees=45
  twistMotion=Free
 }
 {
  bone1Name=spine_04
  bone2Name=spine_03
  swing1LimitDegrees=45
  swing1Motion=Free
  swing2LimitDegrees=45
  swing2Motion=Free
  twistLimitDegrees=45
  twistMotion=Free
 }
 {
  bone1Name=spine_05
  bone2Name=spine_04
  swing1LimitDegrees=45
  swing1Motion=Free
  swing2LimitDegrees=45
  swing2Motion=Free
  twistLimitDegrees=45
  twistMotion=Free
 }
 {
  bone1Name=clavicle_l
  bone2Name=spine_05
  swing1LimitDegrees=45
  swing1Motion=Free
  swing2LimitDegrees=45
  swing2Motion=Free
  twistLimitDegrees=45
  twistMotion=Free
 }
 {
  bone1Name=clavicle_r
  bone2Name=spine_05
  swing1LimitDegrees=45
  swing1Motion=Free
  swing2LimitDegrees=45
  swing2Motion=Free
  twistLimitDegrees=45
  twistMotion=Free
 }
 {
  bone1Name=upperarm_r
  bone2Name=clavicle_r
  swing1LimitDegrees=45
  swing1Motion=Free
  swing2LimitDegrees=45
  swing2Motion=Free
  twistLimitDegrees=45
  twistMotion=Free
 }
 {
  bone1Name=upperarm_l
  bone2Name=clavicle_l
  swing1LimitDegrees=45
  swing1Motion=Free
  swing2LimitDegrees=45
  swing2Motion=Free
  twistLimitDegrees=45
  twistMotion=Free
 }
 {bone1Name=dyn_ribbon_l_01, bone2Name=base_ribbon_l_01, swing1LimitDegrees=30, swing2LimitDegrees=8, twistLimitDegrees=2}
 {bone1Name=dyn_ribbon_l_02, bone2Name=dyn_ribbon_l_01, swing1LimitDegrees=30, swing2LimitDegrees=8, twistLimitDegrees=2}
 {bone1Name=dyn_ribbon_l_03, bone2Name=dyn_ribbon_l_02, swing1LimitDegrees=30, swing2LimitDegrees=8, twistLimitDegrees=2}
 {bone1Name=dyn_ribbon_l_04, bone2Name=dyn_ribbon_l_03, swing1LimitDegrees=30, swing2LimitDegrees=8, twistLimitDegrees=2}
 {bone1Name=dyn_ribbon_l_05, bone2Name=dyn_ribbon_l_04, swing1LimitDegrees=30, swing2LimitDegrees=8, twistLimitDegrees=2}
 {bone1Name=dyn_ribbon_l_06, bone2Name=dyn_ribbon_l_05, swing1LimitDegrees=30, swing2LimitDegrees=8, twistLimitDegrees=2}
 {bone1Name=dyn_ribbon_l_07, bone2Name=dyn_ribbon_l_06, swing1LimitDegrees=30, swing2LimitDegrees=8, twistLimitDegrees=2}
 {bone1Name=dyn_ribbon_r_01, bone2Name=base_ribbon_r_01, swing1LimitDegrees=30, swing2LimitDegrees=8, twistLimitDegrees=2}
 {bone1Name=dyn_ribbon_r_02, bone2Name=dyn_ribbon_r_01, swing1LimitDegrees=30, swing2LimitDegrees=8, twistLimitDegrees=2}
 {bone1Name=dyn_ribbon_r_03, bone2Name=dyn_ribbon_r_02, swing1LimitDegrees=30, swing2LimitDegrees=8, twistLimitDegrees=2}
 {bone1Name=dyn_ribbon_r_04, bone2Name=dyn_ribbon_r_03, swing1LimitDegrees=30, swing2LimitDegrees=8, twistLimitDegrees=2}
 {bone1Name=dyn_ribbon_r_05, bone2Name=dyn_ribbon_r_04, swing1LimitDegrees=30, swing2LimitDegrees=8, twistLimitDegrees=2}
 {bone1Name=dyn_ribbon_r_06, bone2Name=dyn_ribbon_r_05, swing1LimitDegrees=30, swing2LimitDegrees=8, twistLimitDegrees=2}
 {bone1Name=dyn_ribbon_r_07, bone2Name=dyn_ribbon_r_06, swing1LimitDegrees=30, swing2LimitDegrees=8, twistLimitDegrees=2}
 {bone1Name=dyn_ribbon_r_08, bone2Name=dyn_ribbon_r_07, swing1LimitDegrees=30, swing2LimitDegrees=8, twistLimitDegrees=2}
 {
  bone1Name=spine_01
  bone2Name=pelvis
  swing1LimitDegrees=45
  swing1Motion=Free
  swing2LimitDegrees=45
  swing2Motion=Free
  twistLimitDegrees=45
  twistMotion=Free
 }
 {
  bone1Name=base_ribbon_l_01
  bone2Name=dyn_BandanaKnot
  swing1LimitDegrees=45
  swing1Motion=Free
  swing2LimitDegrees=45
  swing2Motion=Free
  twistLimitDegrees=45
  twistMotion=Free
 }
 {
  bone1Name=base_ribbon_r_01
  bone2Name=dyn_BandanaKnot
  swing1LimitDegrees=45
  swing1Motion=Free
  swing2LimitDegrees=45
  swing2Motion=Free
  twistLimitDegrees=45
  twistMotion=Free
 }
 {
  bone1Name=dyn_BandanaKnot
  bone2Name=base_BandanaKnot
  swing1LimitDegrees=45
  swing1Motion=Free
  swing2LimitDegrees=45
  swing2Motion=Free
  twistLimitDegrees=45
  twistMotion=Free
 }
 {
  bone1Name=base_BandanaKnot
  bone2Name=head
  swing1LimitDegrees=45
  swing1Motion=Free
  swing2LimitDegrees=45
  swing2Motion=Free
  twistLimitDegrees=45
  twistMotion=Free
 }
 {
  bone1Name=head
  bone2Name=neck_02
  swing1LimitDegrees=45
  swing1Motion=Free
  swing2LimitDegrees=45
  swing2Motion=Free
  twistLimitDegrees=45
  twistMotion=Free
 }
 {
  bone1Name=neck_02
  bone2Name=neck_01
  swing1LimitDegrees=45
  swing1Motion=Free
  swing2LimitDegrees=45
  swing2Motion=Free
  twistLimitDegrees=45
  twistMotion=Free
 }
 {
  bone1Name=neck_01
  bone2Name=spine_05
  swing1LimitDegrees=45
  swing1Motion=Free
  swing2LimitDegrees=45
  swing2Motion=Free
  twistLimitDegrees=45
  twistMotion=Free
 }
 {
  bone1Name=base_ponytail_01
  bone2Name=head
  swing1LimitDegrees=45
  swing1Motion=Free
  swing2LimitDegrees=45
  swing2Motion=Free
  twistLimitDegrees=45
  twistMotion=Free
 }
 {
  bone1Name=dyn_ponytail_01
  bone2Name=base_ponytail_01
  swing1LimitDegrees=45
  swing1Motion=Free
  swing2LimitDegrees=45
  swing2Motion=Free
  twistLimitDegrees=45
  twistMotion=Free
 }
 {
  bone1Name=dyn_ponytail_02
  bone2Name=dyn_ponytail_01
  swing1LimitDegrees=45
  swing1Motion=Free
  swing2LimitDegrees=45
  swing2Motion=Free
  twistLimitDegrees=45
  twistMotion=Free
 }
 {
  bone1Name=dyn_ponytail_03
  bone2Name=dyn_ponytail_02
  swing1LimitDegrees=45
  swing1Motion=Free
  swing2LimitDegrees=45
  swing2Motion=Free
  twistLimitDegrees=45
  twistMotion=Free
 }
 {
  bone1Name=dyn_ponytail_04
  bone2Name=dyn_ponytail_03
  swing1LimitDegrees=45
  swing1Motion=Free
  swing2LimitDegrees=45
  swing2Motion=Free
  twistLimitDegrees=45
  twistMotion=Free
 }
]
