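# Anim Blueprint for IvoryBrite's face accessory mesh, adding secondary motion to ponytail, front hair and ribbons.
# Copied parent and body poses pass through a Control Rig and noise-driven ribbon-base rotations into cached dynamics stages.
# Bool blends switch the ponytail between AnimDynamics and a fixed adjustment, and ribbons between AnimDynamics, RigidBody variants or none, before front-hair dynamics.
# The EventGraph updates gravity, forces and rotation values from movement states.
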AnimBlueprint /BRCosmetics/Characters/Player/Female/Medium/Bodies/F_MED_IvoryBrite/Meshes/Parts/F_MED_IvoryBrite_FaceAcc_AnimBP
target_skeleton=&/F_MED_IvoryBrite_FaceAcc_Skeleton
graphs [
 AnimGraph {nodes [
   n1 = Anim Dynamics - Bone: dyn_frontHair_c!(ComponentPose, Alpha=1, GravityOverride:Vector="0,0,-980") -> (Pose) {
    angularDampingOverride=1
    angularSpringConstant=5000
    bAngularSpring=true
    bOverrideAngularDamping=true
    bUseGravityOverride=true
    gravityScale=0.35
    simulationSpace=RootRelative
    alphaBoolBlend {blendOption=Linear}
    boundBone {boneName=dyn_frontHair_c}
    componentAppliedLinearAccClamp=<10000 10000 10000>
    componentLinearAccScale=<1 1 1>
    componentLinearVelScale=<1 1 1>
    gravityOverride {z=-980}
    physicsBodyDefinitions [{
      boundBone {boneName=dyn_frontHair_c}
      boxExtents=<20 15 15>
      constraintSetup {angularLimitsMax=<0 0 15>, angularLimitsMin=<0 -20 -15>, angularTarget=<1 0 0>}
      localJointOffset=<1 0 0>
     }]
    simSpaceSettings {maxAngularAcceleration=30, maxAngularVelocity=6, simSpaceAngularAlpha=1}
   }
   n2 = Anim Dynamics - Bone: dyn_frontHair_f_01!(ComponentPose=@n1, Alpha=1, GravityOverride:Vector="0,0,-980") -> (Pose) {
    angularDampingOverride=1
    angularSpringConstant=5000
    bAngularSpring=true
    bOverrideAngularDamping=true
    bUseGravityOverride=true
    gravityScale=0.35
    simulationSpace=RootRelative
    alphaBoolBlend {blendOption=Linear}
    boundBone {boneName=dyn_frontHair_f_01}
    componentAppliedLinearAccClamp=<10000 10000 10000>
    componentLinearAccScale=<1 1 1>
    componentLinearVelScale=<1 1 1>
    gravityOverride {z=-980}
    physicsBodyDefinitions [{
      boundBone {boneName=dyn_frontHair_f_01}
      boxExtents=<20 15 15>
      constraintSetup {angularLimitsMax=<0 50 20>, angularLimitsMin=<0 -30 -20>, angularTarget=<1 0 0>}
      localJointOffset=<1 0 0>
     }]
    simSpaceSettings {maxAngularAcceleration=30, maxAngularVelocity=6, simSpaceAngularAlpha=1}
   }
   n3 = Anim Dynamics - Bone: dyn_frontHair_f_02!(ComponentPose=@n2, Alpha=1, GravityOverride:Vector="0,0,-980") -> (Pose) {
    angularDampingOverride=1
    angularSpringConstant=5000
    bAngularSpring=true
    bOverrideAngularDamping=true
    bUseGravityOverride=true
    gravityScale=0.35
    simulationSpace=RootRelative
    alphaBoolBlend {blendOption=Linear}
    boundBone {boneName=dyn_frontHair_f_02}
    componentAppliedLinearAccClamp=<10000 10000 10000>
    componentLinearAccScale=<1 1 1>
    componentLinearVelScale=<1 1 1>
    gravityOverride {z=-980}
    physicsBodyDefinitions [{
      boundBone {boneName=dyn_frontHair_f_02}
      boxExtents=<20 15 15>
      constraintSetup {angularLimitsMax=<0 50 25>, angularLimitsMin=<0 -40 -25>, angularTarget=<1 0 0>}
      localJointOffset=<1 0 0>
     }]
    simSpaceSettings {maxAngularAcceleration=30, maxAngularVelocity=6, simSpaceAngularAlpha=1}
   }
   n4 = Anim Dynamics - Bone: dyn_frontHair_b_01!(ComponentPose=@n3, Alpha=1, GravityOverride:Vector="0,0,-980") -> (Pose) {
    angularDampingOverride=1
    angularSpringConstant=5000
    bAngularSpring=true
    bOverrideAngularDamping=true
    bUseGravityOverride=true
    gravityScale=0.35
    simulationSpace=RootRelative
    alphaBoolBlend {blendOption=Linear}
    boundBone {boneName=dyn_frontHair_b_01}
    componentAppliedLinearAccClamp=<10000 10000 10000>
    componentLinearAccScale=<1 1 1>
    componentLinearVelScale=<1 1 1>
    gravityOverride {z=-980}
    physicsBodyDefinitions [{
      boundBone {boneName=dyn_frontHair_b_01}
      boxExtents=<20 15 15>
      constraintSetup {angularLimitsMax=<0 50 20>, angularLimitsMin=<0 -30 -20>, angularTarget=<1 0 0>}
      localJointOffset=<1 0 0>
     }]
    simSpaceSettings {maxAngularAcceleration=30, maxAngularVelocity=6, simSpaceAngularAlpha=1}
   }
   Anim Dynamics - Bone: dyn_frontHair_b_02!(ComponentPose=@n4, Alpha=1, GravityOverride:Vector="0,0,-980") -> (Pose) {
    angularDampingOverride=1
    angularSpringConstant=5000
    bAngularSpring=true
    bOverrideAngularDamping=true
    bUseGravityOverride=true
    gravityScale=0.35
    simulationSpace=RootRelative
    alphaBoolBlend {blendOption=Linear}
    boundBone {boneName=dyn_frontHair_b_02}
    componentAppliedLinearAccClamp=<10000 10000 10000>
    componentLinearAccScale=<1 1 1>
    componentLinearVelScale=<1 1 1>
    gravityOverride {z=-980}
    physicsBodyDefinitions [{
      boundBone {boneName=dyn_frontHair_b_02}
      boxExtents=<20 15 15>
      constraintSetup {angularLimitsMax=<0 50 25>, angularLimitsMin=<0 -40 -25>, angularTarget=<1 0 0>}
      localJointOffset=<1 0 0>
     }]
    simSpaceSettings {maxAngularAcceleration=30, maxAngularVelocity=6, simSpaceAngularAlpha=1}
   }
   RigidBody!(ComponentPose, Alpha, OverrideWorldGravity:Vector="0,0,0", ExternalForce) -> (Pose) {
    bOverrideWorldGravity=true
    overridePhysicsAsset=&/F_MED_IvoryBrite_FaceAcc_Physics
    simulationSpace=BaseBoneSpace
    alphaBoolBlend {blendOption=Linear}
    baseBoneRef {boneName=neck_01}
    componentAppliedLinearAccClamp=<3000 3000 3000>
    overrideWorldGravity {z=-980}
    simSpaceSettings {
     dampingAlpha=0
     maxAngularAcceleration=500
     maxAngularVelocity=3
     maxLinearAcceleration=250
     maxLinearVelocity=200
     velocityScaleZ=0.1
     worldAlpha=1
    }
   }
   Output Pose(Result=Use cached pose 'PostDynamics_Cache'()) {layerGroup=None}
   n5 = Anim Dynamics - Bone: dyn_ponytail_01!(ComponentPose=Local To Component(LocalPose=Use cached pose 'PreDynamicsCache'()), Alpha=1, GravityOverride:Vector="0,0,-980") -> (Pose) {
    angularDampingOverride=1
    angularSpringConstant=7500
    bAngularSpring=true
    bOverrideAngularDamping=true
    bUseGravityOverride=true
    bUseSphericalLimits=true
    gravityScale=0.35
    simulationSpace=RootRelative
    alphaBoolBlend {blendOption=Linear}
    boundBone {boneName=dyn_ponytail_01}
    componentAppliedLinearAccClamp=<10000 10000 10000>
    componentLinearAccScale=<1 1 1>
    componentLinearVelScale=<1 1 1>
    gravityOverride {z=-980}
    physicsBodyDefinitions [{
      boundBone {boneName=dyn_ponytail_01}
      boxExtents=<20 25 25>
      constraintSetup {angularLimitsMax=<0 10 0>, angularLimitsMin=<0 -10 -25>, angularTarget=<1 0 0>}
      localJointOffset=<1 0 0>
     }]
    simSpaceSettings {maxAngularAcceleration=30, maxAngularVelocity=6, simSpaceAngularAlpha=1}
    sphericalLimits [{limitRadius=11, drivingBone {boneName=spine_05}, sphereLocalOffset=<3 1.25 0>}]
   }
   n6 = Anim Dynamics - Bone: dyn_ponytail_02!(ComponentPose=@n5, Alpha=1, GravityOverride:Vector="0,0,-980") -> (Pose) {
    angularDampingOverride=1
    angularSpringConstant=7500
    bAngularSpring=true
    bOverrideAngularDamping=true
    bUseGravityOverride=true
    bUseSphericalLimits=true
    gravityScale=0.35
    simulationSpace=RootRelative
    alphaBoolBlend {blendOption=Linear}
    boundBone {boneName=dyn_ponytail_02}
    componentAppliedLinearAccClamp=<10000 10000 10000>
    componentLinearAccScale=<1 1 1>
    componentLinearVelScale=<1 1 1>
    gravityOverride {z=-980}
    physicsBodyDefinitions [{
      boundBone {boneName=dyn_ponytail_02}
      boxExtents=<18 23 23>
      constraintSetup {angularLimitsMax=<0 10 0>, angularLimitsMin=<0 -10 -35>, angularTarget=<1 0 0>}
      localJointOffset=<1 0 0>
     }]
    simSpaceSettings {maxAngularAcceleration=30, maxAngularVelocity=6, simSpaceAngularAlpha=1}
    sphericalLimits [{limitRadius=11, drivingBone {boneName=spine_05}, sphereLocalOffset=<3 1.25 0>}]
   }
   n7 = Anim Dynamics - Bone: dyn_ponytail_03!(ComponentPose=@n6, Alpha=1, GravityOverride:Vector="0,0,-980") -> (Pose) {
    angularDampingOverride=1
    angularSpringConstant=7500
    bAngularSpring=true
    bOverrideAngularDamping=true
    bUseGravityOverride=true
    bUseSphericalLimits=true
    gravityScale=0.35
    simulationSpace=RootRelative
    alphaBoolBlend {blendOption=Linear}
    boundBone {boneName=dyn_ponytail_03}
    componentAppliedLinearAccClamp=<10000 10000 10000>
    componentLinearAccScale=<1 1 1>
    componentLinearVelScale=<1 1 1>
    gravityOverride {z=-980}
    physicsBodyDefinitions [{
      boundBone {boneName=dyn_ponytail_03}
      boxExtents=<16 21 21>
      constraintSetup {angularLimitsMax=<0 10 0>, angularLimitsMin=<0 -10 -40>, angularTarget=<1 0 0>}
      localJointOffset=<1 0 0>
     }]
    simSpaceSettings {maxAngularAcceleration=30, maxAngularVelocity=6, simSpaceAngularAlpha=1}
    sphericalLimits [
     {limitRadius=11, drivingBone {boneName=spine_05}, sphereLocalOffset=<3 1.25 0>}
     {limitRadius=10, drivingBone {boneName=spine_05}, sphereLocalOffset=<3 1 5>}
     {limitRadius=10, drivingBone {boneName=spine_05}, sphereLocalOffset=<3 1 -5>}
    ]
   }
   n8 = Anim Dynamics - Bone: dyn_ponytail_04!(ComponentPose=@n7, Alpha=1, GravityOverride:Vector="0,0,-980") -> (Pose) {
    angularDampingOverride=1
    angularSpringConstant=7500
    bAngularSpring=true
    bOverrideAngularDamping=true
    bUseGravityOverride=true
    bUseSphericalLimits=true
    gravityScale=0.35
    simulationSpace=RootRelative
    alphaBoolBlend {blendOption=Linear}
    boundBone {boneName=dyn_ponytail_04}
    componentAppliedLinearAccClamp=<10000 10000 10000>
    componentLinearAccScale=<1 1 1>
    componentLinearVelScale=<1 1 1>
    gravityOverride {z=-980}
    physicsBodyDefinitions [{
      boundBone {boneName=dyn_ponytail_04}
      boxExtents=<14 19 19>
      constraintSetup {angularLimitsMax=<0 10 0>, angularLimitsMin=<0 -10 -50>, angularTarget=<1 0 0>}
      localJointOffset=<2 0 0>
     }]
    simSpaceSettings {maxAngularAcceleration=30, maxAngularVelocity=6, simSpaceAngularAlpha=1}
    sphericalLimits [
     {limitRadius=11, drivingBone {boneName=spine_05}, sphereLocalOffset=<3 1.25 0>}
     {limitRadius=10, drivingBone {boneName=spine_05}, sphereLocalOffset=<3 1 5>}
     {limitRadius=10, drivingBone {boneName=spine_05}, sphereLocalOffset=<3 1 -5>}
    ]
   }
   n9 = "Transform (Modify) Bone - Bone: dyn_ponytail_01"!(ComponentPose=Local To Component(LocalPose=Use cached pose 'PreDynamicsCache'()), Alpha=1, Rotation="-2,0,0", Scale="1,0.8,0.85") -> (Pose) {
    rotationMode=BMM_Additive
    scaleMode=BMM_Additive
    scaleSpace=BCS_BoneSpace
    boneToModify {boneName=dyn_ponytail_01}
    rotation {pitch=-2}
    scale {y=0.8, z=0.85}
   }
   Save cached pose 'postPonytail_Cache'(Pose=Blend Poses by bool(BlendPose_0=Component To Local(ComponentPose=@n9), BlendPose_1=Component To Local(ComponentPose=@n8))) {cacheName=postPonytail_Cache}
   n10 = RigidBody(
    ComponentPose=Local To Component(LocalPose=Use cached pose 'postPonytail_Cache'())
    OverrideWorldGravity="0,0,-980"
    ExternalForce="0,0,-1000"
   ) {
    bOverrideWorldGravity=true
    overridePhysicsAsset=&F_MED_IvoryBrite_FaceAcc_Physics
    simulationSpace=BaseBoneSpace
    alphaBoolBlend {blendOption=Linear}
    baseBoneRef {boneName=neck_01}
    componentAppliedLinearAccClamp=<3000 3000 3000>
    overrideWorldGravity {z=-980}
    simSpaceSettings {
     dampingAlpha=0
     maxAngularAcceleration=500
     maxAngularVelocity=3
     maxLinearAcceleration=250
     maxLinearVelocity=200
     velocityScaleZ=0.1
     worldAlpha=1
    }
   }
   n11 = RigidBody(ComponentPose=Local To Component(LocalPose=Use cached pose 'postPonytail_Cache'()), OverrideWorldGravity="0,0,0") {
    bOverrideWorldGravity=true
    overridePhysicsAsset=&F_MED_IvoryBrite_FaceAcc_Physics
    simulationSpace=BaseBoneSpace
    alphaBoolBlend {blendOption=Linear}
    baseBoneRef {boneName=neck_01}
    componentAppliedLinearAccClamp=<3000 3000 3000>
    overrideWorldGravity {z=-980}
    simSpaceSettings {
     dampingAlpha=0
     maxAngularAcceleration=100
     maxAngularVelocity=1.5
     maxLinearAcceleration=200
     maxLinearVelocity=150
     velocityScaleZ=0.15
     worldAlpha=1
    }
   }
   n12 = Anim Dynamics - Bone: dyn_ribbon_l_01!(ComponentPose=Local To Component(LocalPose=Use cached pose 'postPonytail_Cache'()), Alpha=1, GravityOverride:Vector="0,0,-980") -> (Pose) {
    angularDampingOverride=1
    angularSpringConstant=10000
    bAngularSpring=true
    bOverrideAngularDamping=true
    bUseGravityOverride=true
    gravityScale=0.35
    simulationSpace=RootRelative
    boundBone {boneName=dyn_ribbon_l_01}
    componentAppliedLinearAccClamp=<5000 5000 5000>
    componentLinearAccScale=<1 1 1>
    componentLinearVelScale=<1 1 1>
    gravityOverride {z=-980}
    physicsBodyDefinitions [{
      boundBone {boneName=dyn_ribbon_l_01}
      boxExtents=<25 35 35>
      constraintSetup {angularLimitsMax=<0 3 15>, angularLimitsMin=<0 -3 -15>, angularTarget=<1 0 0>}
      localJointOffset=<2 0 0>
     }]
    simSpaceSettings {maxAngularAcceleration=36, maxAngularVelocity=6, simSpaceAngularAlpha=1}
   }
   n13 = Anim Dynamics - Bone: dyn_ribbon_l_02!(ComponentPose=@n12, Alpha=1, GravityOverride:Vector="0,0,-980") -> (Pose) {
    angularDampingOverride=1
    angularSpringConstant=10000
    bAngularSpring=true
    bOverrideAngularDamping=true
    bUseGravityOverride=true
    gravityScale=0.35
    simulationSpace=RootRelative
    boundBone {boneName=dyn_ribbon_l_02}
    componentAppliedLinearAccClamp=<5000 5000 5000>
    componentLinearAccScale=<1 1 1>
    componentLinearVelScale=<1 1 1>
    gravityOverride {z=-980}
    physicsBodyDefinitions [{
      boundBone {boneName=dyn_ribbon_l_02}
      boxExtents=<25 35 35>
      constraintSetup {angularLimitsMax=<0 3 15>, angularLimitsMin=<0 -3 -15>, angularTarget=<1 0 0>}
      localJointOffset=<2 0 0>
     }]
    simSpaceSettings {maxAngularAcceleration=36, maxAngularVelocity=6, simSpaceAngularAlpha=1}
   }
   n14 = Anim Dynamics - Bone: dyn_ribbon_l_03!(ComponentPose=@n13, Alpha=1, GravityOverride:Vector="0,0,-980") -> (Pose) {
    angularDampingOverride=1
    angularSpringConstant=10000
    bAngularSpring=true
    bOverrideAngularDamping=true
    bUseGravityOverride=true
    gravityScale=0.35
    simulationSpace=RootRelative
    boundBone {boneName=dyn_ribbon_l_03}
    componentAppliedLinearAccClamp=<5000 5000 5000>
    componentLinearAccScale=<1 1 1>
    componentLinearVelScale=<1 1 1>
    gravityOverride {z=-980}
    physicsBodyDefinitions [{
      boundBone {boneName=dyn_ribbon_l_03}
      boxExtents=<25 35 35>
      constraintSetup {angularLimitsMax=<0 3 15>, angularLimitsMin=<0 -3 -15>, angularTarget=<1 0 0>}
      localJointOffset=<2 0 0>
     }]
    simSpaceSettings {maxAngularAcceleration=36, maxAngularVelocity=6, simSpaceAngularAlpha=1}
   }
   n15 = Anim Dynamics - Bone: dyn_ribbon_l_04!(ComponentPose=@n14, Alpha=1, GravityOverride:Vector="0,0,-980") -> (Pose) {
    angularDampingOverride=1
    angularSpringConstant=10000
    bAngularSpring=true
    bOverrideAngularDamping=true
    bUseGravityOverride=true
    gravityScale=0.35
    simulationSpace=RootRelative
    boundBone {boneName=dyn_ribbon_l_04}
    componentAppliedLinearAccClamp=<5000 5000 5000>
    componentLinearAccScale=<1 1 1>
    componentLinearVelScale=<1 1 1>
    gravityOverride {z=-980}
    physicsBodyDefinitions [{
      boundBone {boneName=dyn_ribbon_l_04}
      boxExtents=<25 35 35>
      constraintSetup {angularLimitsMax=<0 3 15>, angularLimitsMin=<0 -3 -15>, angularTarget=<1 0 0>}
      localJointOffset=<2 0 0>
     }]
    simSpaceSettings {maxAngularAcceleration=36, maxAngularVelocity=6, simSpaceAngularAlpha=1}
   }
   n16 = Anim Dynamics - Bone: dyn_ribbon_l_05!(ComponentPose=@n15, Alpha=1, GravityOverride:Vector="0,0,-980") -> (Pose) {
    angularDampingOverride=1
    angularSpringConstant=10000
    bAngularSpring=true
    bOverrideAngularDamping=true
    bUseGravityOverride=true
    gravityScale=0.35
    simulationSpace=RootRelative
    boundBone {boneName=dyn_ribbon_l_05}
    componentAppliedLinearAccClamp=<5000 5000 5000>
    componentLinearAccScale=<1 1 1>
    componentLinearVelScale=<1 1 1>
    gravityOverride {z=-980}
    physicsBodyDefinitions [{
      boundBone {boneName=dyn_ribbon_l_05}
      boxExtents=<25 35 35>
      constraintSetup {angularLimitsMax=<0 3 15>, angularLimitsMin=<0 -3 -15>, angularTarget=<1 0 0>}
      localJointOffset=<2 0 0>
     }]
    simSpaceSettings {maxAngularAcceleration=36, maxAngularVelocity=6, simSpaceAngularAlpha=1}
   }
   n17 = Anim Dynamics - Bone: dyn_ribbon_l_06!(ComponentPose=@n16, Alpha=1, GravityOverride:Vector="0,0,-980") -> (Pose) {
    angularDampingOverride=1
    angularSpringConstant=10000
    bAngularSpring=true
    bOverrideAngularDamping=true
    bUseGravityOverride=true
    gravityScale=0.35
    simulationSpace=RootRelative
    boundBone {boneName=dyn_ribbon_l_06}
    componentAppliedLinearAccClamp=<5000 5000 5000>
    componentLinearAccScale=<1 1 1>
    componentLinearVelScale=<1 1 1>
    gravityOverride {z=-980}
    physicsBodyDefinitions [{
      boundBone {boneName=dyn_ribbon_l_06}
      boxExtents=<25 35 35>
      constraintSetup {angularLimitsMax=<0 3 15>, angularLimitsMin=<0 -3 -15>, angularTarget=<1 0 0>}
      localJointOffset=<2 0 0>
     }]
    simSpaceSettings {maxAngularAcceleration=36, maxAngularVelocity=6, simSpaceAngularAlpha=1}
   }
   n18 = Anim Dynamics - Bone: dyn_ribbon_l_07!(ComponentPose=@n17, Alpha=1, GravityOverride:Vector="0,0,-980") -> (Pose) {
    angularDampingOverride=1
    angularSpringConstant=10000
    bAngularSpring=true
    bOverrideAngularDamping=true
    bUseGravityOverride=true
    gravityScale=0.35
    simulationSpace=RootRelative
    boundBone {boneName=dyn_ribbon_l_07}
    componentAppliedLinearAccClamp=<5000 5000 5000>
    componentLinearAccScale=<1 1 1>
    componentLinearVelScale=<1 1 1>
    gravityOverride {z=-980}
    physicsBodyDefinitions [{
      boundBone {boneName=dyn_ribbon_l_07}
      boxExtents=<25 35 35>
      constraintSetup {angularLimitsMax=<0 3 15>, angularLimitsMin=<0 -3 -15>, angularTarget=<1 0 0>}
      localJointOffset=<2 0 0>
     }]
    simSpaceSettings {maxAngularAcceleration=36, maxAngularVelocity=6, simSpaceAngularAlpha=1}
   }
   n19 = Anim Dynamics - Bone: dyn_ribbon_r_01!(ComponentPose=@n18, Alpha=1, GravityOverride:Vector="0,0,-980") -> (Pose) {
    angularDampingOverride=1
    angularSpringConstant=10000
    bAngularSpring=true
    bOverrideAngularDamping=true
    bUseGravityOverride=true
    gravityScale=0.35
    simulationSpace=RootRelative
    boundBone {boneName=dyn_ribbon_r_01}
    componentAppliedLinearAccClamp=<5000 5000 5000>
    componentLinearAccScale=<1 1 1>
    componentLinearVelScale=<1 1 1>
    gravityOverride {z=-980}
    physicsBodyDefinitions [{
      boundBone {boneName=dyn_ribbon_r_01}
      boxExtents=<25 35 35>
      constraintSetup {angularLimitsMax=<0 3 15>, angularLimitsMin=<0 -3 -15>, angularTarget=<1 0 0>}
      localJointOffset=<4 0 0>
     }]
    simSpaceSettings {maxAngularAcceleration=36, maxAngularVelocity=6, simSpaceAngularAlpha=1}
   }
   n20 = Anim Dynamics - Bone: dyn_ribbon_r_02!(ComponentPose=@n19, Alpha=1, GravityOverride:Vector="0,0,-980") -> (Pose) {
    angularDampingOverride=1
    angularSpringConstant=10000
    bAngularSpring=true
    bOverrideAngularDamping=true
    bUseGravityOverride=true
    gravityScale=0.35
    simulationSpace=RootRelative
    boundBone {boneName=dyn_ribbon_r_02}
    componentAppliedLinearAccClamp=<5000 5000 5000>
    componentLinearAccScale=<1 1 1>
    componentLinearVelScale=<1 1 1>
    gravityOverride {z=-980}
    physicsBodyDefinitions [{
      boundBone {boneName=dyn_ribbon_r_02}
      boxExtents=<25 35 35>
      constraintSetup {angularLimitsMax=<0 3 15>, angularLimitsMin=<0 -3 -15>, angularTarget=<1 0 0>}
      localJointOffset=<4 0 0>
     }]
    simSpaceSettings {maxAngularAcceleration=36, maxAngularVelocity=6, simSpaceAngularAlpha=1}
   }
   n21 = Anim Dynamics - Bone: dyn_ribbon_r_03!(ComponentPose=@n20, Alpha=1, GravityOverride:Vector="0,0,-980") -> (Pose) {
    angularDampingOverride=1
    angularSpringConstant=10000
    bAngularSpring=true
    bOverrideAngularDamping=true
    bUseGravityOverride=true
    gravityScale=0.35
    simulationSpace=RootRelative
    boundBone {boneName=dyn_ribbon_r_03}
    componentAppliedLinearAccClamp=<5000 5000 5000>
    componentLinearAccScale=<1 1 1>
    componentLinearVelScale=<1 1 1>
    gravityOverride {z=-980}
    physicsBodyDefinitions [{
      boundBone {boneName=dyn_ribbon_r_03}
      boxExtents=<25 35 35>
      constraintSetup {angularLimitsMax=<0 3 15>, angularLimitsMin=<0 -3 -15>, angularTarget=<1 0 0>}
      localJointOffset=<4 0 0>
     }]
    simSpaceSettings {maxAngularAcceleration=36, maxAngularVelocity=6, simSpaceAngularAlpha=1}
   }
   n22 = Anim Dynamics - Bone: dyn_ribbon_r_04!(ComponentPose=@n21, Alpha=1, GravityOverride:Vector="0,0,-980") -> (Pose) {
    angularDampingOverride=1
    angularSpringConstant=10000
    bAngularSpring=true
    bOverrideAngularDamping=true
    bUseGravityOverride=true
    gravityScale=0.35
    simulationSpace=RootRelative
    boundBone {boneName=dyn_ribbon_r_04}
    componentAppliedLinearAccClamp=<5000 5000 5000>
    componentLinearAccScale=<1 1 1>
    componentLinearVelScale=<1 1 1>
    gravityOverride {z=-980}
    physicsBodyDefinitions [{
      boundBone {boneName=dyn_ribbon_r_04}
      boxExtents=<25 35 35>
      constraintSetup {angularLimitsMax=<0 3 15>, angularLimitsMin=<0 -3 -15>, angularTarget=<1 0 0>}
      localJointOffset=<4 0 0>
     }]
    simSpaceSettings {maxAngularAcceleration=36, maxAngularVelocity=6, simSpaceAngularAlpha=1}
   }
   n23 = Anim Dynamics - Bone: dyn_ribbon_r_05!(ComponentPose=@n22, Alpha=1, GravityOverride:Vector="0,0,-980") -> (Pose) {
    angularDampingOverride=1
    angularSpringConstant=10000
    bAngularSpring=true
    bOverrideAngularDamping=true
    bUseGravityOverride=true
    gravityScale=0.35
    simulationSpace=RootRelative
    boundBone {boneName=dyn_ribbon_r_05}
    componentAppliedLinearAccClamp=<5000 5000 5000>
    componentLinearAccScale=<1 1 1>
    componentLinearVelScale=<1 1 1>
    gravityOverride {z=-980}
    physicsBodyDefinitions [{
      boundBone {boneName=dyn_ribbon_r_05}
      boxExtents=<25 35 35>
      constraintSetup {angularLimitsMax=<0 3 15>, angularLimitsMin=<0 -3 -15>, angularTarget=<1 0 0>}
      localJointOffset=<4 0 0>
     }]
    simSpaceSettings {maxAngularAcceleration=36, maxAngularVelocity=6, simSpaceAngularAlpha=1}
   }
   n24 = Anim Dynamics - Bone: dyn_ribbon_r_06!(ComponentPose=@n23, Alpha=1, GravityOverride:Vector="0,0,-980") -> (Pose) {
    angularDampingOverride=1
    angularSpringConstant=10000
    bAngularSpring=true
    bOverrideAngularDamping=true
    bUseGravityOverride=true
    gravityScale=0.35
    simulationSpace=RootRelative
    boundBone {boneName=dyn_ribbon_r_06}
    componentAppliedLinearAccClamp=<5000 5000 5000>
    componentLinearAccScale=<1 1 1>
    componentLinearVelScale=<1 1 1>
    gravityOverride {z=-980}
    physicsBodyDefinitions [{
      boundBone {boneName=dyn_ribbon_r_06}
      boxExtents=<25 35 35>
      constraintSetup {angularLimitsMax=<0 3 15>, angularLimitsMin=<0 -3 -15>, angularTarget=<1 0 0>}
      localJointOffset=<4 0 0>
     }]
    simSpaceSettings {maxAngularAcceleration=36, maxAngularVelocity=6, simSpaceAngularAlpha=1}
   }
   n25 = Anim Dynamics - Bone: dyn_ribbon_r_07!(ComponentPose=@n24, Alpha=1, GravityOverride:Vector="0,0,-980") -> (Pose) {
    angularDampingOverride=1
    angularSpringConstant=10000
    bAngularSpring=true
    bOverrideAngularDamping=true
    bUseGravityOverride=true
    gravityScale=0.35
    simulationSpace=RootRelative
    boundBone {boneName=dyn_ribbon_r_07}
    componentAppliedLinearAccClamp=<5000 5000 5000>
    componentLinearAccScale=<1 1 1>
    componentLinearVelScale=<1 1 1>
    gravityOverride {z=-980}
    physicsBodyDefinitions [{
      boundBone {boneName=dyn_ribbon_r_07}
      boxExtents=<25 35 35>
      constraintSetup {angularLimitsMax=<0 3 15>, angularLimitsMin=<0 -3 -15>, angularTarget=<1 0 0>}
      localJointOffset=<4 0 0>
     }]
    simSpaceSettings {maxAngularAcceleration=36, maxAngularVelocity=6, simSpaceAngularAlpha=1}
   }
   n26 = Anim Dynamics - Bone: dyn_ribbon_r_08!(ComponentPose=@n25, Alpha=1, GravityOverride:Vector="0,0,-980") -> (Pose) {
    angularDampingOverride=1
    angularSpringConstant=10000
    bAngularSpring=true
    bOverrideAngularDamping=true
    bUseGravityOverride=true
    gravityScale=0.35
    simulationSpace=RootRelative
    boundBone {boneName=dyn_ribbon_r_08}
    componentAppliedLinearAccClamp=<5000 5000 5000>
    componentLinearAccScale=<1 1 1>
    componentLinearVelScale=<1 1 1>
    gravityOverride {z=-980}
    physicsBodyDefinitions [{
      boundBone {boneName=dyn_ribbon_r_08}
      boxExtents=<25 35 35>
      constraintSetup {angularLimitsMax=<0 3 15>, angularLimitsMin=<0 -3 -15>, angularTarget=<1 0 0>}
      localJointOffset=<4 0 0>
     }]
    simSpaceSettings {maxAngularAcceleration=36, maxAngularVelocity=6, simSpaceAngularAlpha=1}
   }
   n27 = Blend Poses by bool(
    BlendPose_0=Blend Poses by bool(BlendPose_0=Component To Local(ComponentPose=@n10), BlendPose_1=Component To Local(ComponentPose=@n11), BlendTime_0=5)
    BlendPose_1=Component To Local(ComponentPose=@n26)
   )
   n28 = Use cached pose 'postPonytail_Cache'()
   n29 = Anim Dynamics - Bone: dyn_NoMask_frontHair_c!(
    ComponentPose=Local To Component(LocalPose=Blend Poses by bool(BlendPose_0=@n28, BlendPose_1=@n27))
    Alpha=1
    GravityOverride:Vector="0,0,-980"
   ) -> (Pose) {
    angularDampingOverride=1
    angularSpringConstant=5000
    bAngularSpring=true
    bOverrideAngularDamping=true
    bUseGravityOverride=true
    gravityScale=0.35
    simulationSpace=RootRelative
    alphaBoolBlend {blendOption=Linear}
    boundBone {boneName=dyn_NoMask_frontHair_c}
    componentAppliedLinearAccClamp=<10000 10000 10000>
    componentLinearAccScale=<1 1 1>
    componentLinearVelScale=<1 1 1>
    gravityOverride {z=-980}
    physicsBodyDefinitions [{
      boundBone {boneName=dyn_NoMask_frontHair_c}
      boxExtents=<20 15 15>
      constraintSetup {angularLimitsMax=<0 0 15>, angularLimitsMin=<0 -20 -15>, angularTarget=<1 0 0>}
      localJointOffset=<1 0 0>
     }]
    simSpaceSettings {maxAngularAcceleration=30, maxAngularVelocity=6, simSpaceAngularAlpha=1}
   }
   n30 = Anim Dynamics - Bone: dyn_frontHair_NoMask_f_01!(ComponentPose=@n29, Alpha=1, GravityOverride:Vector="0,0,-980") -> (Pose) {
    angularDampingOverride=1
    angularSpringConstant=5000
    bAngularSpring=true
    bOverrideAngularDamping=true
    bUseGravityOverride=true
    gravityScale=0.35
    simulationSpace=RootRelative
    alphaBoolBlend {blendOption=Linear}
    boundBone {boneName=dyn_frontHair_NoMask_f_01}
    componentAppliedLinearAccClamp=<10000 10000 10000>
    componentLinearAccScale=<1 1 1>
    componentLinearVelScale=<1 1 1>
    gravityOverride {z=-980}
    physicsBodyDefinitions [{
      boundBone {boneName=dyn_frontHair_NoMask_f_01}
      boxExtents=<20 15 15>
      constraintSetup {angularLimitsMax=<0 50 20>, angularLimitsMin=<0 -30 -20>, angularTarget=<1 0 0>}
      localJointOffset=<1 0 0>
     }]
    simSpaceSettings {maxAngularAcceleration=30, maxAngularVelocity=6, simSpaceAngularAlpha=1}
   }
   n31 = Anim Dynamics - Bone: dyn_frontHair_NoMask_f_02!(ComponentPose=@n30, Alpha=1, GravityOverride:Vector="0,0,-980") -> (Pose) {
    angularDampingOverride=1
    angularSpringConstant=5000
    bAngularSpring=true
    bOverrideAngularDamping=true
    bUseGravityOverride=true
    gravityScale=0.35
    simulationSpace=RootRelative
    alphaBoolBlend {blendOption=Linear}
    boundBone {boneName=dyn_frontHair_NoMask_f_02}
    componentAppliedLinearAccClamp=<10000 10000 10000>
    componentLinearAccScale=<1 1 1>
    componentLinearVelScale=<1 1 1>
    gravityOverride {z=-980}
    physicsBodyDefinitions [{
      boundBone {boneName=dyn_frontHair_NoMask_f_02}
      boxExtents=<20 15 15>
      constraintSetup {angularLimitsMax=<0 50 25>, angularLimitsMin=<0 -40 -25>, angularTarget=<1 0 0>}
      localJointOffset=<1 0 0>
     }]
    simSpaceSettings {maxAngularAcceleration=30, maxAngularVelocity=6, simSpaceAngularAlpha=1}
   }
   n32 = Anim Dynamics - Bone: dyn_frontHair_NoMask_b_01!(ComponentPose=@n31, Alpha=1, GravityOverride:Vector="0,0,-980") -> (Pose) {
    angularDampingOverride=1
    angularSpringConstant=5000
    bAngularSpring=true
    bOverrideAngularDamping=true
    bUseGravityOverride=true
    gravityScale=0.35
    simulationSpace=RootRelative
    alphaBoolBlend {blendOption=Linear}
    boundBone {boneName=dyn_frontHair_NoMask_b_01}
    componentAppliedLinearAccClamp=<10000 10000 10000>
    componentLinearAccScale=<1 1 1>
    componentLinearVelScale=<1 1 1>
    gravityOverride {z=-980}
    physicsBodyDefinitions [{
      boundBone {boneName=dyn_frontHair_NoMask_b_01}
      boxExtents=<20 15 15>
      constraintSetup {angularLimitsMax=<0 50 20>, angularLimitsMin=<0 -30 -20>, angularTarget=<1 0 0>}
      localJointOffset=<1 0 0>
     }]
    simSpaceSettings {maxAngularAcceleration=30, maxAngularVelocity=6, simSpaceAngularAlpha=1}
   }
   n33 = Anim Dynamics - Bone: dyn_frontHair_NoMask_b_02!(ComponentPose=@n32, Alpha=1, GravityOverride:Vector="0,0,-980") -> (Pose) {
    angularDampingOverride=1
    angularSpringConstant=5000
    bAngularSpring=true
    bOverrideAngularDamping=true
    bUseGravityOverride=true
    gravityScale=0.35
    simulationSpace=RootRelative
    alphaBoolBlend {blendOption=Linear}
    boundBone {boneName=dyn_frontHair_NoMask_b_02}
    componentAppliedLinearAccClamp=<10000 10000 10000>
    componentLinearAccScale=<1 1 1>
    componentLinearVelScale=<1 1 1>
    gravityOverride {z=-980}
    physicsBodyDefinitions [{
      boundBone {boneName=dyn_frontHair_NoMask_b_02}
      boxExtents=<20 15 15>
      constraintSetup {angularLimitsMax=<0 50 25>, angularLimitsMin=<0 -40 -25>, angularTarget=<1 0 0>}
      localJointOffset=<1 0 0>
     }]
    simSpaceSettings {maxAngularAcceleration=30, maxAngularVelocity=6, simSpaceAngularAlpha=1}
   }
   n34 = Anim Dynamics - Bone: dyn_frontRibbonCenter!(ComponentPose=@n33, Alpha=0.8, GravityOverride:Vector="0,0,0") -> (Pose) {
    angularDampingOverride=0.9999
    angularSpringConstant=14000
    bAngularSpring=true
    bOverrideAngularDamping=true
    bUseGravityOverride=true
    gravityScale=0.35
    simulationSpace=RootRelative
    alphaBoolBlend {blendOption=Linear}
    boundBone {boneName=dyn_frontRibbonCenter}
    componentAppliedLinearAccClamp=<5000 5000 5000>
    componentLinearAccScale=<1 1 1>
    componentLinearVelScale=<1 1 1>
    gravityOverride {z=-980}
    physicsBodyDefinitions [{
      boundBone {boneName=dyn_frontRibbonCenter}
      boxExtents=<30 35 45>
      constraintSetup {angularLimitsMax=<15 5 15>, angularLimitsMin=<-15 -5 -15>, angularTarget=<1 0 0>}
      localJointOffset=<1 0 0>
     }]
    simSpaceSettings {maxAngularAcceleration=30, maxAngularVelocity=6, simSpaceAngularAlpha=1}
   }
   n35 = Anim Dynamics - Bone: dyn_RibbonUp!(ComponentPose=@n34, Alpha=0.6, GravityOverride:Vector="0,0,0") -> (Pose) {
    angularDampingOverride=0.9999
    angularSpringConstant=14000
    bAngularSpring=true
    bOverrideAngularDamping=true
    bUseGravityOverride=true
    gravityScale=0.35
    simulationSpace=RootRelative
    alphaBoolBlend {blendOption=Linear}
    boundBone {boneName=dyn_RibbonUp}
    componentAppliedLinearAccClamp=<5000 5000 5000>
    componentLinearAccScale=<1 1 1>
    componentLinearVelScale=<1 1 1>
    gravityOverride {z=-980}
    physicsBodyDefinitions [{
      boundBone {boneName=dyn_RibbonUp}
      boxExtents=<30 35 45>
      constraintSetup {angularLimitsMax=<15 15 15>, angularLimitsMin=<-15 -15 -15>, angularTarget=<1 0 0>}
      localJointOffset=<0.5 0 0>
     }]
    simSpaceSettings {maxAngularAcceleration=30, maxAngularVelocity=6, simSpaceAngularAlpha=1}
   }
   n36 = Anim Dynamics - Bone: dyn_RibbonDn!(ComponentPose=@n35, Alpha=0.8, GravityOverride:Vector="0,0,0") -> (Pose) {
    angularDampingOverride=0.9999
    angularSpringConstant=14000
    bAngularSpring=true
    bOverrideAngularDamping=true
    bUseGravityOverride=true
    gravityScale=0.35
    simulationSpace=RootRelative
    alphaBoolBlend {blendOption=Linear}
    boundBone {boneName=dyn_RibbonDn}
    componentAppliedLinearAccClamp=<5000 5000 5000>
    componentLinearAccScale=<1 1 1>
    componentLinearVelScale=<1 1 1>
    gravityOverride {z=-980}
    physicsBodyDefinitions [{
      boundBone {boneName=dyn_RibbonDn}
      boxExtents=<30 35 45>
      constraintSetup {angularLimitsMax=<15 15 15>, angularLimitsMin=<-15 -15 -15>, angularTarget=<1 0 0>}
      localJointOffset=<0.5 0 0>
     }]
    simSpaceSettings {maxAngularAcceleration=30, maxAngularVelocity=6, simSpaceAngularAlpha=1}
   }
   Save cached pose 'DynamicsCache'(Pose=Component To Local(ComponentPose=@n36)) {cacheName=DynamicsCache}
   n37 = Layered blend per bone(
    BasePose=Copy Pose From Mesh() {bUseAttachedParent=true}
    BlendPoses_0=Copy Pose From Mesh(SourceMeshComponent=Get BodyMesh() -> (BodyMesh:Skeletal Mesh Component Object Reference)) {bUseAttachedParent=true}
   ) {layerSetup [{branchFilters [{boneName=upperarm_r}, {boneName=upperarm_l}, {boneName=thigh_r}, {boneName=thigh_l}]}]}
   n38 = Control Rig!(
    Source=@n37
    Alpha
    isTargeting:bool=false
    isMoving2D:bool=false
    isTacticalJumping:bool=false
    isSkyFloating:bool=false
    RigidBodiesEnabled:bool=false
    animDynamicsEnabled:bool=false
    Moving2D:bool=false
    isWearingBackBling:bool=false
    isMeleeSwinging:bool=false
    isWearingCape:bool=false
    isPlayingEmote:bool=false
    isFrontEnd:bool=false
    isParachuteOpened:bool=false
   ) -> (Pose) {controlRigAssetReference {blueprintRigClass=&/F_MED_IvoryBrite_FaceAcc_PRE_CtrlRig.F_MED_IvoryBrite_FaceAcc_PRE_CtrlRig_C}}
   Save cached pose 'CRCache_Input'(Pose=@n38) {cacheName=CRCache_Input}
   Save cached pose 'PostDynamics_Cache'(Pose=Use cached pose 'DynamicsCache'()) {cacheName=PostDynamics_Cache}
   n39 = "Transform (Modify) Bone - Bone: base_ribbon_l_01"(
    ComponentPose=Local To Component(LocalPose=Use cached pose 'CRCache_Input'())
    Alpha=1
    Translation="0,0,0"
    Rotation=Get HairRotator_01() -> (HairRotator_01:Rotator)
    Scale="1,1,1"
   ) {rotationMode=BMM_Additive, rotationSpace=BCS_BoneSpace, alphaBoolBlend {blendOption=Linear}, boneToModify {boneName=base_ribbon_l_01}}
   n40 = "Transform (Modify) Bone - Bone: base_ribbon_r_01"(ComponentPose=@n39, Alpha=1, Translation="0,0,0", Rotation=Get HairRotator_02() -> (HairRotator_02:Rotator), Scale="1,1,1") {rotationMode=BMM_Additive, rotationSpace=BCS_BoneSpace, alphaBoolBlend {blendOption=Linear}, boneToModify {boneName=base_ribbon_r_01}}
   Save cached pose 'PreDynamicsCache'(Pose=Blend Poses by bool(BlendPose_0=Use cached pose 'CRCache_Input'(), BlendPose_1=Component To Local(ComponentPose=@n40), BlendTime_0=0.25, BlendTime_1=0.25)) {cacheName=PreDynamicsCache}
  ]}
 EventGraph {nodes [
   n1 = Event BlueprintUpdateAnimation()
   > Set Delta Time X(Delta Time X=@n1.DeltaTimeX)
   > n2 = Sequence!(execute) -> (then_0, then_1, then_2:Exec, then_3:Exec)
   > n3 = Sequence!(execute) -> (then_0, then_1)
   Set Delta Time S(execute=@n3.then_1, Delta Time S:float=float + float(A=@n1.DeltaTimeX, B=Get Delta Time S() -> (Delta Time S:float))) -> (Output_Get:float)
   > n4 = "Gravity Override State Function (Parameter Asset) (Nativized)"(
    CharacterPartAnimInstance=Self-Reference()
    GameTimeInSeconds=GetGameTimeInSeconds()
    Parameters=/BRCosmetics/Characters/Player/Female/Medium/Bodies/F_MED_IvoryBrite/Meshes/F_MED_IvoryBrite_Parameters
   )
   > Set GravityOverride(GravityOverride=@n4)
   > n5 = Branch(Condition=OR Boolean(A=Get bAnimDynamicsEnabled() -> (bAnimDynamicsEnabled:bool), B=Get bRigidBodyEnabled()))
   > Set TempRigidBod(TempRigidBod:bool=true) -> (Output_Get:bool)
   > n6 = Branch(Condition=Get bRigidBodyEnabled())
   > n7 = RandomRotatorFunction!(
    execute
    GameTimeInSeconds:double=Get Delta Time S()
    PerlinScale:double=1
    Frequency:Rotator="1,3,0"
    Amplitude:Rotator="10,22,0"
    Offset:Rotator="0,-20,0"
    TimeOffset:double=0
   ) -> (then, RandomRotator:Rotator)
   > Set HairRotator_01(HairRotator_01:Rotator=@n7) -> (Output_Get:Rotator)
   > n8 = RandomRotatorFunction(GameTimeInSeconds=Get Delta Time S(), PerlinScale=1, Frequency="2.5,4.5,0", Amplitude="8,-18,0", Offset="0,-30,0", TimeOffset=20)
   > Set HairRotator_02(HairRotator_02:Rotator=@n8) -> (Output_Get:Rotator)
   n9 = RandomRotatorFunction(
    execute=@n6.else
    GameTimeInSeconds=Get Delta Time S()
    PerlinScale=2
    Frequency="2.5,3.5,0"
    Amplitude="5,15,0"
    Offset="0,-55,0"
    TimeOffset=0
   )
   > Set HairRotator_01(HairRotator_01=@n9)
   > n10 = RandomRotatorFunction(GameTimeInSeconds=Get Delta Time S(), PerlinScale=2, Frequency="2.5,3.5,0", Amplitude="5,-15,0", Offset="0,-55,0", TimeOffset=0)
   > Set HairRotator_02(HairRotator_02=@n10)
   Set HairRotator_01(execute=@n5.else, HairRotator_01="0, 0, 0")
   > Set HairRotator_02(HairRotator_02="0, 0, 0")
   n11 = Branch(execute=@n2.then_1, Condition=Get bIsSurfaceSwimming())
   > Set RotationAlpha(RotationAlpha:double=0.5) -> (Output_Get:double)
   Set RotationAlpha(execute=@n11.else, RotationAlpha=1)
   n12 = Branch(execute=@n2.then_2, Condition=Get bIsSprinting())
   > Set RIbbonGravity(RIbbonGravity:Vector="0,-50000,-75") -> (Output_Get:Vector)
   Set RIbbonGravity(execute=@n12.else, RIbbonGravity=Get FloatyGravity() -> (FloatyGravity:Vector))
   n13 = BreakVector(InVec=Get RIbbonGravity_Interp() -> (RIbbonGravity_Interp:Vector))
   n14 = BreakVector(InVec=Get RIbbonGravity() -> (RIbbonGravity:Vector))
   n15 = Get Delta Time X()
   Set RIbbonGravity_Interp(execute=@n2.then_3, RIbbonGravity_Interp:Vector=MakeVector(Y=0, Z=FInterpTo(Current=@n13.Z, Target=@n14.Z, DeltaTime=@n15, InterpSpeed=5))) -> (Output_Get:Vector)
   > Set RotationAlphaInterp(RotationAlphaInterp:double=FInterpTo(
     Current=Get RotationAlphaInterp() -> (RotationAlphaInterp:double)
     Target=Get RotationAlpha() -> (RotationAlpha:double)
     DeltaTime=Get Delta Time X()
     InterpSpeed=4.75
    )) -> (Output_Get:double)
   > Set backBlingDBNO(backBlingDBNO:bool=OR Boolean!(A=Get isWearingBackbling(), B=Get bIsDBNOCarried(), C:bool=Get bIsDBNOCarrying() -> (bIsDBNOCarrying:bool)) -> (ReturnValue)) -> (Output_Get:bool)
   Event OnEndFalling()
   > Set isTacticalJumping(isTacticalJumping:bool=false) -> (Output_Get:bool)
   > Set MainRB_ExternalForce(MainRB_ExternalForce:Vector="0,0,-1200") -> (Output_Get:Vector)
   > Delay(Duration=2)
   > Set RIbbonGravity(RIbbonGravity=Get FloatyGravity())
   > Set MainRB_ExternalForce(MainRB_ExternalForce="0,0,-50")
   Event OnBeginCrouching()
   > Set RIbbonGravity(RIbbonGravity="0,0,-50")
   n16 = Event OnCharacterCustomizationLoaded()
   > Set BodyMesh(BodyMesh=Get Array Item with EFort Custom Part Type(Array=@n16.CustomMeshParts))
   n17 = CheckForBackbling()
   > Set isWearingBackbling(isWearingBackbling=OR Boolean(A=@n17, B=@n17.bIsBackblingACape, C=@n17.bIsBackblingLong))
   > Set isWearingCape(isWearingCape=@n17.bIsBackblingACape)
   Event OnEndCrouching()
   > Delay(Duration=1)
   > Set RIbbonGravity(RIbbonGravity=Get FloatyGravity())
   Event OnBeginJumping()
   > n18 = Set isTacticalJumping(isTacticalJumping=Get bIsTacticalSprinting())
   > n19 = Branch(Condition=@n18)
   > Set RIbbonGravity(RIbbonGravity="0,0,-100")
   Set RIbbonGravity(execute=@n19.else, RIbbonGravity=Get HeavyGravity() -> (HeavyGravity:Vector))
   RandomRotatorFunction(GameTimeInSeconds=0, PerlinScale=1, Frequency="5.5,5,0", Amplitude="25,25,0", Offset="0,-25,0", TimeOffset=0)
   RandomRotatorFunction(GameTimeInSeconds=0, PerlinScale=1, Frequency="5.5,5,0", Amplitude="-25,-25,0", Offset="0,-35,0", TimeOffset=0)
   Set RIbbonGravity(RIbbonGravity=Get FloatyGravity())
   Set RIbbonGravity(RIbbonGravity=Get HeavyGravity())
  ]}
]
variables [
 backBlingDBNO {type=boolean}
 BodyMesh {type=SkeletalMeshComponent}
 Delta Time S {type=number}
 Delta Time X {type=number}
 FloatyGravity {type=Vector, default=<0 0 -25>}
 GravityOverride {type=Vector}
 HairRotator_01 {type=Rotator}
 HairRotator_02 {type=Rotator}
 HeavyGravity {type=Vector, default=<0 0 -75>}
 isTacticalJumping {type=boolean}
 isWearingBackbling {type=boolean}
 isWearingCape {type=boolean}
 MainRB_ExternalForce {type=Vector, default=<0 0 -50>}
 RIbbonGravity {type=Vector, default=<0 0 -25>}
 RIbbonGravity_Interp {type=Vector, default=<0 0 -75>}
 RotationAlpha {type=number}
 RotationAlphaInterp {type=number}
 TempRigidBod {type=boolean}
]
